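# Physics asset for the SkippingClouds medium male player body, named as kinematic collision for its cloth assets, likely for ML Deformer (MLD) cloth.
# It holds a single pelvis body with no collision shapes listed and no constraints.
# As written, it acts as a near-empty placeholder rather than a full-body collision setup.
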PhysicsAsset /Game/Characters/Player/Male/Medium/Bodies/M_MED_SkippingClouds/Meshes/ClothAssets/M_MED_SkippingClouds_MLDKineCollision
bodies {pelvis []}
constraints []
