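# Chaos Cloth asset for the male medium SkippingClouds player body, named for ML Deformer (MLD) training.
# Its Dataflow merges a remeshed sim mesh with a separately imported render mesh, then binds selected regions via a proxy deformer.
# Painted max-distance and stretch maps feed a config chain through bending, mass, damping, gravity, physics-asset collision and solver.
# A low-iteration test solver branches off collision alongside the one feeding the terminal.
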ChaosClothAsset /Game/Characters/Player/Male/Medium/Bodies/M_MED_SkippingClouds/Meshes/ClothAssets/M_MED_SkippingClouds_CA_MLDTraining
name=EmbeddedDataflow
graph=&self:EmbeddedDataflow
connections [
 {fromNode=SkeletalMeshImport_0, fromPinGuid=#1, toNode=Remesh_0, toPinGuid=#2}
 {
  fromNode=SimulationBendingConfig_01
  fromPinGuid=#3
  toNode=SimulationMassConfig_01
  toPinGuid=#4
 }
 {
  fromNode=SimulationCollisionConfig
  fromPinGuid=#5
  toNode=SimulationSolverConfig_01
  toPinGuid=#6
 }
 {
  fromNode=MaxDistWeightMap
  fromPin=OutputName.StringValue
  fromPinGuid=#7
  toNode=SimulationMaxDistanceConfig
  toPin=MaxDistance.WeightMap
  toPinGuid=#8
 }
 {
  fromNode=SimulationStretchConfig_01
  fromPinGuid=#9
  toNode=SimulationBendingConfig_01
  toPinGuid=#10
 }
 {
  fromNode=SimulationMassConfig_01
  fromPinGuid=#11
  toNode=SimulationDampingConfig_01
  toPinGuid=#12
 }
 {
  fromNode=SimulationCollisionConfig
  fromPinGuid=#5
  toNode=LOW ITERATION FOR TESTING SimulationSolverConfig
  toPinGuid=#13
 }
 {
  fromNode=SimulationMaxDistanceConfig
  fromPinGuid=#14
  toNode=StretchWeightMap
  toPinGuid=#15
 }
 {
  fromNode=Selection_2
  fromPin=OutputName.StringValue
  fromPinGuid=#16
  toNode=ProxyDeformer
  toPin="SelectionFilterSets[0].RenderSelection.StringValue"
  toPinGuid=#17
 }
 {
  fromNode=SkeletalMeshImport_00
  fromPinGuid=#18
  toNode=MergeClothCollections
  toPin="Collections[1]"
  toPinGuid=#19
 }
 {
  fromNode=StretchWeightMap
  fromPin=OutputName.StringValue
  fromPinGuid=#20
  toNode=SimulationStretchConfig_01
  toPin=StretchStiffnessWeft.WeightMap
  toPinGuid=#21
 }
 {
  fromNode=SimulationSolverConfig_01
  fromPinGuid=#22
  toNode=ClothAssetTerminal
  toPin="CollectionLods[0]"
  toPinGuid=#23
 }
 {fromNode=Selection_02, fromPinGuid=#24, toNode=ProxyDeformer, toPinGuid=#25}
 {
  fromNode=SimulationGravityConfig
  fromPinGuid=#26
  toNode=SetPhysicsAsset
  toPinGuid=#27
 }
 {
  fromNode=StretchWeightMap
  fromPinGuid=#28
  toNode=SimulationStretchConfig_01
  toPinGuid=#29
 }
 {
  fromNode=Selection_02
  fromPin=OutputName.StringValue
  fromPinGuid=#30
  toNode=ProxyDeformer
  toPin="SelectionFilterSets[0].SimSelection.StringValue"
  toPinGuid=#31
 }
 {fromNode=ProxyDeformer, fromPinGuid=#32, toNode=MaxDistWeightMap, toPinGuid=#33}
 {fromNode=Selection_2, fromPinGuid=#34, toNode=Selection_02, toPinGuid=#35}
 {
  fromNode=StretchWeightMap
  fromPin=OutputName.StringValue
  fromPinGuid=#20
  toNode=SimulationStretchConfig_01
  toPin=StretchStiffnessBias.WeightMap
  toPinGuid=#36
 }
 {
  fromNode=SimulationDampingConfig_01
  fromPinGuid=#37
  toNode=SimulationGravityConfig
  toPinGuid=#38
 }
 {
  fromNode=Remesh_0
  fromPinGuid=#39
  toNode=MergeClothCollections
  toPin="Collections[0]"
  toPinGuid=#40
 }
 {fromNode=MergeClothCollections, fromPinGuid=#41, toNode=Selection_2, toPinGuid=#42}
 {
  fromNode=StretchWeightMap
  fromPin=OutputName.StringValue
  fromPinGuid=#20
  toNode=SimulationStretchConfig_01
  toPin=StretchStiffnessWarp.WeightMap
  toPinGuid=#43
 }
 {
  fromNode=SetPhysicsAsset
  fromPinGuid=#44
  toNode=SimulationCollisionConfig
  toPinGuid=#45
 }
 {
  fromNode=MaxDistWeightMap
  fromPinGuid=#46
  toNode=SimulationMaxDistanceConfig
  toPinGuid=#47
 }
]
nodes [
 SimulationMaxDistanceConfig {
  type=FChaosClothAssetSimulationMaxDistanceConfigNode
  guid=#48
  properties {MaxDistance=(bIsAnimatable=True,High=150,WeightMap="MaxDistance")}
 }
 SimulationStretchConfig_01 {
  type=FChaosClothAssetSimulationStretchConfigNode
  guid=#49
  properties {
   AreaStiffness=(bIsAnimatable=True,Low=1,WeightMap="AreaStiffness")
   StretchAnisoDamping=(bIsAnimatable=True,Low=1,High=1,WeightMap="StretchAnisoDamping")
   StretchStiffness=(bIsAnimatable=True,Low=1,WeightMap="StretchStiffness")
   StretchStiffnessBias=(bIsAnimatable=True,Low=10,High=10000000,WeightMap="StretchStiffnessBias")
   StretchStiffnessWarp=(bIsAnimatable=True,Low=10,High=10000000,WeightMap="StretchStiffnessWarp")
   StretchStiffnessWeft=(bIsAnimatable=True,Low=10,High=10000000,WeightMap="StretchStiffnessWeft")
  }
 }
 SimulationBendingConfig_01 {
  type=FChaosClothAssetSimulationBendingConfigNode
  guid=#50
  inputPins [
   Collection {type=FManagedArrayCollection}
   FlatnessRatio.WeightMap {type=FString}
   RestAngle.WeightMap {type=FString}
   BendingStiffness.WeightMap {type=FString}
   BendingStiffnessWarp.WeightMap {type=FString}
   BendingStiffnessWeft.WeightMap {type=FString}
   BendingStiffnessBias.WeightMap {type=FString}
   BendingDamping.WeightMap {type=FString}
   BendingAnisoDamping.WeightMap {type=FString}
   AnisoBucklingRatio.WeightMap {type=FString}
   BucklingStiffness.WeightMap {type=FString}
   BucklingStiffnessWarp.WeightMap {type=FString}
   BucklingStiffnessWeft.WeightMap {type=FString}
   BucklingStiffnessBias.WeightMap {type=FString}
   BucklingRatioWeighted.WeightMap {type=FString}
  ]
  properties {
   AnisoBucklingRatio=(bIsAnimatable=True,Low=0.5,High=0.5,WeightMap="BucklingRatio")
   BendingAnisoDamping=(bIsAnimatable=True,Low=1,High=1,WeightMap="BendingAnisoDamping")
   BendingStiffnessBias=(bIsAnimatable=True,Low=1,High=5,WeightMap="BendingStiffnessBias")
   BendingStiffnessWarp=(bIsAnimatable=True,Low=0.1,High=5,WeightMap="BendingStiffnessWarp")
   BendingStiffnessWeft=(bIsAnimatable=True,Low=0.1,High=5,WeightMap="BendingStiffnessWeft")
   BucklingStiffnessBias=(bIsAnimatable=True,Low=2,High=50,WeightMap="BucklingStiffnessBias")
   BucklingStiffnessWarp=(bIsAnimatable=True,Low=2,High=50,WeightMap="BucklingStiffnessWarp")
   BucklingStiffnessWeft=(bIsAnimatable=True,Low=2,High=50,WeightMap="BucklingStiffnessWeft")
  }
 }
 SimulationSolverConfig_01 {
  type=FChaosClothAssetSimulationSolverConfigNode
  guid=#51
  properties {
   MaxNumIterations=3
   NumIterations=3
   NumSelfCollisionSubsteps=10
   NumSubstepsImported=(ImportedValue=6)
   bEnableNumSelfCollisionSubsteps=True
  }
 }
 SimulationMassConfig_01 {
  type=FChaosClothAssetSimulationMassConfigNode
  guid=#52
  properties {DensityWeighted=(Low=0.7,High=0.7,WeightMap="Density")}
 }
 SimulationDampingConfig_01 {
  type=FChaosClothAssetSimulationDampingConfigNode
  guid=#53
  properties {
   DampingCoefficientWeighted=(bIsAnimatable=True,Low=0.5,High=0.5,WeightMap="DampingCoefficient")
   LocalDampingAngularCoefficient=0.1
   LocalDampingLinearCoefficient=0.1
   LocalDampingSpace=CenterOfMass
  }
 }
 SimulationGravityConfig {
  type=FChaosClothAssetSimulationGravityConfigNode
  guid=#54
  properties {GravityOverrideImported=(ImportedValue=(X=0,Y=0,Z=-200))}
 }
 MaxDistWeightMap {type=FChaosClothAssetWeightMapNode, guid=#55, properties {OutputName=(StringValue="MaxDistance")}}
 ClothAssetTerminal {type=FChaosClothAssetTerminalNode_v2, guid=#56}
 SkeletalMeshImport_0 {
  type=FChaosClothAssetSkeletalMeshImportNode_v2
  guid=#57
  properties {
   SkeletalMesh=&SkeletalMesh'&/M_MED_SkippingClouds_MLD_Sim'
   UVScale=(X=70,Y=70)
   bImportRenderMesh=False
   bImportSimMesh=True
   bImportSingleSection=True
  }
 }
 SkeletalMeshImport_00 {
  type=FChaosClothAssetSkeletalMeshImportNode_v2
  guid=#58
  properties {
   SkeletalMesh=&SkeletalMesh'&/M_MED_SkippingClouds_ForClothTests'
   UVScale=(X=70,Y=70)
   bImportRenderMesh=True
   bImportSimMesh=False
   bImportSingleSection=True
  }
 }
 MergeClothCollections {type=FChaosClothAssetMergeClothCollectionsNode_v2, guid=#59}
 StretchWeightMap {type=FChaosClothAssetWeightMapNode, guid=#60, properties {OutputName=(StringValue="StretchConfig")}}
 Selection_2 {
  type=FChaosClothAssetSelectionNode_v2
  guid=#61
  inputPins [
   Collection {type=FManagedArrayCollection}
   InputName.StringValue {type=FString}
   TransferCollection {type=FManagedArrayCollection}
  ]
  outputPins [
   Collection {type=FManagedArrayCollection}
   OutputName.StringValue {type=FString}
  ]
  properties {
   Group=(Name="RenderFaces")
   Import=()
   ImportSecondary=()
   Indices=(23788,23791,23789,23810,23787,23823,23822,23824,23828,23829,24100,24103,24098,24099,24095,24094,23477,24093,24096,24089,24087,23475,23474,23470,23471,24086,24085,24084,24082,24083,24080,24079,24065,24064,24070,24060,24058,24059,24053,24051,24052,24048,24046,24045,24047,24049,24044,24032,24042,24043,24040,24038,24031,24027,24025,24026,24028,24023,24022,24021,24019,24020,24014,24012,24013,24009,23989,24008,24005,24010,24011,24015,24004,24000,24002,23994,23995,23990,23991,23987,23986,23983,23984,23980,23950,23979,23977,23978,23974,23948,23973,23976,23965,23949,23964,23967,23961,23951,23960,23963,23957,23954,23956,23958,23955,23877,23953,23939,23938,23884,23883,23875,23874,23872,23866,23873,23879,23868,23869,23863,23860,23859,23858,23838,23857,23867,23855,23840,23854,23852,23871,23870,23876,23878,23880,23853,23850,23846,23851,23856,23847,23845,23848,23813,23849,23811,23862,23861,23864,23865,23970,23969,24132,24134,24130,24128,24129,24131,24125,24126,24122,24120,24105,23831,23832,24121,24108,24109,24104,24102,24101,24106,24097,24091,24090,24088,24092,24081,24107,24078,24077,24075,24076,24073,24069,24074,24071,24072,24066,24061,24067,24057,24054,24056,24050,24055,24016,24030,24029,24018,24017,24024,24062,24063,24001,23998,23999,24003,23996,23997,23993,23992,24006,23981,23982,23975,24140,23971,23972,23966,23968,24136,24138,24137,24139,24135,24133,24123,24119,24124,24117,24007,24116,24115,24113,24111,24112,24114,24110,24118,24141,23962,23959,23885,23882,23881,24068,24127,23817,23818,23814,23812,23809,23803,23800,23804,23807,23802,23799,23801,23805,23806,23815,23819,23816,23842,23841,23835,23843,23844,23839,23834,23830,23833,23825,23826,23821,23702,23820,23827,23836,23837,23808,23887,23888,23886,23889,23893,23890,23909,23908,23911,23904,23903,23906,23901,23899,23894,23891,23892,23895,23896,23897,23898,23900,23902,23905,23907,23912,23913,23914,23915,23916,23918,23917,23919,23922,23920,23924,23929,23936,23937,23934,23952,23935,23941,23932,23928,23933,23940,23930,23926,23931,23942,23944,23943,23947,23946,23925,23923,23921,23927,23945,23910,23985,23988,24033,24034,24035,24036,24037,24039,24041,5291,5285,5287,5280,5279,5282,5276,5275,5271,5278,5281,5277,5273,5283,5297,5284,5274,5306,5305,5309,5324,5310,5311,5269,5264,5261,5263,5258,5266,5272,5265,5257,5267,5268,5270,5252,5251,5249,5250,5243,5241,5242,5238,5237,5236,5234,5232,5233,5235,5231,5226,5230,5229,5225,5222,5228,5227,5240,5224,5223,5221,5220,5218,5209,5219,5216,5217,5211,5210,5208,5205,5204,5203,5207,5206,5212,5201,5198,5200,5202,5191,5190,5193,5187,5186,5189,5185,5184,5183,5162,5161,5164,5159,5154,5160,5157,5163,5166,5174,5165,5168,5158,5155,5156,5151,5146,5153,5152,5144,5143,5140,5139,5133,5142,5141,5145,5149,5150,5147,5170,5454,5169,5167,5172,5173,5450,5449,5453,5456,5455,5460,5466,5461,5463,5457,5458,5459,5148,5137,5136,5134,5128,5135,5132,5129,5127,5122,5120,5121,5117,5115,5114,5116,5118,5119,5123,5126,5124,5130,5138,5131,5464,5462,5465,5469,5468,5478,5480,5479,5476,5484,5485,5486,5481,5491,5492,5494,5498,5493,5495,5475,5496,5497,5473,5451,5472,5467,5452,5447,5444,5446,5441,5448,5443,5440,6024,6022,6025,5438,6023,6011,6021,6019,6013,6017,6018,6014,6012,6015,5518,6016,5504,5505,5501,5500,5502,5503,5507,5490,5489,5487,5488,5483,5482,5506,5508,5509,5510,5511,5513,5514,5377,5374,5379,5378,5381,5375,5376,5371,5365,5372,5373,5338,5362,5336,5335,5307,5304,5308,5312,5293,5288,5296,5295,5313,5317,5303,5300,5332,5331,5325,5350,5358,5349,5348,5327,5326,5323,5321,5314,5322,5319,5316,5320,5328,5197,5194,5196,5192,5199,5213,5195,5188,5330,5182,5179,5181,5177,5176,5180,5334,5178,5175,5342,5343,5341,5347,5337,5339,5340,5346,5354,5353,5436,5439,5435,5434,5356,5357,5368,5355,5352,5359,5363,5364,5367,5370,5427,5369,5366,5426,5430,5433,5431,5432,5429,5428,6006,5425,5422,5423,5416,5411,5380,6009,6010,5360,5361,5351,5344,5345,5329,5437,5333,5442,5445,5171,5315,5260,5215,5259,5254,5262,5246,5253,5248,5256,5255,5244,5247,5245,5214,6020,5499,5474,5477,5470,5471,5125,5239,16850,16843,16841,16842,16839,16838,16836,16837,16833,16832,16537,16536,16535,16539,16533,16531,16532,16528,16527,16529,16524,16523,16522,16526,16519,16518,16521,16513,16512,16516,16510,16509,16514,16507,16506,16508,16511,16515,16562,16563,16561,16558,16565,16567,16564,16560,16570,16573,16571,16650,16559,16556,16553,16555,16551,16653,16656,16651,16652,16655,16648,16649,16643,16640,16645,16654,16644,16642,16647,16683,16646,16693,16641,16637,16636,16634,16632,16635,16639,16602,16601,16599,16593,16600,16589,16587,16584,16580,16579,16577,16582,16581,16583,16578,16576,16574,16569,16575,16572,16568,16566,16585,16586,16588,16590,16594,16591,16592,16595,16598,16596,16603,16606,16604,16607,16610,16608,16609,16612,16614,16613,16618,16619,16620,16615,16625,16626,16628,16631,16627,16629,16605,16630,16623,16633,16703,16638,16696,16694,16700,16702,16701,16705,16698,16699,16695,16692,16687,16685,16689,16688,16697,16728,16727,16789,16788,16790,16691,16796,16795,16798,16802,16799,16690,16800,16811,16676,16675,16674,16670,16821,16819,16822,16825,16668,16667,16546,16542,16547,16826,16544,16829,16831,16830,16835,16828,16827,16820,16849,16848,16840,16818,16813,16823,16810,16808,16801,16797,16793,16791,16783,16792,16784,16740,16785,16734,16733,16730,16729,16731,16723,16720,16722,16724,16721,16716,16714,16715,16712,16710,16711,16706,16704,16707,16708,16622,16621,16617,16624,16616,16709,16713,16717,16718,16719,16726,16725,16737,16739,16738,16735,16743,16744,16745,16742,16741,21284,17756,21283,21280,21277,21279,21276,21239,21238,21235,21236,21228,21230,21224,21223,21206,21204,21201,21196,21199,21200,21194,21192,21195,21175,21197,21198,21202,21178,21179,21172,21171,21160,21151,21159,21156,21162,21165,21161,21158,21163,21164,17788,17785,17789,17792,17786,17787,17783,17782,17781,17776,17775,17773,17778,17784,17777,17774,17780,21168,17779,21177,17772,17771,17770,17764,21182,21181,21183,17765,21186,21187,21188,21189,21191,21237,21190,21193,17766,17761,17759,17754,17760,17763,17757,17758,17755,17753,17750,17749,17752,17751,17762,17767,17768,17769,21184,21185,21176,21174,21173,21170,21169,21167,21166,17790,17793,17791,17794,17797,17795,17796,17799,17801,17800,17806,17808,17807,17811,17803,17804,17805,17809,17812,17810,17815,17817,17820,17821,17823,17829,17822,17825,17835,17824,21145,17819,17816,17818,21152,17814,17813,21153,21150,21144,21148,21142,21141,21146,21138,21140,21136,21135,21134,17837,21132,21133,17849,17850,17843,17839,17844,17847,17841,17842,17840,17838,17833,17831,17828,17832,17836,17830,17826,17827,17845,17848,17846,17851,17855,17852,17853,17856,17858,17857,17861,17864,17862,17859,17869,17870,17875,17876,17878,17877,21137,21139,21209,21212,21208,21155,21207,21210,21203,21180,21154,21147,21143,21211,21213,21216,21217,21215,17892,21214,17880,17879,17874,17882,17886,17881,17883,17872,17871,17868,17866,17865,17867,17863,17884,17885,17887,17890,17891,17893,17903,17902,17913,17897,17898,17900,17894,17895,17889,17888,17896,17899,17901,17904,17907,17910,17914,21243,17915,17912,21227,21225,21233,21242,21234,21231,21229,21232,21226,21222,21220,21221,21219,21218,17906,17873,21205,17860,17854,17834,21149,17802,21157,17798,16732,16736,16834,16543,16538,16534,16540,16541,16545,16530,16552,16548,16549,16550,16525,16520,16657,16662,16663,16665,16664,16669,16672,16673,16677,16671,16666,16679,16681,16680,16686,16684,16682,16660,16661,16658,16659,16678,16611,16597,16554,16557,16517,11499,11526,11527,11523,11528,11529,11535,11537,11533,11534,11536,11530,11531,11524,11520,11525,11522,11519,11521,11517,11516,11514,11515,11510,11508,11509,11506,11504,11505,11501,11502,11500,11503,11385,11507,11375,11511,11512,11513,11371,11518,11372,11367,11362,11368,11365,11359,11366,11361,11358,11364,11532,11363,11357,11542,11541,11543,11547,11353,11348,11354,11351,11346,11352,11355,11356,11345,11343,11341,11339,11342,11347,11340,11338,11344,11349,11350,11560,11558,11562,11563,11561,11559,11557,11555,11553,11554,11556,11551,11548,11550,11546,11552,11544,11545,11540,11538,11539,11549,11337,11336,11360,11369,11370,11373,11374,11376,11381,11380,11377,11383,11387,11382,11379,11384,11386,11388,11390,11391,11395,11396,11472,11476,11470,11467,11471,11474,11469,11465,11468,11466,11473,11475,11477,11480,11478,11479,11483,11482,11486,11487,11488,11490,11481,11393,11392,11394,11495,11497,11492,11493,11494,11491,11489,11496,11498,11485,11484,11464,11460,11458,11459,11456,11454,11455,11451,11449,11448,11450,11452,11446,11443,11445,11447,11444,11442,11440,11441,11439,11436,11434,11425,11424,11422,11423,11420,11418,11419,11414,11413,11412,11410,11416,11415,11409,11417,11421,11411,11408,11457,11453,11461,11462,11463,11407,11403,11402,11400,11401,11397,11398,11389,11399,11404,11405,11406,11426,11427,11432,11428,11429,11430,11431,11433,11435,11437,11438,11378,11302,11128,11123,11144,11127,11121,11129,11148,11147,11149,11151,11150,11153,11156,11154,11161,11143,11142,11140,11141,11137,11136,11134,11135,11130,11131,11126,11124,11125,11116,11114,11111,11115,11112,11118,11122,11117,11119,11113,11120,11110,11109,11108,11132,11133,11138,11139,11145,11146,11163,11176,11162,11179,11175,11168,11178,11177,11185,11183,11188,11190,11187,11186,11189,11192,11193,11195,11194,11202,11203,11207,11208,11210,11223,11209,11205,11212,11216,11211,11206,11213,11214,11217,11220,11221,11301,11299,11294,11300,11297,11304,11305,11303,11298,11306,11307,11309,11308,11310,11312,11311,11313,11315,11314,11317,11319,11318,11316,11320,11322,11321,11323,11325,11324,11326,11327,11329,11328,11330,11331,11332,11334,11197,11196,11174,11173,11171,11172,11169,11165,11170,11167,11166,11164,11160,11158,11157,11159,11155,11152,11191,11199,11198,11335,11333,11296,11293,11292,11290,11291,11288,11286,11283,11287,11284,11285,11281,11278,11280,11282,11279,11276,11272,11277,11274,11273,11271,11268,11270,11266,11265,11263,11262,11264,11261,11259,11254,11251,11253,11267,11243,11242,11239,11269,11234,11231,11233,11236,11230,11228,11229,11226,11235,11227,11224,11219,11225,11218,11215,11289,11275,11232,11237,11238,11240,11241,11247,11248,11249,11246,11250,11252,11255,11256,11257,11258,11260,11245,11244,11295,11204,11201,11200,11222,11184,11182,11181,11180,28601,28599,28603,28605,28607,28609,28611,28613,28615,28616,28614,28612,28610,28608,28606,28604,28602,28600,28598,28623,28625,28621,28619,28617,28618,28620,28622,28624,28626,28628,28630,28632,28634,28635,28633,28631,28629,28627,19695,10388,10384,10390,10389,10559,10387,10385,10391,10394,10346,10345,10343,10338,10677,10676,10678,10340,10680,10679,10682,10683,10686,10685,10688,10693,10687,10684,10689,10695,10323,10320,10322,10318,10326,10327,10324,10330,10331,10339,10328,10333,10337,10332,10329,10334,10335,10336,10341,10342,10344,10349,10347,10348,10351,10352,10374,10383,10373,10369,10376,10381,10375,10371,10379,10380,10382,10386,10392,10549,10393,10396,10400,10395,10378,10397,10398,10399,10401,10404,10405,10407,10406,10420,10419,10544,10543,10547,10550,10548,10552,10551,10554,10556,10555,10560,10562,10561,10558,10665,10668,10664,10566,10663,10649,10666,10667,10669,10671,10674,10672,10675,10670,10563,10673,10681,10691,10692,10690,10694,10698,10699,10701,10708,10700,10696,10703,10724,10702,10697,10734,10735,10733,10726,10737,10740,10738,10736,10741,10743,10742,10313,10745,10747,10746,10760,10759,10762,10766,10763,10768,10769,10772,10776,10771,10770,10773,10774,10284,10285,10279,10277,10272,10271,10269,10266,10270,10273,10274,10275,10276,10278,10280,10281,10282,10283,10286,10287,10289,10293,10288,10292,10300,10301,10307,10298,10295,10299,10304,10296,10297,10294,10291,10290,10302,10305,10303,10309,10311,10310,10317,10308,10312,10321,10314,10315,10316,10306,10265,10267,10260,10259,10256,10262,10268,10261,10258,10781,10779,10782,10263,10784,10783,10788,10790,10789,10786,10792,10797,11001,11002,11000,10995,11003,11006,11004,10997,11012,11016,11013,11010,11015,11020,11014,11011,11023,11024,11027,11026,11030,11029,11040,11039,11097,11102,11095,11091,11096,11099,10627,10624,10630,10634,10631,10638,10628,10625,10629,10636,10574,10575,10570,10567,10571,10637,10568,10569,10564,10565,10557,10553,10546,10545,10541,10542,10538,10536,10537,10534,10533,10414,10532,10527,10424,10526,10529,10528,10535,10539,10540,10531,10577,10576,10580,10581,10584,10583,10623,10579,10578,11041,11042,11037,11034,10987,10990,10819,10818,10808,10809,10804,10805,10801,10798,10800,10796,10806,10807,10795,10793,10794,10791,10264,10253,10254,10250,10249,10247,10248,10244,10243,10246,10240,10239,10242,10236,10235,10238,10241,10237,10811,10814,10810,10803,10813,10812,10802,10799,10245,10815,10418,10416,10417,10413,10412,10415,10409,10408,10411,10403,10402,10410,10530,10251,10252,10255,10257,10991,10994,10993,10996,11017,11018,11021,11031,11022,11019,11028,11033,10573,10572,10641,10640,10643,10647,10642,10639,10646,10651,10648,10653,10657,10659,10662,10660,10650,10661,10705,10707,10704,10706,10709,10712,10713,10715,10721,10714,10723,10725,10728,10727,10739,10749,10751,10750,10748,10755,10757,10756,10761,10752,11007,10765,10787,10764,10767,10785,11005,11008,11009,10758,10753,10754,10732,10731,10729,10720,10730,11025,10722,10719,11107,11105,11106,10717,11104,11103,11101,10654,11100,11098,10645,10644,10652,10656,10655,10658,10711,10716,10710,10718,10744,10780,10778,10777,10775,10377,10372,10370,10367,10365,10366,10364,10363,10360,10361,10362,10358,10357,10354,10353,10356,10350,10355,10359,10368,10325,10319,19700,19697,19690,19387,19689,19693,19685,19389,19684,19687,19682,19391,19678,19677,19401,19397,19673,19672,19675,19674,19676,19670,19667,19669,19664,19671,19413,19417,19409,19408,19405,19404,19402,19403,19400,19398,19395,19399,19396,19393,19394,19390,19392,19382,19376,19383,19388,19380,19381,19375,19374,19373,19371,19377,19378,19379,19384,19385,19701,19703,19704,19706,19705,19709,19711,19708,19710,19713,19714,19715,19717,19721,19718,19719,19723,19724,19726,19731,19725,19728,20950,19727,19722,20957,20956,20959,20960,20965,20967,20966,20968,20962,19699,20961,19702,19707,19712,19716,19696,19698,20969,19694,19691,19692,19688,20970,20973,20975,20974,20972,20981,20985,20980,20982,20979,20978,19658,20977,19662,20976,19683,19681,19666,19679,19680,19686,19665,19659,19668,19660,19654,19656,19649,19558,19648,19651,19647,19561,19646,19645,19553,19557,19549,19546,19550,19555,19547,19545,19548,19542,19661,19663,19543,19540,19451,19539,19809,19802,19805,19799,19798,19804,19796,19795,19792,19800,19801,19806,20890,20893,20889,20882,20892,20895,20899,20910,21052,20909,20912,20903,20896,20904,20907,20901,20894,20902,20905,19797,19794,19793,19791,19787,19786,19784,19789,19788,19790,19783,19780,19782,19779,19777,19778,19772,19770,19771,19767,19765,19764,19766,19769,19761,19756,19762,19763,19755,19754,19758,19750,19749,19752,19747,19745,19746,19742,19741,19744,19738,19733,19740,19739,19743,19748,20932,19753,19759,19751,19760,20930,20928,20931,20934,20935,20937,20942,20936,20933,20938,20940,19737,19736,19734,19735,19729,19730,19732,20941,20945,20947,20948,20949,20952,20953,20955,20954,20958,21005,20964,20963,20984,20991,20983,20971,20986,20990,20993,20997,20992,20995,20988,20987,20989,20994,19652,19644,19653,19657,19650,19643,19642,19640,20996,21001,21003,21002,21013,20999,21000,21007,21012,21008,21004,21010,21014,21009,21006,21011,21016,20951,21017,21019,21023,21018,21045,21044,21130,21128,21131,21015,21129,21127,21125,21124,21122,21123,21115,21113,19639,19641,21109,21114,21108,21111,21110,21116,21117,21119,21118,21126,21112,21120,21053,21121,21047,21046,21040,21049,21048,21042,21051,21056,21050,21043,21041,21036,20920,21034,21035,21033,21031,20922,20919,20921,20927,20918,20916,20917,20923,20914,20913,20906,20926,20924,20925,19785,19781,19776,19773,19775,19774,19768,20929,20915,20908,21028,21027,21026,20939,20944,21021,21024,21020,21022,21025,21029,21030,21037,21038,21032,21039,20911,21058,21068,20946,20943,20998,19757,19423,19422,19415,19425,19424,19419,19427,19436,19426,19421,19418,19420,19544,19416,19411,19410,19412,19407,19406,19414,19655,19720,19369,19363,19370,19367,19362,19368,19372,19364,19360,19365,19366,19359,19357,19356,19358,19361,19386,28659,11791,11794,11777,11790,11786,11793,11795,11792,11788,11797,11800,11798,11789,11809,11810,11808,11812,11816,11811,11814,11806,11802,11805,11804,11807,11813,11815,11822,11821,11830,11832,11831,11828,11841,11845,11840,11838,11843,11848,11842,11839,11844,11854,11868,11853,11856,11865,11855,11858,11851,11849,11850,11852,11847,11846,11857,11859,11860,11861,11862,11863,11866,11867,11877,11878,11876,11881,11871,11872,11874,11869,11873,11880,11883,11882,11887,11890,11886,11885,11888,11889,11891,11893,11895,11897,11894,11892,11896,11879,11898,11899,11884,11875,11870,11864,11837,11836,11835,11833,11827,11834,11829,11826,11825,11823,11819,11824,11818,11820,11817,11803,11801,11796,11799,11787,11784,11782,11781,11783,11785,11779,11774,11778,11780,11772,11770,11764,11771,11768,11763,11769,11773,11765,11760,11766,11767,11756,11754,11755,11751,11749,11750,11748,11747,11745,11742,11746,11743,11740,11744,11739,11753,11757,11752,11758,11761,11759,11738,11735,11733,11732,11734,11736,11737,11762,11741,11775,11776,28661,28657,28658,28660,28662,28664,28666,28668,28670,28672,28674,28676,28677,28675,28673,28671,28669,28667,28665,28663,11608,28644,28646,28642,28640,28638,28636,28637,28639,28641,28643,28645,28647,28649,28651,28653,28655,28656,28654,28652,28650,28648,11605,11616,11607,11610,11603,11602,11601,11597,11600,11595,11593,11594,11591,11596,11606,11617,11618,11626,11628,11627,11624,11631,11634,11632,11629,11640,11645,11641,11638,11643,11648,11642,11644,11639,11637,11636,11635,11633,11650,11652,11651,11656,11659,11657,11662,11654,11649,11653,11655,11647,11646,11630,11623,11621,11615,11622,11625,11619,11614,11620,11613,11612,11611,11609,11604,11661,11663,11666,11668,11667,11664,11671,11670,11673,11676,11674,11677,11665,11678,11680,11682,11679,11684,11687,11685,11690,11694,11689,11683,11691,11692,11693,11695,11698,11700,11699,11704,11696,11697,11701,11705,11702,11703,11688,11686,11675,11672,11709,11711,11710,11719,11721,11718,11707,11720,11723,11716,11708,11715,11717,11712,11722,11725,11727,11729,11726,11724,11728,11730,11714,11713,11731,11706,11681,11660,11658,11669,11589,11588,11590,11587,11584,11586,11583,11582,11580,11581,11576,11575,11574,11572,11578,11577,11571,11579,11585,11573,11569,11598,11599,11568,11566,11565,11564,11567,11570,11592,19435,19442,19430,19431,19433,19428,19429,19432,19434,19437,19438,19441,19446,19440,19443,19448,19452,19449,19454,19456,19458,19461,19457,19447,19459,19464,19460,19445,19462,19463,19465,19469,19485,19473,19472,19481,19484,19480,19495,19475,19474,19477,19470,19471,19468,19467,19476,19478,19479,19482,19483,19489,19491,19490,19493,19486,20610,20612,20609,20607,20603,20608,20598,20611,20613,20615,20591,20581,20590,20586,20583,20588,20587,20594,20596,20597,20601,20600,20599,20602,20604,20606,20605,24272,20595,20592,20593,20589,24273,24274,24275,24271,24268,24270,24267,24266,24265,24262,24261,24256,24263,24264,24269,24276,24279,24280,24281,24283,24284,24285,24290,24286,24257,24287,24291,24259,24260,24251,24248,24252,24249,24246,24250,24255,24258,24245,24244,24242,24237,24243,24247,24240,24236,24241,24254,24238,24239,24234,24233,24231,24230,24232,24235,24307,24308,24306,24309,24303,24301,24302,24296,24294,24295,24293,24299,24300,24297,24315,24316,24314,24318,24310,24323,24317,24320,24312,24311,24313,24319,24321,24327,24329,24328,24332,24325,24326,24330,24333,24331,24337,24324,24322,24305,20521,24304,24298,24289,20519,24288,24292,20518,20517,20513,20514,20510,20509,20482,20481,20478,20474,20515,20516,20520,20475,20473,20470,20524,20464,20465,20462,20466,20458,20454,20459,20463,20457,20452,20461,20546,20460,20758,20455,20456,20451,20444,20440,20445,20447,20429,20421,20428,20436,20424,20419,20426,20432,20425,20422,20430,20438,20431,20480,20427,20423,20433,20434,20486,20485,20492,20494,20493,20490,20496,20498,20499,20502,20500,20503,20491,20489,20504,20487,20488,20483,20484,20479,20477,20476,20472,20471,20442,20437,20443,20449,20439,20435,20441,20446,20448,20450,20453,20469,20468,20506,20505,20508,24340,20507,20511,20512,24349,24352,24344,24347,24342,24341,24338,24343,24345,24335,24334,24336,24346,24348,24350,24351,24353,24354,24355,24356,24357,24359,24360,24361,24363,24364,24365,24368,24371,24369,24414,24412,24413,24411,24409,24405,24403,24404,24401,24399,24400,24402,24397,24396,20369,20368,20365,20366,20360,20358,20359,20354,20353,20356,20348,20346,20347,20343,20216,20342,20341,20345,20344,20339,20349,20352,20350,20338,20362,20372,20361,20357,20364,20370,20363,20371,20373,20377,20379,20378,20375,20381,20384,20382,20386,20376,20335,20334,20332,20328,20333,20337,20330,20329,20331,20336,20227,20226,20225,20220,20340,20217,20219,20211,20223,20224,20228,20221,20222,20206,20202,20201,20197,20196,20194,20191,20195,20198,20184,20183,20179,20177,20180,20199,20200,20203,20204,20205,20207,20208,20210,20209,20213,20212,20214,20215,20218,20176,20171,20181,20182,20178,20185,20189,20186,20190,20192,20193,20234,20236,20233,20231,20238,20240,20239,20243,20245,20244,20248,20241,20237,20242,20246,20188,20187,20175,20174,20172,20166,20173,20169,20170,20164,20163,20158,20157,20160,20152,20151,20154,20146,20145,20148,20141,20140,20143,20138,20136,20137,20133,20131,20132,20127,19949,20126,20124,20129,20128,20123,20130,20135,20125,20121,20269,20268,20270,20115,20271,20272,20267,20264,20149,20263,20262,20266,20265,20261,20275,20274,20277,20278,20280,20284,20279,20281,20276,20110,20097,20109,20105,20112,20111,20273,20108,20106,20107,20104,20113,20116,20114,19968,20117,20118,20120,19964,20119,19958,20122,19959,19954,19950,19955,19948,19944,19947,19946,19945,19943,19936,19934,19935,19932,19930,19931,19933,19929,19924,19928,19927,19923,19920,19926,19925,19937,19922,19921,19919,19914,19913,19910,19917,19918,19915,19953,19951,19957,19960,19956,19952,19961,19963,19967,19962,19970,19969,19976,19965,19966,19971,19912,19911,19908,19903,19972,19973,19975,19905,19980,19900,19901,19897,19896,19894,19889,19895,19898,19899,19902,19904,19906,19907,19909,19888,19891,19885,19884,19883,19887,19878,19877,19880,19873,19872,19875,19870,19843,19869,19874,19867,19868,19862,19845,19861,19859,19865,19866,20007,19863,19864,19857,20019,20022,20018,20015,20021,20026,20020,20017,20031,20033,20032,20029,20035,20039,20824,20834,20823,20821,20826,20832,20825,20822,20828,20827,20830,20831,20833,20843,20845,20844,20847,20841,20835,20842,20839,20846,20849,20848,20853,20852,20855,20856,20858,20857,20861,20862,20867,20866,20871,21080,20876,20874,20886,21070,20885,20888,20879,20875,20880,20883,20877,20873,20878,20881,19813,19818,19810,19807,19803,19808,19811,19812,19816,19817,19821,19814,19815,19819,19820,19822,19825,19830,20865,20864,20868,20869,20870,20872,20863,19834,19839,19838,20854,19832,19853,19851,20851,20850,19860,19856,20024,20023,20028,20840,20027,20025,20030,20836,19858,19852,19850,19848,19854,19855,19847,19844,19846,19835,19836,19831,19827,19833,19849,19826,19829,19824,19823,19828,19837,19840,19841,19842,20884,20891,20897,20898,20900,21055,21054,21059,21060,21062,21065,21061,21057,21067,21074,21073,21101,21100,21096,21104,21102,21098,21103,21099,21105,21106,20706,21107,19635,19632,19636,19634,19629,19637,19638,19581,19580,19573,19569,19568,19562,19559,19560,19554,19551,19552,19541,19535,19533,19534,19536,19524,19523,19455,19526,19521,19522,19518,19517,19516,19519,19511,19510,19488,19466,19487,19505,19507,19504,19509,19501,19500,19503,19506,19502,20635,19498,19494,19499,19496,20630,20633,20629,20627,20618,19497,20616,20617,20614,20622,20623,20625,20619,20620,20621,20579,20575,20574,20573,20571,20569,20572,20554,20553,20543,20537,20577,20578,20576,20580,20539,20582,20584,20585,20540,20532,20528,20531,20526,24278,24277,20527,24282,20523,20530,20535,20529,20525,20533,20536,20534,20467,20538,20542,20544,20545,20547,20550,20549,20548,20552,20558,20755,20756,20753,20748,20754,20759,20751,20747,20752,20761,20749,20750,20745,20743,20744,20740,20739,20676,20671,20677,20738,20674,20669,20675,20672,20680,20684,20737,20736,20734,20741,20742,20799,20802,20797,20798,20796,20794,20793,20792,20791,20789,20768,20767,20762,20764,20770,20769,20784,20787,20783,20781,20772,20771,20773,20766,20765,20763,20778,20774,20775,20776,20420,20416,20310,20415,20315,20414,20418,20409,20404,20411,20410,20417,20406,20412,20413,20408,20405,20407,24386,20396,20395,20392,20391,20389,20390,20387,24387,24389,24391,24388,24390,24385,24381,24380,20495,24383,24378,20497,24376,24377,24373,20501,24372,24374,24366,24367,24362,24358,24370,24375,24379,24407,24384,24382,24394,24398,24392,24393,24395,24406,24410,24408,20388,20385,20380,20383,20398,20397,20393,20400,20403,20399,20401,20394,20402,20319,20320,20316,20317,20314,20313,20311,20312,20306,20305,20292,20289,20291,20294,20285,20282,20283,20254,20249,20255,20252,20259,20258,20256,20253,20257,20260,20161,20156,20162,20159,20167,20168,20165,20155,20153,20150,20147,20247,20303,20293,20298,20287,20286,20288,20093,20295,20297,20296,20299,20095,20091,20096,20087,20780,20782,20779,20777,20785,20786,20088,20080,20077,20081,20079,20788,20790,20074,20083,20084,20795,20076,20801,20805,20800,20803,20807,20810,20808,20812,20815,20817,20820,20816,20806,20818,20045,20819,20041,20071,20070,20075,20043,20069,20062,20073,20066,20058,20044,20057,20060,20051,20046,20052,20055,20049,20050,20048,20047,19996,19997,19993,19990,19992,19985,19984,19982,19983,19977,19979,19978,20054,20053,20101,20102,20100,20099,20061,20056,20064,20090,20063,20059,20067,20068,20086,20089,20085,20082,20078,20092,20098,20065,20094,20103,19987,19986,19981,19988,19989,19890,19991,19892,19893,19886,19995,19882,19881,19879,19999,19876,19871,20003,20002,20006,20009,20014,20008,20005,20012,20016,20013,20010,20037,20042,20036,20034,20040,20004,20011,20000,20001,19998,19994,20072,20038,20813,20814,20811,20733,20809,20804,20735,20732,20730,20727,20729,20731,20728,20694,20690,20695,20722,20692,20689,20693,20698,20691,20688,20696,20701,20697,20686,20699,20700,20703,20707,20704,20709,20711,20712,20713,20715,20718,20716,21086,20714,21091,21092,21088,21089,21087,21083,21079,21082,21084,21081,20860,21085,20838,20837,20859,21077,21078,21072,21071,21069,21064,21063,21066,21075,21076,21090,21093,21095,21094,20710,21097,20719,20724,20717,20720,20723,20721,20726,20725,20829,20702,20708,20705,19631,19630,19627,19624,19623,19622,19617,19625,19626,19577,19628,19570,19576,19572,19574,19566,19567,19563,19564,19556,19575,19538,19531,19537,19532,19528,19525,19527,19529,19520,19514,19515,19530,19512,19513,19586,19585,19592,19594,19593,19603,19588,19587,19584,19590,19597,19589,19591,19595,19578,19579,19582,19583,19596,19598,19599,19610,19612,19609,19604,19614,19620,19615,20667,19611,19606,20657,20660,20655,20653,19608,19605,19607,20649,20651,20647,20643,20648,20645,20650,20652,20646,20654,20656,20658,20567,20566,20564,20659,20560,20557,20559,20555,20562,20563,20565,20561,20556,20570,20568,20637,20636,20631,20639,20644,20638,20634,20641,20640,20642,20632,20628,20626,20624,20551,20662,20665,20663,20664,20670,20668,20661,19602,19600,19601,20666,19621,20673,20678,20679,20683,20681,20682,20687,20685,19618,19619,19616,19613,19508,19565,19571,19633,20300,20302,20301,20309,20304,20307,20308,20290,20251,20250,20322,20323,20321,20325,20318,20326,20324,20235,20232,20327,20230,20229,20760,20757,20746,20541,19492,19453,19450,20887,19974,19916,19940,19939,19942,19941,19938,20144,20139,20142,20134,20374,20355,20351,20367,24339,20522,24253,19444,19439,11094,11092,10635,10632,10626,10633,10620,11088,11086,11081,11087,11083,11090,11093,11089,11085,11084,11048,11044,11053,11063,11051,11046,11052,11057,11049,11047,11050,11054,10978,10968,10977,10971,10980,10979,10975,10989,10988,11036,11045,11035,11032,11038,11043,10999,10998,10992,10986,10984,10985,10982,10981,10821,10820,10831,10816,10817,10824,10825,10830,10833,10970,10832,10828,10846,10845,10965,10969,10964,10962,10967,10966,10960,10958,10961,10881,10880,10850,10849,10847,10848,10842,10841,10837,10836,10834,10835,10827,10226,10826,10823,10829,10229,10822,10225,10230,10233,10227,10222,10228,10231,10232,10234,10221,10219,10218,10223,10215,10213,10214,10211,10209,10210,10205,10204,10207,10202,10200,10201,10197,10196,10096,10095,10090,10098,10103,10097,10092,10101,10105,10102,10099,10107,10110,10108,10104,10113,10173,10117,10116,10167,10172,10166,10165,10137,10136,10119,10118,10122,10114,10115,10112,10094,10111,10093,10087,10088,10079,10074,10080,10085,10077,10072,10078,10083,10084,10086,10091,10089,10071,10068,10067,10062,10070,10073,10069,10064,10075,10076,10082,10081,10121,10120,10124,10125,10126,10127,10129,10140,10128,10130,10131,10065,10066,10060,10059,10057,10058,10051,10050,10048,10049,10047,10045,10046,10037,10038,10030,10028,10029,10023,10021,10018,10022,10020,10026,10027,10033,10024,10019,10025,10031,10003,10002,9998,9997,9994,9995,9992,9990,9991,9983,9950,9982,9977,9985,9988,9984,9979,9986,9989,9987,10161,9981,9978,9980,10154,9974,9970,9975,9972,9968,9973,9954,9976,9953,9951,9947,9952,9949,9945,9944,9943,9941,9939,9940,9931,9926,9932,9937,9929,9925,9930,9927,9935,9936,9938,9942,9955,9948,9946,9958,9956,9959,9961,9960,9965,9964,9967,9966,9963,9962,9542,9538,9543,9540,9548,9549,9902,9901,9896,9904,9903,9934,9933,9928,9906,9905,9900,9909,9910,9923,9922,9920,9916,9919,9918,9921,9924,9917,9864,9915,9914,9913,9912,9866,9862,9865,9860,9868,9877,9867,9861,9870,9869,9875,9876,9878,9894,9908,9892,9881,9893,9899,9890,9887,9891,9897,9889,9886,9888,9895,9885,9884,9555,9552,9556,9559,9553,9550,9554,9486,9557,9558,9560,9564,9568,9729,9728,9733,9734,9831,9830,9880,9883,9879,9874,9882,9837,9873,9872,9842,9871,9850,9841,9851,9847,9853,9856,9852,9849,9854,9855,9859,9827,9857,9858,9826,9823,9822,9821,9848,9817,9845,9814,9810,9808,9812,9811,9844,9843,9839,9838,9813,9809,9815,9805,9802,9804,9807,9796,9795,9793,9789,9794,9806,9788,9791,9780,9779,9777,9778,9773,9771,9772,9770,9768,9767,9764,9762,9763,9759,9758,9691,9690,9693,9685,9684,9688,9682,9681,9683,9686,9642,9638,9643,9633,9687,9694,9697,9695,9622,9698,9699,9700,9604,9705,9706,9707,9598,9709,9712,9710,9595,9714,9715,9718,9719,9723,9740,9727,9732,9725,9720,9726,9730,9724,9722,9721,9575,9570,9579,9578,9573,9581,9580,9576,9585,9590,9586,9582,9588,9589,9591,9592,9593,9600,9597,9596,9602,9605,9603,9607,9608,9619,9621,9618,9623,9613,9615,9620,9614,9610,9617,9616,9626,9654,9625,9628,9649,9627,9630,9624,9631,9640,9647,9639,9641,9637,9634,9635,9636,9632,9629,9644,9648,9645,9646,9652,9140,9139,9663,9136,9135,9133,9134,9127,9124,9128,9131,9125,9126,9123,9122,9121,9129,9132,9130,9137,9143,9138,9141,9145,9142,9147,9150,9148,9151,9152,9154,9153,9157,9158,9300,9301,9299,9167,9298,9303,9296,9295,9210,9211,9207,9168,9206,9205,9171,9166,9172,9175,9169,9170,9164,9163,9160,9159,9156,9162,9165,9161,9155,9173,9176,9174,9177,9180,9178,9181,9184,9182,9187,9190,9188,10455,10458,10454,10453,10456,10450,10447,10452,10457,10451,10449,10474,10471,10475,10476,10478,10477,10481,10482,10487,10486,10496,10502,10497,10494,10500,10504,10501,10592,10498,10495,10499,10595,10493,10492,9362,9360,10600,10603,10599,10602,10604,10601,10597,10606,10609,10607,10598,10619,10621,10622,10593,10589,10588,10585,10591,10590,10596,10587,10586,10514,10519,10510,10505,10509,10506,10512,10511,10517,10518,10522,10515,10516,10513,10435,10436,10433,10429,10432,10428,10427,10425,10422,10426,10421,10423,10521,10520,10523,10430,10431,10434,10437,10438,10440,10443,10439,10508,10507,10480,10479,10444,10448,10441,10442,10445,10446,10525,10524,10582,10503,10613,10608,10611,10612,11080,11079,11077,11075,11076,11078,11074,11067,11072,11073,11069,11068,11070,11071,10941,10936,10942,10953,10939,10934,10940,10946,10937,10938,10929,10930,10924,10923,10912,10911,10909,10910,10905,10906,10902,10899,10901,10907,10894,10889,10893,10896,10890,10887,10891,10892,10191,10182,10192,10869,10189,10184,10862,10866,10861,10187,10859,10860,10857,10194,10856,10203,10855,10852,10858,10863,10864,10867,10870,10865,10873,10871,10868,10878,10886,10879,10876,10884,10888,10885,10945,10882,10877,10883,10951,10950,10948,10956,10957,10959,10963,11055,11056,11059,11061,11060,11058,11064,11065,11062,11066,10955,10952,10954,10949,10947,10943,10944,10908,10874,10872,10875,10853,10854,10851,10843,10844,10840,10216,10839,10838,10220,10224,10217,10212,10208,10206,10199,10193,10198,10195,10190,10186,10106,10185,10188,10180,10109,10179,10183,10177,10174,10178,10181,10175,10176,10170,10171,10168,10169,10164,9518,10163,10138,9521,9517,9525,9524,10133,10132,10134,10135,9527,10139,10141,9529,10145,10147,10146,10143,10149,10152,10148,10144,10150,10155,10042,10041,10039,10034,10040,10156,10032,9999,9993,10162,10160,10158,10157,10159,10044,10036,10035,10043,10052,10053,10142,10151,10153,9971,9969,9537,9533,9536,9539,9531,9528,9532,9534,9495,9497,9492,9491,9490,9493,9479,9480,9476,9474,9477,9494,9469,9502,9467,9458,9466,9460,9457,9462,9461,9472,9475,9473,9470,9481,9482,9483,9484,9471,9468,9465,9562,9561,9563,9565,9440,9436,9435,9434,9566,9431,9430,9427,9423,9429,9572,9428,9425,9583,9584,9587,9426,9418,9414,9419,9611,9416,9413,9417,9311,9421,9612,9679,9680,9677,9658,9674,9675,9670,9669,9660,9659,9657,9656,9662,9667,9661,9655,9666,9668,9665,9144,9664,9671,9146,9673,9672,9149,9676,9678,9302,9653,9651,9650,9307,9304,9310,9305,9313,9312,9316,9308,9309,9306,9314,9291,9289,9292,9213,9214,9212,9218,9220,9221,9285,9284,9288,9315,9318,9319,9321,9320,9324,9399,9325,9329,9322,9323,9326,9281,9282,9277,9276,9245,9246,9239,9227,9238,9241,9235,9230,9237,9236,9240,9242,9247,9252,9248,9250,9254,9251,10464,9234,9233,9232,10461,9229,9201,9228,9231,9223,9199,9222,9219,9226,9217,9224,9225,9198,9215,9200,9216,9196,9204,9209,9203,9194,9193,9185,9186,9179,9191,9183,9195,9192,9197,9189,9208,9202,10460,10462,10463,10466,10459,10465,10468,10467,10472,10473,10470,9297,10469,9294,10484,10483,10485,9293,10488,10489,10490,10491,9290,9287,9266,9286,9283,9280,9263,9279,9261,9278,9355,9275,9262,9274,9270,9349,9348,9354,9356,9357,9359,9372,9358,9371,10605,10610,10614,10615,10617,10616,10618,9369,9365,9368,9363,10935,10933,10931,10928,10932,9366,10927,10925,9370,9373,9367,9351,9350,9347,9374,9343,9345,9341,9336,9342,9339,9375,9376,9377,9378,9381,9379,10922,10920,10919,10918,9383,10917,10916,9397,9449,9395,9384,9394,9388,9445,9448,9444,9447,9443,9390,9442,9441,9401,9391,9400,9396,9411,9433,9410,9424,9404,9405,9398,9408,9402,9403,9406,9409,9407,9412,9415,9420,9422,9437,9432,9439,9455,9438,9464,9463,9454,9446,9393,9392,9333,9338,9331,9328,9330,9332,9327,9334,9337,9340,9344,9386,9387,9385,9382,9380,9389,9335,9273,9269,9268,9265,9271,9272,9346,9264,9249,9244,9253,9257,9260,9259,9258,9255,9256,9267,9243,9451,9450,9456,9453,9459,9452,9508,9507,10915,10913,10904,10914,10921,10903,9516,9515,10898,9511,9506,9513,9512,9509,9504,9510,9500,9514,9519,9520,9499,9522,9523,9526,9496,9498,9501,9505,9503,10897,10900,10895,9353,9364,9352,10926,9361,9317,9609,9577,9571,9574,9569,9567,9478,9485,9487,9489,9488,9541,9544,9545,9546,9547,9535,9530,11082,10594,9606,9601,9599,9594,9736,9731,9825,9824,9820,9828,9829,9832,9833,9835,9834,9840,9836,9818,9819,9816,9803,9800,9799,9797,9798,9792,9787,9790,9785,9784,9781,9782,9776,9774,9748,9744,9747,9742,9750,9766,9749,9752,9745,9743,9746,9717,9751,9753,9708,9754,9755,9702,9701,9692,9704,9696,9689,9703,9757,9756,9760,9761,9765,9769,9713,9711,9716,9739,9737,9738,9741,9735,9783,9801,9786,9775,9846,9551,9898,9907,9911,9863,9957,9996,10000,10001,10004,10006,10005,10011,10010,10013,10007,10008,10009,10012,10014,10015,10017,10016,10054,10055,10056,10061,10063,10123,10100,10972,10973,10974,10976,10983,21240,21246,21241,21259,21274,21258,21261,21253,21254,21256,21251,21250,21248,21245,21247,21244,21249,17933,17932,17922,17919,17923,17926,17920,17917,17921,17924,17927,17925,17928,17931,17929,17937,17938,17939,17940,17942,17948,17941,17945,17947,17949,17954,17984,17955,17960,17952,17953,17958,17950,17951,17956,17957,17959,17966,17970,17965,17968,17962,17963,17964,17967,17969,17973,17974,17977,17978,18033,18073,18032,18029,18035,18034,18039,18072,18041,18065,18040,18037,18043,18061,18042,18038,18046,18052,18047,18044,18050,18053,18051,18048,18055,18057,21332,21327,21335,21337,21334,21330,21333,21336,21331,21329,17710,17709,17707,17708,17701,17702,17694,17693,17691,17688,17692,17699,17700,17703,17706,17704,17714,17715,17716,17717,17719,17718,17722,17726,17723,17728,17720,17721,17729,17731,17730,17736,17739,17737,17742,17734,17735,17740,17741,17745,17746,17748,17747,21285,21287,21289,21286,21288,21282,21290,21293,21291,21295,21296,21297,21298,21301,21304,21300,21299,21294,21302,21303,21269,21268,21266,21263,21267,18009,21305,21307,21309,21306,21308,17738,21310,21312,21311,21313,18013,21315,21317,21314,21316,17727,21318,21320,21319,21321,21324,21322,18021,21326,18020,18017,18014,18012,18008,18015,18016,18018,17985,17981,17986,17989,17983,17980,17979,17976,17961,17975,17982,18019,18024,18027,18025,17972,18028,18031,18030,18036,17971,18026,18022,18045,18023,18049,17990,17993,17987,17988,17991,17992,17994,17997,17998,18003,18006,18004,18001,18002,21257,17999,17996,18000,21252,17935,17934,17944,17930,17943,17946,17936,21255,17995,21262,21265,21264,21260,21270,21271,21273,21272,21281,21292,21275,21278,18005,18007,18010,18011,21323,21328,21325,17712,17713,17725,17711,17724,17733,17732,17743,17744,17705,17687,17680,17690,17689,17682,17695,17698,17696,21343,17685,17686,17681,17678,17677,17675,17671,17676,17679,17670,17668,17673,17672,17667,17674,17684,17669,21357,17665,17664,17662,17663,17657,17658,17653,17652,17650,17655,17659,17654,17649,17656,17651,21368,21371,21367,21366,21369,21364,21358,21365,21370,21363,21386,21360,21356,21359,21361,21353,21351,21352,21350,21354,21348,21349,21347,21344,21346,18074,21345,18059,18058,18056,18054,18062,18068,18063,18066,18060,21491,18067,18064,18096,21494,18095,18093,18099,21499,18097,18098,18092,18087,18088,18083,18071,18082,18080,18085,18084,18086,18081,18079,18078,18077,18076,18089,18091,18090,18094,18100,18103,18101,21507,21508,21505,21506,21510,21503,21504,21498,21497,21496,21495,21493,21492,21489,21490,21487,21488,21484,21362,21483,21387,21482,21480,21486,21485,21481,21501,21500,21502,21479,21511,21509,21514,21516,21515,21521,21513,21512,21518,21519,21522,21527,21530,21528,21532,21525,21520,21526,18134,21531,21533,21534,21538,21541,21539,21552,21536,21537,21843,21842,21844,21847,21529,21524,21523,21517,21849,21850,21463,21851,21852,21853,21462,21854,21856,21791,21454,21790,21861,21784,21452,21783,21786,21780,21450,21779,21777,21782,21781,21785,21796,21794,21866,21867,21864,21862,21863,21865,21860,21859,21858,21841,21857,21855,21845,21839,21846,21848,21838,21836,21830,21837,21831,21827,21551,21826,21825,21549,21550,21546,21543,21545,21542,21544,21547,18140,18137,18141,18139,18144,18145,21548,21553,21555,21554,18148,21557,21558,21560,21563,21559,21815,21813,21820,21822,21821,21829,21818,21814,21819,21833,21816,21808,21817,21823,21803,21804,21800,21797,21799,21824,21795,21792,21789,21793,21597,21598,21593,21594,21590,21589,21588,21587,21584,21585,21582,21581,21578,18504,21577,21576,18445,18442,18446,18448,18443,18444,18440,18441,18437,18201,18436,18433,18439,18438,18449,18435,18434,18432,18429,18213,18428,18425,18431,18430,18455,18427,18424,18426,18461,18423,18422,18420,18417,18421,18225,18214,18224,18227,18218,18219,18211,18210,18208,18205,18209,18216,18217,18220,18223,18221,18228,18231,18229,18234,18236,18239,18240,18244,18245,18247,18381,18246,18250,18242,18243,18248,18249,18251,18254,18365,18255,18258,18252,18253,18256,18257,18259,18262,18263,18353,18358,18352,18349,18355,18360,18354,18351,18362,18369,18363,18356,18676,18675,18679,18680,18682,18681,18687,18686,18725,18726,18724,18723,18692,18696,18689,18688,18690,18685,18341,18684,18683,18346,18340,18347,18357,18344,18337,18345,18350,18342,18335,18343,18348,18270,18271,18267,18266,18264,18261,18265,18268,18269,18272,18275,18273,18276,18279,18277,18280,18283,18281,18284,18285,18287,18286,18290,18291,18319,18324,18318,18316,18322,18326,18323,18320,18700,18699,18703,18706,18704,22427,18321,18317,22430,18311,18309,18312,18315,18299,18300,18295,18294,18292,18293,18289,18288,18296,18297,18298,18301,18304,18302,18305,18308,18306,18310,22453,22526,22525,22527,22529,22528,22531,22532,22534,22533,22537,22539,22538,22535,22542,22545,22543,22540,22557,22558,22559,22547,22561,22570,22564,22568,22562,22563,22566,22560,22554,22565,22567,22573,22556,22555,22553,22550,22549,22551,22546,22544,22552,22585,22574,22589,22591,22590,22587,22594,22596,22593,22595,22592,22597,22598,22599,22600,22606,22607,22853,22852,22851,22848,22822,22821,22620,22619,22610,22611,22614,22608,22605,22609,22612,22615,22613,16911,22616,22617,22618,22623,16914,16915,16910,16908,16907,16906,16912,16901,16900,16903,16895,16894,16897,16890,16892,16888,16886,16866,16887,16891,16884,16881,16885,16889,16882,16879,16883,16878,16937,16940,16936,16935,16938,16930,16898,16929,16927,16933,16939,16934,16951,16931,16932,16926,16924,16917,16925,16922,16928,22630,22634,22629,16923,22628,22632,22626,22625,16921,16920,16918,16919,16913,16909,16916,22631,22633,22627,22640,22810,22639,22635,22642,22807,22641,22636,22647,22653,22646,22649,22643,22644,22638,22637,22645,16956,16954,16950,16943,16948,16946,16947,16941,16952,16944,16945,16942,16877,16873,16874,16870,16869,16867,16868,16864,16824,16863,16857,16856,16852,16859,16876,16858,16854,16861,16880,16860,16855,16862,16865,16853,16847,16845,16844,16846,16851,16875,16816,16817,16812,16809,16807,16814,16804,16803,16805,16794,16786,16787,16781,16776,16775,16755,16754,16757,16749,16750,16747,16752,16756,16751,16748,16753,16758,16759,16760,16761,16765,16766,16767,16768,16770,16973,16769,16773,16774,16779,16968,16780,16777,16963,16962,16965,16967,16964,16961,16966,16969,16953,16949,16957,16958,16959,16872,16815,16871,16960,16806,16970,16971,16976,16955,16977,16979,16974,16975,16978,16972,16981,16984,16980,16982,16986,16983,16988,16989,16991,16994,16992,16996,17001,22657,22656,22658,22660,22659,22655,22661,22664,22662,22654,22667,22800,22666,22669,22668,22673,22677,22799,22798,22803,22796,22795,22687,22686,22790,22680,22676,22681,22684,22678,22674,22679,22682,17026,17025,17022,17021,17019,17020,17018,17016,17007,17017,17014,17009,17015,22671,17012,17008,17013,17010,22665,22663,17011,17004,17005,17000,16999,16997,16998,16995,16993,16990,17002,17003,17006,22670,22672,22675,17023,17024,17027,17029,17028,17033,22683,17037,17036,22689,22692,22688,22690,22691,22694,22695,22775,22774,22777,22779,22776,22778,22772,22773,22769,22766,22771,22770,22782,22763,22760,22762,22764,22756,22754,22755,22751,22750,22744,22743,22741,22742,22739,22736,22738,22737,22740,22504,22503,22501,22496,22502,22493,22490,22492,22728,22488,22411,22487,22485,22486,22481,22413,22480,22478,22483,22484,22482,22495,22494,22499,22505,22500,22577,22497,22498,22580,22576,22575,22510,22511,22514,22508,22509,22507,22506,22476,22473,22477,22479,22474,22471,22475,22416,22417,22412,22408,22406,22409,22385,22414,22415,22381,22418,22424,22419,22378,22421,22428,22422,22376,22423,22426,22431,22433,22432,22439,22429,22440,22441,22444,22434,18313,18314,22445,18307,22451,22455,22452,22459,22454,22447,22457,22462,22456,22449,22460,22464,22461,22466,22450,22448,22442,22443,22436,22435,22438,22437,22425,22467,22470,22420,22469,22472,22468,22513,22512,22516,22519,22517,22465,22520,22523,22521,22463,22524,22522,22530,22536,22518,22515,22541,22446,22458,18708,22375,22372,18712,22371,22369,22374,22377,22373,22370,22379,22380,22382,22362,22383,22386,22384,22358,22387,22390,22388,22355,22391,22394,22392,22348,22393,22350,22396,22398,22397,22400,22402,22404,22407,22403,22395,22405,22410,22724,22723,22727,22730,22729,22734,22735,22726,22732,22731,22725,22733,22721,22717,22765,22716,22711,22706,22712,22708,22705,22703,22699,22696,22698,22697,22693,17039,17038,17041,17034,17032,17035,17040,17042,17043,17044,17046,17049,17045,22701,22700,22702,17054,22704,22337,22334,22707,22709,22714,22710,22340,22713,22715,22718,22399,22351,22341,22338,22335,22339,22332,22344,22345,22342,22343,22327,22323,22322,22318,22315,22317,22314,22311,22313,22347,22309,22304,22310,22353,22307,22303,22308,22357,22305,22301,22306,22361,22126,22122,22127,22365,22115,22114,22113,18722,22109,22108,18720,18717,18721,18735,18719,18714,18713,18711,18710,18698,18697,18695,18701,18693,18694,18691,18339,18327,18338,18332,18328,18333,18336,18330,18329,18331,18334,18278,18282,18325,18707,18702,18705,18709,18715,18718,18716,22367,22366,22106,18734,18741,18728,18729,18732,18733,18740,18747,18746,18751,18760,22098,22099,22097,22095,22102,22103,22105,22110,22104,22111,22112,22107,22116,22120,22117,22121,22123,22131,22124,22130,22298,22300,22296,22282,22297,22294,22302,22291,22295,22299,22292,22290,22293,22288,22312,22285,22287,22316,22283,22280,22284,17070,22320,22321,22319,22324,17066,17067,17063,17064,17058,17059,17056,17055,17053,17061,17062,17057,22329,22333,22328,22326,22331,22330,22325,22336,17052,17051,17050,17047,17048,17060,17065,17068,17069,17071,17075,17076,17078,17072,17073,17074,17077,17079,17080,17081,17083,17082,17086,17087,22261,22266,22259,22250,22260,22264,22257,22253,22258,22262,22254,22252,22255,22256,22142,22146,22139,22136,22138,22140,22134,22133,22092,22093,22087,22090,22083,22082,22080,22081,22075,22074,18767,18764,18768,18771,18766,18762,18769,18774,18770,18754,18772,18776,18773,18752,18778,18781,18779,18756,18782,18785,18783,18758,18791,18793,18792,18789,18795,19038,18797,18796,19033,19035,19032,18802,19031,19030,18870,18865,18872,18991,18871,18874,18867,18864,18866,18860,18869,18873,18868,18875,18862,18859,18861,18863,18855,18848,18856,18853,18854,18850,18815,18849,18825,18821,18830,18846,18828,18823,18829,18826,18835,18843,18834,18831,18837,18840,18836,18838,18833,18832,18593,18594,18588,18589,18583,18577,18586,18590,18587,18581,18919,18918,18922,18923,18925,18924,18928,18929,18931,18932,18934,18935,18937,19255,18936,18940,19245,18941,18938,18944,19241,18945,18942,18948,19235,18949,18952,18946,18947,18943,18916,18913,18917,18914,18920,18939,18921,18911,18915,18910,18839,18907,18841,18906,18909,18897,18893,18898,18904,18895,18891,18896,18894,18901,18903,18905,18955,18954,18958,18961,18959,18962,18963,18965,18968,18964,19200,19215,19199,19193,19202,19201,19213,19214,19219,19211,19204,19212,19216,19209,19206,19210,19207,19346,19345,19352,19350,19354,19208,19355,17319,19353,19351,19349,19347,17361,17360,17323,17320,17327,17349,17326,17321,17329,17328,17345,17348,17344,17347,17337,17332,17342,17346,17343,17412,17340,17335,17341,17338,17419,17417,17421,17422,17427,17430,17426,17428,17425,17424,17423,17429,17431,17434,17435,17433,17436,17437,17438,17440,17441,17442,17446,17450,17447,17456,17457,17458,17461,17462,17463,17464,17465,17468,17469,17473,17475,17474,17478,17471,17470,17467,17472,17476,17460,17453,17459,17455,17451,17454,17449,17452,17402,17400,17404,17403,17445,17444,17409,17443,17448,17414,17406,17405,17407,17394,17370,17393,17401,17384,17368,17383,17379,17391,17390,17397,17399,17398,17395,17477,17483,17484,17486,17500,17485,17488,17479,17492,17487,17481,17490,17495,17489,17491,17482,17480,17493,17496,17494,17498,17501,17507,17502,17505,17509,17506,17511,17517,17541,17540,17542,17543,17548,17549,17550,17552,17555,17553,17557,17560,17558,17561,17564,17562,17569,17572,17570,17573,17576,17574,17581,17584,17582,17587,17589,17592,17596,17593,17594,17599,17603,17600,17602,17605,17606,17608,17610,17612,17615,17613,17616,17619,17617,17625,17626,17627,17630,17631,17635,17636,17638,17637,21381,21384,21377,21375,21379,21398,21378,21373,21380,21402,21376,21382,21372,21468,21383,21385,21475,21474,21477,21476,21470,21469,21467,21465,21464,21406,21409,21410,21459,21458,21456,21461,21473,21460,21457,21453,21455,21448,21445,21447,21451,21444,21441,21443,21449,21438,21433,21440,21439,21436,21775,21773,21771,21772,21770,21774,21615,21776,21778,21613,21612,21611,21788,21605,21606,21602,21601,21599,21596,21600,18512,21603,21604,18516,21607,21610,21608,18532,21617,21619,21618,21614,21621,21625,21766,21767,21764,21760,21765,21769,21762,21761,21759,21758,21755,21756,21750,21749,17621,17611,17620,17623,17614,17622,17629,17618,21415,17628,21394,17624,21412,21393,21392,17632,21396,21389,21388,21391,21395,21390,21397,21400,21399,21403,21413,21404,21401,21418,21416,21419,21417,21427,21426,21446,21442,21421,21423,21432,21422,21424,21420,21405,21407,21425,21408,21411,21429,21428,21434,21435,21437,21414,21763,21768,21431,21430,21757,17609,17607,21745,17604,17601,21744,21742,21738,21743,21746,21740,21739,21737,21731,21727,21732,21734,21729,21726,21728,21730,21724,21722,21646,21643,21645,21656,21641,21638,21642,21649,21639,21636,21640,21647,18563,18541,18562,18566,18560,18559,18543,18542,18540,18535,18536,18533,18524,18534,18537,18530,18526,21634,21635,21633,18528,21632,21637,21628,21623,21630,21644,21629,21626,21735,21733,21736,21631,21741,21748,21751,21753,21752,21754,21627,21624,21622,21620,18531,18527,18529,18522,18519,18521,18525,18517,18515,18518,18523,18490,18489,18488,18492,18483,18480,18482,18478,18486,18487,18484,18491,18494,18496,18497,18507,18506,18511,18514,18513,18501,18509,18510,18508,21586,18503,18502,18505,18498,18495,18499,18500,18453,18457,18450,18447,18485,18456,18459,18451,18452,18454,18458,18462,18476,18463,18466,18460,18472,18467,18470,18464,18465,18468,18469,18471,18549,18548,18554,18553,18596,18595,18598,18603,18599,18600,18415,18605,18608,18606,18621,18413,18409,18412,18407,18408,18401,18397,18402,18399,18400,18395,18390,18396,18241,18392,18385,18384,18387,18391,18386,18380,18389,18388,18393,18394,18651,18649,18652,18653,18655,18654,18660,18661,18663,18662,18794,18798,18799,18800,18801,18805,18806,18808,18858,18807,18803,18811,18812,18809,18810,18804,18656,18648,18657,18647,18645,18646,18641,18638,18642,18639,18640,18636,18635,18406,18403,18405,18626,18625,18629,18630,18627,18632,18631,18637,18643,18644,18634,18814,18813,18817,18820,18818,18633,18819,18619,18822,18824,18617,18616,18611,18609,18612,18602,18827,18601,18597,18591,18592,18580,18579,18557,18558,18569,18555,18550,18556,18552,18561,18568,18565,18564,18570,18572,18574,18585,18575,18567,21652,21659,21651,21648,21654,21661,21653,21650,21657,21703,21658,21655,21712,21710,21716,21718,21717,21714,21720,21723,21719,21715,21721,21725,17590,17586,17591,17597,17588,17598,17595,17583,17585,17579,17575,17580,17577,21709,21713,21707,21699,21708,21711,21706,21698,21704,21705,21700,21695,21701,21671,21702,21676,21667,21662,21666,21669,21663,21664,21660,21665,18927,18584,18926,18578,18582,21668,18933,21670,21672,19262,19261,19259,19270,21673,19269,19265,19283,21678,19281,19279,19282,19284,17536,17535,17532,17534,17528,17526,17527,17529,17514,17515,17510,17512,17521,17520,17524,17518,17519,17513,17497,17499,17522,17525,17523,19292,17503,17504,19299,19297,19300,19301,17392,17389,17396,17387,17388,17385,19303,19306,19307,19308,17386,19312,19314,19313,19310,19316,19319,19317,19325,19326,19330,19334,19331,19315,19332,19338,17376,17372,17378,17382,17377,17380,17374,17369,17375,17367,17364,17359,17365,17362,17357,17363,17373,19340,17356,17353,17351,17350,17354,17358,17355,17352,17371,17408,17410,17411,17415,17418,17416,17413,17420,17432,17366,17381,19339,19342,19341,19348,19344,19336,19333,19335,19329,19337,19343,19226,19221,19227,19232,19224,19220,19225,19222,19230,19231,19233,19238,19324,19239,19236,19249,19248,19321,19322,19320,19318,19276,19251,19275,19272,19294,19311,19293,19296,19289,19271,19288,19286,19291,19295,19290,19287,19285,19280,19277,17531,17530,19274,19278,19273,19266,19263,19267,19264,19258,19254,19260,19253,19252,19247,19256,19257,19250,19246,19243,19242,19240,19237,19234,19229,19228,19223,18956,18957,18953,18950,18951,18908,18960,19218,19217,19244,19268,19298,19302,19305,19304,19309,19323,19327,19328,17508,17533,17516,17537,17544,17538,17545,17547,17551,17546,21685,17556,17566,17554,21688,17565,17568,17559,21694,17567,17578,17563,21697,21692,21693,21696,21690,21687,21689,21691,21680,21681,21675,21674,21683,21677,21682,21684,21679,21686,17539,17571,18571,18576,18546,18547,18544,18545,18539,18477,18538,18474,18481,18493,18473,18551,18573,18604,18610,18613,18607,18615,18618,18614,18620,18623,18628,18622,18816,18624,18404,18398,18650,18857,18658,18659,18664,18383,18382,18376,18372,18377,18374,18665,18666,18667,18671,18672,18674,18737,18673,18677,18669,18670,18731,18678,18736,18730,18727,18739,18744,18738,18745,18748,18755,18749,18759,18761,18750,18742,18743,18753,18668,18757,18375,18368,18366,18367,18361,18364,18359,18373,18378,18370,18371,18379,18238,18235,18237,18233,18232,18230,18226,18411,18416,18410,18414,18418,18419,18222,18475,18479,18520,21747,21616,21609,21595,21787,21471,21472,21466,21478,21374,17641,17642,17647,17639,17640,17634,17633,17643,17646,17644,17645,17648,17439,17466,17339,17336,17334,17331,17330,17325,17333,17324,17322,17318,17315,17314,17313,17310,17317,17316,17312,17309,17311,17308,19198,19196,19205,19191,19197,19203,19194,19190,19195,19192,19187,19188,19183,19181,18970,18969,18972,18966,18967,18971,18900,18890,18899,18902,18889,18887,18882,18888,18884,18880,18879,18852,18851,18847,18844,18892,18845,18877,18878,18881,18885,18886,18978,18883,18876,18988,18987,18989,18990,18993,18992,18996,19002,19028,19027,19029,19023,19036,19034,19039,19041,19040,19044,19037,19047,19045,19050,19042,19043,19048,18790,18787,18784,18788,19049,18786,22062,18780,22061,22063,18777,22064,22065,22066,18775,22069,22070,22071,22076,22072,22077,22079,22078,22084,22088,22086,22144,22085,22149,22143,22141,22147,22151,22148,22145,22153,22156,22244,22248,22243,22241,22246,22251,22245,22242,22247,17094,22249,17095,17091,17090,17088,17089,17085,17084,17092,17093,17096,17097,17099,17098,17102,17103,22238,22240,22236,22234,22237,22239,22235,22233,22159,22155,22160,22218,22157,22152,22158,22163,22154,22150,22161,22167,22162,22164,22089,22169,22165,22073,22166,22171,22068,22057,22067,22053,22056,22060,22047,22046,19068,19069,19066,19051,19065,22043,19059,19060,19053,19052,19057,19058,19061,19054,19055,19046,19056,19024,19021,19025,19022,19014,19015,19008,19012,19004,19003,19001,18994,18995,18999,18985,18984,18983,18997,18981,18977,18982,18986,18979,18980,18974,18975,19132,18976,18973,19172,19171,19176,19178,19177,19174,19180,19182,19179,19175,19184,19186,19185,19189,17303,17300,17304,17307,17301,17298,17302,17305,17306,17297,17294,17293,17292,17296,17295,17299,19164,17291,17290,17288,17285,17289,17286,19163,19162,19160,19165,19166,19167,19168,19169,19173,19170,19143,19142,19140,19141,19135,19138,19130,19127,19131,19126,19010,19009,18998,19110,19005,19000,19007,19013,19006,19011,19016,19017,19019,19081,19020,19018,19074,19076,19073,19072,19026,19075,19067,19062,19070,19079,19071,19064,22035,22034,22038,22041,22039,22042,19063,22045,22044,22048,22049,22051,22052,22174,22173,22176,22179,22175,22178,22172,22181,22177,22170,22183,22190,22184,22168,22211,22210,22216,22217,22219,22220,22222,22221,22227,22226,22230,22232,17110,17111,17107,17106,17104,17105,17101,17100,17108,17109,17112,17113,17115,17114,17118,17119,22224,22223,22225,22228,22215,22214,22212,22213,22207,22208,22200,22204,22196,22189,22195,22198,22187,22182,22188,22193,22185,22180,22186,22191,22054,22050,22055,22058,22040,22037,22036,22033,22059,22875,22879,22874,22873,22032,22028,22029,22023,22022,22020,22021,22017,22018,22014,22013,19090,19086,19092,19091,19137,22009,19136,19134,21901,21900,22006,22004,22008,22011,22010,22015,22016,22024,22025,22027,22026,22864,22862,22869,22870,22871,22867,22876,22877,22878,22872,22885,22888,22883,22884,22880,22886,22882,22199,22881,22192,22201,22194,22887,22202,22197,22205,22894,22206,22921,22203,22919,22923,22209,22924,22922,22231,22229,17126,17127,17123,17122,17120,17121,17117,17116,17124,17125,17128,17129,17131,17130,17134,17135,22918,22920,22916,22914,22915,22917,22913,22911,22896,22893,22895,22898,22889,22909,22897,22900,22891,22890,22892,22866,22899,22901,22902,21981,21976,21984,21983,21979,21987,22861,21988,21985,22856,22858,22854,21992,22855,22007,22857,22859,22012,22003,21998,22002,22005,22000,21999,21997,21993,21996,21989,21990,21982,21986,21978,21968,21977,21980,21971,21966,21972,21969,21975,21964,21970,21973,17162,17163,17159,17160,17156,17157,17152,17148,17151,17147,17158,22905,22906,22903,22904,22907,22908,22910,22912,17142,17143,17139,17138,17136,17137,17133,17132,17140,17141,17144,17145,17146,17149,17150,17153,17154,17155,17164,17165,17166,17172,17167,17168,17170,17173,17171,17175,17176,17178,17182,17177,17179,17184,17174,17169,21954,21955,21953,21947,21958,21963,21959,21949,21961,21965,21960,21951,21962,21956,21967,21952,21957,21991,21918,21915,21917,21994,21911,21907,21912,21909,21908,21910,21913,17252,17256,17248,17245,17247,17244,17250,17255,17249,17251,17240,17237,17235,17230,17236,17243,17233,17234,17229,17226,17225,17221,17231,17232,17220,17238,17239,17241,17242,21920,21923,21919,21914,21922,21926,21921,21916,21927,21932,21950,21933,21938,21930,21931,21928,21936,21937,21934,21940,21943,21939,21941,21935,17192,17191,17188,17183,17180,17181,17186,17187,17189,17195,17190,17193,17194,17196,17200,17224,17201,17204,17198,17199,17202,17203,17205,17209,17212,17208,17210,17207,17206,17211,17213,17214,17215,17218,17216,17219,17217,17222,17223,17227,17228,21925,21924,17197,21929,21942,17185,21944,21946,21945,21948,17246,17253,17257,17254,17259,17260,17262,17261,17265,17266,21884,21882,21886,21888,21887,21889,21885,21883,19125,19128,19122,19104,19121,19118,21881,21879,21877,21880,21878,17269,21876,21875,21874,21872,17273,17270,17274,17271,17277,19161,17278,17275,17281,17282,19158,19156,19159,19154,19153,19151,19152,19148,19145,19149,19120,19119,19116,19112,19123,19124,19111,19108,19103,19109,19106,19099,19098,19084,19083,19087,19096,19089,19088,19093,19097,19095,19105,19094,19129,19133,21895,21894,21897,21890,21902,21896,21892,21899,21904,21898,21893,21906,21891,17258,21905,21995,21903,22001,19100,19102,19101,19107,19113,19115,19117,19114,21873,21871,21869,21868,21870,19085,19080,19144,19139,19147,19150,19146,19155,19157,17287,17279,17280,17276,17283,17284,17272,17268,17267,17264,17263,17161,21974,22860,22863,22865,22868,19082,19077,22019,19078,22030,22031,18842,18912,18930,18765,18763,22091,22094,22100,22101,22096,22119,22118,22129,22132,22128,22125,22135,22137,22270,22269,22271,22263,22278,22281,22276,22265,22277,22274,22289,22267,22275,22286,22272,22268,22273,22279,22364,22368,22360,22363,22356,22359,22352,22354,22346,22349,22719,22720,22722,22401,17031,17030,22389,22548,22572,22578,22581,22579,22584,22571,22586,22583,22569,22748,22588,22602,22749,22601,22847,22846,22850,22849,22845,22842,22844,22841,22768,22839,22836,22838,22835,22837,22840,22833,22831,22830,22828,22825,22829,22823,22820,22824,22843,22817,22819,22812,22802,22811,22809,22814,22813,22815,22816,22818,22622,22621,22624,22808,22805,22801,22797,22826,22794,22793,22791,22792,22788,22789,22780,22827,22783,22784,22786,22781,22832,22785,22787,22767,22761,22758,22759,22757,22752,22753,22834,22747,22746,22745,22582,22489,22491,22685,22804,22806,22651,22652,22650,16985,22648,16987,16764,16763,16782,16778,16771,16762,16772,16746,16905,16904,16902,16899,16896,16893,22604,22603,18303,18274,18260,18204,18197,18207,18206,18199,18212,18215,18196,18198,18195,18202,18203,18200,18194,18187,18184,18186,18183,18190,18191,18188,18182,18189,18178,18175,18177,18174,18169,18173,18171,21573,21575,21572,21570,21579,21583,21580,21574,21592,21591,21802,21805,21801,21798,21806,21809,21811,21810,21807,21812,21569,21568,21564,21561,21562,21565,18157,18153,18158,18161,18154,18155,18151,18150,18147,18146,18143,18149,18152,18142,18156,18159,18162,18160,18163,18166,18164,18167,18170,18168,18172,18176,18179,18180,18181,18165,21566,21567,21571,18185,18192,18193,21835,21840,21832,21834,21828,21556,18136,18135,18130,18129,18126,18132,18131,18138,18128,21535,21540,18125,18127,18122,18133,18119,18121,18118,18124,18123,18117,18112,18113,18110,18107,18111,18109,18105,18108,18106,18104,18102,18114,18115,18116,18120,21355,18075,18070,18069,21342,21339,21341,21340,21338,17660,17661,17666,17683,17697,17916,17909,17908,17905,17911,17918,6007,6008,6005,6004,6003,5585,6001,5999,6000,6002,5998,5996,5421,5420,5418,5419,5415,5414,5412,5413,5407,5408,5404,5406,5400,5388,5399,5395,5402,5403,5405,5984,5985,5983,5978,5986,5987,5989,5981,5982,5976,5972,5969,5971,5974,5966,5964,5967,5606,5968,5973,5597,5596,5594,5595,5589,5590,5586,5588,5578,5575,5579,5576,5583,5582,5587,5580,5577,5581,5523,5584,5521,5516,5522,5519,5525,5529,5528,5573,5572,5569,5568,5531,5530,5534,5526,5527,5524,5520,5532,5533,5535,5538,5539,5560,5562,5559,5556,5564,5565,5567,5566,5599,5598,5602,5603,5605,5604,5609,5613,5611,5610,5934,5933,5936,5935,5940,5941,5943,5942,5948,5949,5951,5950,5955,5959,5961,6048,5962,5953,6054,6052,6058,6060,6059,6062,6056,6053,6055,6057,6044,6042,6043,6036,6034,6035,6032,6030,6031,6028,6027,6026,5302,5301,5392,5294,5290,5299,5386,5298,5318,5292,5289,5286,5383,5382,5385,5389,5393,5390,5387,5394,5396,5401,5990,5991,5398,6033,6037,6029,6041,6038,6039,6045,6047,6046,6049,6050,6051,5965,5970,6040,5397,5391,6061,6063,6064,6068,6070,6069,6066,6071,6075,6078,6074,6076,6067,6065,5956,5957,5952,5954,5944,5945,5939,5924,5937,5922,5938,5917,5923,5932,5919,5915,5925,5926,5930,5931,5929,5928,5619,5617,5618,5621,5614,5612,5615,5616,5561,5555,5553,5557,5547,5548,5543,5542,5540,5541,5537,5536,5544,5545,5546,5549,5552,5550,5554,5627,5635,5636,5634,5632,5638,5641,5639,5646,5633,5630,5625,5631,5628,5827,5826,5830,5832,5831,5828,5834,5837,5835,5833,5841,5845,5895,5896,5894,5892,5901,5902,5904,5912,5903,5893,5906,5905,5909,5913,5911,5921,5910,5907,6088,6086,6092,6094,6091,6093,6087,6084,6081,6083,6085,6079,6077,6073,6072,6089,6090,6095,6097,6096,6099,6101,6098,6103,6107,6104,5900,6105,6108,5882,5883,5879,5880,5875,5876,5873,5872,5855,5852,5857,5856,5865,5864,5867,5871,5874,5870,6135,5866,5868,6141,5869,5862,8342,8345,8340,8341,8339,8007,8338,8334,8009,8333,8336,8331,8329,8011,8008,8010,8014,8005,8002,8004,7997,8012,8013,8015,8018,8019,8319,8322,8317,8316,8318,8320,8314,8311,8315,8032,8026,8033,8043,8030,8027,8031,8028,8036,8042,8037,8034,8039,8046,8038,8040,8035,7964,7963,7953,7949,7954,7947,7943,7948,7945,8041,8044,8047,8045,8050,7946,7936,7935,7933,7929,7934,7939,7931,7932,7926,7924,7925,7927,7917,7916,6234,6233,6230,6225,6224,6222,6223,6217,6216,6219,6214,6204,6215,6218,6211,6213,6206,6205,6203,6200,6199,6196,6202,6201,6207,6194,6192,6193,6195,6185,6184,6187,6182,6180,6183,6186,6181,6188,6179,6176,6178,6174,6173,6172,6170,6168,6169,6165,6163,6164,6158,6156,6157,6151,6150,6149,6126,6125,6133,6122,6120,6121,6112,6102,6111,6114,6106,6118,6113,6116,6109,6119,6115,6117,6110,6124,5878,5877,6127,6123,6129,6132,6130,6136,6139,6137,6134,6142,6145,6143,6140,8001,8000,8006,7998,8003,7999,7993,7992,6148,7990,7991,7989,6160,7988,7986,6171,6167,6177,6189,6175,7978,7977,7980,7974,7973,6191,6190,6198,6197,7970,7968,7960,7969,7975,7967,7962,7971,7982,7972,8025,7965,7961,7966,8029,7958,7957,7955,7952,7956,7959,6210,6209,6212,7950,7942,7951,7941,6220,6221,6226,6227,7938,7937,7940,7928,7944,8024,8021,8020,8023,8017,8016,7984,7981,7983,7995,7979,7976,7994,7985,7987,8022,6208,6162,6166,6159,6161,6155,6152,6154,6153,6146,6138,6147,7996,6144,6128,6131,6100,6228,6229,6231,6232,6235,6236,6240,6239,7909,7907,7912,7915,7913,7914,7918,7919,7922,7923,7930,7920,8057,8059,8056,8055,8061,8062,8063,8073,8058,8052,8285,8284,8286,8054,8289,8290,8291,8296,8292,8294,8053,8049,8048,8295,8300,8301,8303,8302,8307,8310,8312,8313,8326,8381,8325,8321,8328,8332,8327,8330,8323,8335,8337,8357,8372,8356,8347,8364,8370,8362,8363,8352,8351,8348,8346,8343,8344,8354,8349,8350,8353,5863,5861,5797,5792,5796,5794,5800,5859,5801,5803,5798,5795,5799,5802,5677,5675,5676,5806,5673,5669,5672,5674,5659,5663,5654,5653,5651,5652,5647,5645,5648,5643,5642,5644,5649,5650,5820,5822,5819,5656,5818,5813,5821,5824,5823,5829,5815,5839,5838,5840,5817,5842,5844,5846,5847,5851,5853,5848,5885,5888,5884,5881,5886,5889,5887,5898,5897,5891,5890,5849,5816,5812,5807,5810,5805,5811,5814,5808,5809,5804,5850,5854,5843,5825,5640,5637,5655,5657,5661,5658,5660,5662,5664,5665,5668,5681,5667,5670,5671,5678,5680,5679,5682,5684,5789,5683,5686,5685,5689,5694,5788,5787,5785,5780,5786,5783,5790,5793,5791,8359,5784,5781,5775,5782,5777,8369,8375,8367,8368,8365,8373,8376,8374,8371,8379,8382,8378,8380,8383,8384,8387,8388,8389,8393,8385,8386,8391,8392,8395,8399,8400,8401,8404,8397,8398,8410,8308,8304,8309,8416,8297,8293,8420,8421,8419,8422,8425,8423,8427,8418,8415,8417,8613,8412,8409,8411,8414,8407,8405,8406,8408,8403,5769,8402,5762,5770,5767,8396,8394,8390,5773,5768,5774,5771,5772,5766,5765,5733,5722,5732,5743,5727,5728,5719,5718,5716,5717,5713,5712,5707,5715,5714,5709,5720,5723,5721,5711,5776,5778,5779,5710,5704,5701,5703,5705,5696,5698,5693,5690,5692,5687,5688,5666,5691,5695,5697,5699,5700,5702,5708,5706,5724,5725,5726,5729,5730,5731,5734,5737,5735,5738,5741,5739,5744,5746,5745,5748,5751,5749,8582,8587,8581,8577,8584,8583,8588,8591,8594,8592,8586,8644,8642,8648,8650,8649,8646,8652,8655,8651,8653,8647,8645,8641,8640,8485,8484,8471,8470,8468,8469,8457,8452,8458,8465,8455,8456,8450,8449,8262,8263,8258,8259,8256,8257,8252,8248,8253,8131,8254,8255,8102,8097,8108,8106,8122,8125,8120,8117,8121,8123,8119,8118,7801,7800,7799,7803,7795,7794,7793,7789,7802,7804,7809,7805,6866,7807,7808,7810,7814,7818,8111,8112,8110,8109,7854,7855,7849,7847,7848,7844,7842,7820,7817,7819,7822,7815,7813,7816,7821,6859,6858,6857,6852,7824,7825,7823,7828,7833,7827,7826,7829,6675,7831,7834,7832,7836,6673,6672,6287,6284,6288,6285,6291,6292,6667,6666,6660,6669,6674,6668,6662,6671,6670,6676,6680,6677,6681,6837,6836,6832,6827,6834,6849,6833,6830,6841,6845,6840,6835,6843,6844,6846,6847,6872,6877,6871,6869,6875,6879,6876,6873,6881,6883,7779,7782,7778,7781,7773,7772,7770,7776,7777,7780,8154,8163,8153,8156,8149,8148,7785,7784,7787,7790,7786,7788,7783,6874,6870,6867,6865,6868,6863,6864,6854,6850,6853,6848,6856,6860,6855,6851,6862,6861,7791,7792,7796,8140,7798,7797,8128,8132,8127,8130,8124,8135,8129,8133,8126,8137,8134,8243,8242,8240,8246,8247,8249,8250,8462,8463,8461,8464,8467,8466,8472,8475,8473,8477,8482,8481,8661,8663,8658,8659,8660,8662,8657,8568,8656,8575,8654,8570,8580,8585,8578,8572,8579,8576,8574,8573,8571,8566,8564,8565,8561,8559,8560,8555,8553,8552,8550,8551,8545,8548,8541,8540,8538,8539,8533,8530,8534,8536,8519,8515,8518,8526,8513,8514,8510,8511,8507,8500,8506,8512,8502,8504,8497,8496,8494,8489,8495,8505,8491,8486,24557,24556,24558,8488,24561,24562,24563,8234,24564,24565,8228,8227,8225,8226,8219,8220,8217,8218,8210,8181,8208,8209,8203,8177,8201,8202,8198,8173,8196,8197,8193,8192,8175,8176,8171,8165,8161,8166,8169,8170,8179,8167,8164,8168,8172,7775,7774,7771,24142,7768,7767,7765,7766,7761,7762,7759,7760,7754,6895,7752,7749,7753,7757,7751,7750,7748,7747,7744,7743,7741,7742,7737,7735,7729,7730,7723,7713,7722,7725,7717,7718,7712,7711,7709,7707,7710,7714,7701,7696,7700,7703,7697,7694,7698,7699,7046,7047,7038,7035,7037,7040,7031,7028,7030,7025,7032,7039,7041,7042,7045,7050,7044,7051,7056,7057,7059,7058,7692,7690,7693,7691,7689,7687,7061,7062,7060,7684,7683,7685,7680,7681,7678,7677,7666,7166,7665,7662,7668,7667,7672,7675,7676,7679,7716,7688,7686,7719,7715,7695,7720,7721,7726,7728,7727,7733,7724,7736,7734,7731,7738,24176,7740,7756,7739,7755,24180,24178,24182,24181,24184,24186,7758,7764,24185,24183,24147,24144,24148,24153,24145,24143,24146,8190,24149,24150,24152,24155,24154,24161,24162,24163,24157,24168,24179,24167,24165,24170,24169,24174,24177,24173,24175,24172,24171,7674,7659,7673,7732,7669,7671,7663,7660,7664,7661,7658,7179,7657,7651,7652,7648,7647,7178,7175,7171,7285,7284,7288,7289,7301,7286,7287,7283,7280,7278,7279,7277,7282,7281,7290,7269,7265,7270,7267,7268,7260,7259,7251,7250,7248,7249,7242,7241,7239,7237,7240,7246,7247,7252,7253,7255,7254,7257,7263,7256,7258,7261,7262,7264,7295,7298,7294,7291,7296,7299,7297,7302,7293,7292,7300,7303,7639,7637,7641,7644,7642,7646,7650,7649,7653,7654,7656,7655,24166,7645,7643,7640,7636,24160,24159,24164,24158,7631,7632,7627,7624,7626,7622,24156,7623,7618,7619,7616,7615,7361,7355,7360,7362,7346,7345,7343,7338,7344,7340,7625,7336,7333,7335,7331,7328,7330,7332,7325,7324,7322,7320,7315,7321,7326,7318,7314,7319,7323,7316,7317,7309,7308,7311,7304,7305,7306,7307,7310,7312,7313,7634,7633,7630,7635,7638,7629,7628,7334,7327,7329,7337,7339,7341,7342,7347,7348,7350,7351,7353,7356,7352,7349,7354,7358,7363,7359,7375,7374,7606,7605,7607,7611,7357,7612,7609,7617,8186,7621,8191,7620,8189,8185,8183,8184,8188,8187,7614,7610,7613,7608,8194,8195,7602,7601,7600,7598,8200,8199,8204,8205,8207,8213,8206,8214,8216,8215,24577,24580,24576,8221,24575,24578,24571,24567,24572,24574,24569,24566,24568,24570,24559,24560,24553,24552,24550,24551,24545,24544,8517,8516,8524,24539,8523,8527,8520,8525,8528,8529,8531,24427,8532,8535,24420,24425,24419,24418,8557,8556,8549,8558,8554,24416,24415,24417,24421,24424,24422,24428,24432,24429,24434,24435,24441,24444,24442,24443,24446,24447,24448,24454,24449,24451,24452,24456,24462,24457,24453,24460,24464,24461,24466,24467,24468,7498,24469,24471,24470,24473,24475,24474,24476,24472,7520,7524,7516,7515,7512,7507,7513,7500,7499,7497,7493,7492,7488,7495,7494,7496,7490,7487,7491,7489,7503,7485,7469,7484,7482,7479,7481,7480,7483,7486,7478,7476,7477,7472,7473,7470,7471,7458,7457,7456,7467,7454,7451,7455,7461,7453,7452,7450,7449,7448,7447,7446,7418,7414,7424,7437,7423,7416,7426,7430,7425,7428,7421,7422,7417,7415,7412,7407,7419,7420,7433,7409,7564,7405,7400,7404,7402,7401,7403,7406,7408,7410,7411,7413,7399,7396,7395,7397,7390,7389,7388,7393,7386,7385,7391,7380,7377,7382,7387,7381,7378,7384,7578,7383,7592,7379,7373,7372,7370,7367,7371,7376,7366,7369,7365,7364,7368,7595,7594,7599,7603,7604,7597,7596,7591,7588,7582,7590,7593,7589,7586,8212,8211,24585,24586,24584,24582,24588,24591,24587,24581,24589,24594,24583,24579,24607,24606,24608,24573,24609,24611,24554,24555,24549,24548,24546,24547,24541,24540,24538,24533,24532,24535,24528,24527,24431,24430,24438,24525,24436,24437,24426,24423,24440,24433,24439,24520,24519,24523,24524,24521,24529,24530,24531,24526,24536,24622,24537,24534,24542,24543,24616,24620,24614,24613,24615,24618,24612,24610,24605,24604,24603,24602,24600,24599,24596,24593,24595,24597,24592,24590,7585,7575,7584,7587,7581,7579,7574,7569,7568,7566,7563,7567,7571,7561,7557,7562,24598,7559,7554,7560,24601,7556,7553,24619,24617,24623,24625,24624,24621,24627,24631,24626,24628,24629,24518,24632,24634,24513,24635,24502,24633,24630,7550,7547,7549,7552,7545,7543,7546,7551,7443,7434,7442,7438,7538,7536,7537,7534,7440,7532,7529,7533,7530,7535,7539,7540,7542,7548,7541,7544,24497,24496,24494,24499,24503,24498,24495,24501,24508,24500,24504,24492,24491,24487,24486,24483,24482,24481,24478,24480,24479,24463,24485,24484,24477,24488,24489,24493,7526,24490,7523,7525,7521,7531,7522,7519,7517,7511,7508,7510,7506,7465,7505,7468,7463,7462,7459,7460,7444,7464,7528,7527,7445,7441,7439,7436,7504,7509,7514,7501,7502,7518,24505,24506,24455,24507,24458,24512,24514,24510,24511,24516,24517,24522,24515,24459,24445,24450,24509,7431,7435,7429,7427,7432,7555,7558,7565,7570,7572,7394,7573,7392,7576,7580,7577,7583,7398,7466,7475,7474,24465,24151,7266,7233,7236,7238,7232,7229,7231,7227,7225,7223,7224,7226,7222,7220,7212,7209,7211,7213,7208,7105,7207,7204,7203,7112,7107,7115,7198,7116,7110,7119,7122,7195,7193,7197,7201,7205,7202,7199,7235,7243,7234,7230,7228,7214,7210,7206,7245,7244,7272,7200,7196,7192,7273,7274,7276,7187,7184,7186,7182,7180,7183,7188,7181,7143,7144,7139,7138,7134,7130,7135,7126,7189,7190,7127,7123,7111,7108,7109,7103,7101,7104,7106,7099,7098,7100,7102,7097,7095,7096,7092,7091,7093,7088,7084,7087,7090,7082,7080,7083,7079,7085,7077,7075,7076,7073,7078,7072,7081,7086,7125,7124,7128,7149,7129,7071,7131,7682,7151,7153,7152,7157,7155,7150,7146,7145,7142,7148,7154,7147,7159,7177,7158,7163,7156,7160,7164,7162,7167,7161,7165,7168,7169,7170,7173,7174,7172,7185,7176,7137,7140,7136,7141,7133,7117,7132,7118,7114,7094,7113,7089,7120,7121,7068,7067,7065,7066,7063,7064,7053,7052,7055,7049,7048,7043,7054,7069,7070,7074,7191,7194,7275,7271,7215,7216,7218,7217,7219,7221,7670,8180,7023,7017,7024,7021,7022,7016,7013,7015,7009,7007,7001,7008,7005,7006,7000,6997,6999,6993,6991,6990,6992,6989,6983,6988,6981,6987,6985,6984,6982,6986,23282,23283,23281,6980,23280,23279,6976,6969,6975,23271,6971,6966,6972,23267,6760,6759,6757,6752,23261,23260,23266,23268,23269,23272,23275,23273,23270,23277,23290,23278,23276,23286,23285,23287,23288,23289,23291,23293,23294,23295,23296,23297,23298,23301,23302,23303,23307,23308,23312,23313,23314,23322,23324,23323,23320,23328,23329,23331,23342,23330,23332,23344,23345,23346,23347,23349,23348,23352,23354,23353,23351,23358,23359,23365,23366,23396,23394,23397,23398,23355,23399,23400,23403,23404,23405,23408,23412,23409,23410,23413,23414,23415,23420,23416,23418,23422,23419,23424,23426,23425,23430,23436,23431,23434,23427,23428,23429,23432,23435,23433,23438,23441,23439,23445,23446,23447,23448,23450,23455,23449,23453,23460,23454,23451,23495,23494,23497,23503,23498,23501,23452,23442,23440,23443,23437,23502,23507,23510,23509,23508,23521,23621,23520,23523,23513,23515,23514,23518,23519,23522,23534,23533,23544,23615,23545,23542,23558,23610,23557,23552,23560,23559,23607,23609,23606,23566,23605,23611,23603,23599,23604,23601,23608,23652,23613,23623,23646,23622,23628,23617,23612,23618,23614,23620,23626,23619,23616,23627,23630,23624,23625,23629,23631,23633,23506,23505,23634,23637,23638,23642,23641,23741,23746,23740,23739,23649,23650,23644,23643,23647,23640,23636,23639,23635,23632,23645,23648,23655,23654,23731,23733,23730,23670,23729,23726,23732,23734,23728,23736,23737,23738,23742,23727,23723,23719,23724,23721,23743,23750,23752,23751,23755,23745,23744,23747,23753,23756,23754,23759,23749,23748,23500,23499,23496,23760,23491,23488,23487,23483,23761,23764,23765,23766,23489,23771,23769,23775,23777,23776,23773,23780,23782,23785,23786,23774,23790,23793,23792,23797,23796,23795,23798,23713,23712,23710,23707,23711,23709,23704,23703,23701,23698,23697,23694,23700,23699,23693,23705,23706,23708,23714,23695,23696,23691,23686,23685,23683,23680,23684,23692,23679,23678,23682,23681,23677,23687,23688,23690,23689,23720,23725,23675,23673,23669,23664,23665,23658,23659,23656,23657,23653,23602,23600,23596,23593,23595,23589,23598,23597,23590,23574,23570,23575,23578,23572,23568,23573,23576,23192,23191,23188,23189,23184,23183,23179,23186,23185,23187,23181,23180,23178,23176,23174,23175,23169,23167,23164,23168,23172,23166,23163,23165,23170,23160,23155,23161,23085,23162,6397,6398,6393,6392,6390,6387,6391,6386,6384,6385,6379,6378,6382,6376,6367,6375,6371,6380,6383,6381,6389,6377,6373,6403,6406,6402,6400,6408,6410,6413,6414,6416,6417,6419,6420,6568,6573,6567,6564,6570,6575,6569,6566,6577,6580,6578,6571,6582,6586,6589,6590,6593,6595,6592,6598,6599,6734,6737,6733,6730,6736,6739,6735,6732,6741,6743,6744,6745,6749,6750,6746,23095,23094,23257,23256,23259,23264,23265,23262,23306,23304,23309,23311,23310,23263,23315,23317,23316,23318,23321,23326,23325,23327,23319,23253,23247,23252,23255,23249,23246,23250,23251,23116,23115,23090,23087,23091,6742,6740,23086,23084,23113,23111,23108,23112,23117,23109,23107,23110,23114,23105,23102,23106,23125,23103,23104,23101,23096,23093,23097,23071,23099,23100,23070,23136,23137,23135,23140,23133,23132,23138,23129,23131,23134,23130,23235,23127,23124,23128,23239,23126,23122,23245,23244,23241,23248,23119,23121,23118,23120,23238,23334,23335,23333,23336,23337,23339,23341,23338,23361,23360,23362,23243,23367,23369,23368,23364,23374,23383,23375,23372,23380,23379,23385,23387,23386,23389,23390,23392,23393,23402,23406,23401,23407,23529,23528,23530,23531,23532,23535,23538,23536,23388,23540,23546,23541,23543,23547,23550,23549,23553,23555,23554,23551,23556,23215,23561,23562,23563,23569,23564,23213,23565,23567,23205,23207,23203,23201,23198,23194,23193,23196,23195,23197,23177,23182,23173,23199,23171,23158,23156,23157,23152,23151,23149,23150,23143,23142,23144,23146,23077,23078,23074,23073,6412,6405,6411,23064,6409,6404,6399,6394,6388,6396,6395,6401,23067,23063,6415,23062,23065,23060,6574,23059,23061,6581,6576,6584,6610,6583,6579,6587,6605,6588,6585,6609,6604,23080,6602,6738,23081,23079,23089,23076,23075,23072,6612,23092,23088,23098,23066,23069,23068,23082,23083,23147,23148,23153,23154,23145,23223,23141,23222,23224,23226,23139,23225,23231,23236,23232,23234,23240,23237,23242,23370,23371,23373,23376,23377,23378,23233,23228,23227,23229,23218,23214,23217,23220,23212,23209,23211,23208,23216,23219,23206,23202,23210,23204,23159,23200,23230,23548,23221,23381,23382,23571,23537,23539,23526,23524,23525,23527,23517,23516,23421,23417,23411,23395,23391,23384,23363,23357,23356,23350,23343,23340,23123,23254,23258,23305,23300,23274,23299,6731,6747,6748,6751,6755,6756,6753,6761,6762,6764,6961,6763,6963,6960,6959,6768,6773,6783,6771,6772,6765,6766,6758,6754,6726,6725,6722,6716,6713,6717,6603,6720,6721,6600,6723,6724,6727,6728,6596,6729,6594,6597,6549,6548,6544,6545,6542,6540,6517,6518,6513,6514,6511,6510,6496,6494,6495,6503,6492,6488,6493,6497,6490,6483,6491,6485,6501,6500,6635,6634,6638,6639,6643,6645,6642,6641,6644,6647,6304,6305,6300,6301,6296,6294,6297,6302,6303,6306,6307,6309,6308,6312,6313,6327,6640,6326,6322,6481,6480,6478,6324,6477,6472,6482,6484,6479,6474,6486,6489,6487,6476,6499,6522,6498,6523,6521,6526,6519,6537,6536,6538,6530,6531,6534,6528,6527,6529,6532,6451,6450,6448,6442,6449,6439,6533,6557,6558,6556,6560,6553,6552,6535,6555,6543,6539,6547,6550,6546,6541,6554,6551,6591,6572,6562,6565,6561,6563,6559,6428,6423,6427,6430,6421,6422,6418,6424,6407,6374,6372,6370,6425,6364,6361,6365,6362,6368,6369,6366,6363,6359,6358,6357,6355,6352,6356,6354,6360,6426,6431,6432,6433,6435,6353,6350,6349,6347,6344,6348,6351,6343,6342,6346,6339,6338,6341,6335,6334,6331,6337,6336,6340,6445,6452,6444,6441,6447,6454,6446,6443,6440,6436,6460,6455,6458,6459,6462,6463,6470,6469,6475,6467,6473,6468,6471,6464,6457,6465,6466,6325,6329,6321,6320,6318,6315,6319,6328,6330,6333,6332,6314,6311,6317,6316,6310,6323,6456,6453,6524,6525,6461,6345,6438,6437,6429,6434,6293,6290,6298,6299,6656,6295,6659,6654,6649,6655,6661,6652,6648,6653,6663,6650,6646,6651,6630,6657,6658,6665,6679,6682,6678,6685,6683,6684,6687,6632,6627,6633,6690,6629,6625,6636,6693,6637,6623,6703,6702,6706,6707,6709,6714,6711,6710,6770,6774,6769,6767,6712,6777,6775,6778,6779,6781,6786,6780,6784,6787,6785,6782,6950,6949,6955,6956,6958,6957,6964,6962,6967,6970,6968,6973,6974,6979,6977,6954,6953,6951,6952,6946,6947,6944,6945,6941,6940,6788,6939,6793,6790,6798,6809,6797,6800,6791,6792,6789,6719,6794,6799,6795,6699,6796,6701,6801,6803,6802,6805,6807,6810,6813,6811,6815,6818,6820,6900,6821,6906,6908,6905,6910,6903,6904,6901,6907,7745,6911,6923,6922,6918,6913,6920,7708,6919,6916,6925,6924,7705,7706,7704,7702,7034,6928,7033,7036,7027,7020,7026,7029,6932,7018,7012,7019,7014,7011,7010,7004,6937,7002,6996,7003,6998,6994,6995,6948,6943,6942,6938,6935,6930,6936,6929,6817,6808,6816,6915,6812,6814,6909,6914,6917,6921,6926,6927,6931,6934,6933,6806,6912,7746,6902,6896,6897,6892,6890,6893,6839,6898,6899,6824,6829,6828,6831,6822,6823,6819,6804,6704,6705,6695,6691,6694,6689,6697,6696,6692,6688,6826,6825,6838,6842,6889,6894,6882,6878,6886,6891,6885,6880,6888,7763,6887,6884,6700,6698,6718,6715,6978,6776,6708,6615,6611,6616,6606,6607,6520,6608,6516,6515,6512,6614,6509,6507,6508,6618,6505,6502,6506,6621,6504,6631,6620,6626,6628,6624,6617,6622,6613,6619,6686,6664,6289,6286,6601,6965,23190,23577,23580,23582,23581,23579,23585,23588,23584,23583,23586,23591,23592,23594,23660,23661,23662,23663,23666,23668,23667,23671,23672,23674,23676,23587,23717,23716,23715,23718,23722,23794,23757,23758,23768,23772,23767,23770,23762,23763,23781,23784,23783,23779,23492,23778,23490,23486,23484,23481,23485,23482,23480,23466,23463,23465,23459,23457,23464,23469,23461,23462,23458,23456,23467,23472,23468,23473,23476,23479,23478,23493,23504,23735,23651,23512,23511,23423,23444,23292,23284,7769,8162,8174,8159,8155,8160,8178,8157,8150,8158,8182,8152,8147,8223,8222,8224,8145,8229,8230,8231,8232,8235,8236,8238,8237,8241,8487,8245,8244,8480,8483,8479,8478,8251,8490,8474,8476,8493,8492,8498,8501,8499,8503,8509,8508,8522,8521,8547,8546,8542,8537,8544,8543,8563,8562,8567,8239,8233,8138,8141,8136,8144,8139,8146,8142,8143,8151,8569,6282,6281,6278,6277,6276,6280,6273,6275,6270,6269,6268,6272,6265,6264,6261,6267,6271,6266,7871,6263,6262,6260,6257,6254,6259,6258,7885,6256,6255,6252,6253,6249,6248,6251,6245,6244,6247,6243,6241,6242,6238,6237,6246,7899,7898,7895,7901,7900,7906,7908,7910,7911,7921,8060,8064,8067,8065,8068,8069,8071,8076,8070,8078,8085,8079,8072,8277,8276,8280,8281,8283,8282,8288,8287,8426,8279,8278,8275,8429,8433,8428,8431,8435,8430,8432,8442,8441,8618,8619,8616,8614,8615,8617,8611,8607,8612,8608,8602,8413,8601,8600,5759,5756,5760,5758,5754,5752,5753,5740,5755,5742,5761,5763,5764,5736,5747,5757,8598,5750,8597,8595,8603,8606,8604,8599,8624,8622,8628,8633,8629,8631,8626,8623,8625,8621,8627,8630,8454,8453,8447,8448,8439,8436,8440,8445,8437,8434,8438,8443,8270,8271,8267,8266,8264,8265,8261,8260,8098,8092,8099,8096,8090,8088,8089,8091,8081,7892,8080,8083,8075,8074,7903,7902,7893,7890,7891,7884,7882,7883,7875,7874,7877,7876,7873,7880,7881,7887,7886,7878,8095,8100,8094,8101,8103,7879,7869,7864,7867,7868,7862,7859,7861,7858,7863,7865,6274,7866,7870,7857,6279,7837,7856,7851,7839,7850,7845,7853,7860,7852,7841,7846,7843,7840,7835,7830,7838,6283,7872,8104,8105,8107,8115,8113,8114,8116,7896,7888,7889,7894,7897,7905,7904,8077,8082,8086,8093,8087,8084,8269,8268,8272,8273,8444,8274,8446,8451,8459,8460,8632,8638,8639,8636,8637,8643,8620,8634,8635,8609,8610,8596,8593,8589,8590,8605,8424,8066,6250,7812,7811,7806,8306,8299,8298,8305,8377,8366,8361,8360,8358,8355,5858,5860,8051,8324,5908,6080,6082,5920,5958,5916,5918,5914,5899,5836,5927,5629,5624,5622,5620,5623,5558,5551,5626,5960,5963,5946,5947,5608,5607,5601,5600,5593,5592,5570,5571,5574,5563,5517,5515,5512,5591,5995,5997,5993,5992,5979,5975,5980,5977,5994,5988,5409,5384,5410,5417,5424,13041,13039,13042,13044,13047,13043,13045,13037,12906,13036,13034,12909,13032,13033,13029,12924,13027,13028,13023,12932,13021,13020,13022,13025,13017,13018,13015,12979,13014,13011,13024,13026,13016,13031,13030,13035,13237,13040,13038,13238,13236,13235,13219,13218,13223,13216,13215,13213,13220,13222,13221,13232,13217,13224,13214,13211,13212,13208,13206,13209,13056,13055,13051,13050,13053,13052,13057,13048,13046,13049,13054,12898,12899,12895,12894,12892,12893,12889,12888,12854,12851,12855,12861,12852,12849,12853,12856,12806,12805,12803,12804,12799,12798,12802,12796,12792,12797,12800,12794,12790,12793,12780,12795,12776,12801,12807,12808,12810,12809,12813,12814,12845,12848,12844,12839,12847,12850,12846,12842,12897,12900,12896,12901,12903,12902,12843,12904,12907,12905,12841,12908,12910,12911,12917,12912,12913,12837,12834,12832,12833,12835,12830,12824,12829,12826,12827,12825,12751,12823,12820,12819,12758,12759,12752,12749,12745,12750,12756,12747,12743,12748,12753,12757,12754,12755,12760,12762,12765,12766,12769,12768,12772,12773,12770,12767,12771,12778,12781,12779,12782,12785,12783,12784,12786,12788,12789,12791,12787,12774,12775,12777,12764,12812,12815,12811,12816,12818,12817,12821,12822,12836,12838,12763,12761,12742,12740,12738,12735,12739,12736,12744,12746,12741,12737,12732,12729,12731,12727,12828,12831,12724,12728,12914,12725,12723,12726,12916,12720,12717,12719,12920,12714,12713,12711,12706,12712,12708,12703,12715,12718,12716,12705,12721,12722,12704,12699,12698,12696,12692,12697,12694,12702,12693,12695,12700,12701,12691,12686,12684,12681,12685,12682,12689,12690,12687,12688,12683,12734,12733,12730,12679,12678,12680,12677,12707,12709,12710,12938,12942,12937,12940,12929,12930,12926,12925,12927,12928,12933,12922,12919,12921,12931,12918,12915,12934,12936,12941,12935,12939,12959,12958,12960,12965,12969,12964,12967,12962,12961,12957,12963,12966,12968,12970,12971,12973,12972,12976,12980,12977,12982,12974,12975,12983,12985,12984,12991,12992,12997,12999,13000,13003,13001,12987,13005,13007,13006,13009,13012,13243,13244,13241,13240,13239,13242,13234,13233,13230,13226,13228,13229,13231,13250,13248,13253,13254,13255,13260,13256,13258,13264,13259,13262,13266,13263,13272,13227,13225,13277,13276,13278,13210,13279,13282,13280,13283,13204,13205,13201,13200,13198,13199,13195,13194,13072,13066,13071,13082,13068,13067,13064,13063,13061,13062,13058,13069,13075,13070,13073,12890,12891,12885,12886,12880,12879,12877,12878,12874,12871,12873,12869,12875,12870,12865,12868,12863,12862,12860,12864,12859,12858,12857,12866,12867,12872,12876,12881,12882,12884,12883,13078,13077,13086,13087,13088,13089,13090,13091,13092,13093,13098,13099,13100,13102,13105,13103,13107,13112,13108,13110,13113,13111,13115,13118,13116,13117,13120,13122,13121,13126,13127,13132,13133,13135,13134,13164,13167,13163,13316,13160,13137,13159,13319,13156,13144,13155,13324,13150,13151,13146,13145,13143,13141,13142,13140,13138,13129,13139,13136,13131,13130,13128,13125,13124,13123,13147,13148,13149,13152,13153,13154,13157,13336,13158,13161,13339,13162,13165,13345,13166,13171,13346,13172,13352,13351,13353,13354,13357,13355,13358,13361,13359,13881,13879,13882,13884,13883,13886,13887,13888,13889,13892,13895,13893,13902,13903,13904,13905,13906,13907,13909,13911,13910,13913,13908,13912,13918,13919,13920,13923,13921,14137,13916,13917,14142,13914,13915,24204,24202,24205,24203,24201,24199,24200,24197,24196,24195,24193,24191,24189,24192,24190,24194,14146,14145,14143,14144,14139,14136,14138,14140,14133,14131,14134,13927,13928,13931,13924,13925,13922,13926,13899,13898,13896,13897,13891,13890,13868,13863,13878,13876,13885,13877,13866,13861,13867,13850,13869,13870,13874,13873,13872,13875,13365,13871,13370,13367,13374,13622,13373,13376,13368,13366,13369,13371,13348,13347,13342,13343,13338,13340,13372,13377,13378,13380,13384,13379,13375,13382,13480,13381,13618,13617,13621,13624,13623,13633,13845,13634,13626,13844,13847,13848,13849,13857,13856,13853,13859,13860,13862,13865,13864,13930,13933,13929,13932,13935,13939,14129,14130,14128,14081,14127,14132,14123,14125,14118,14085,14117,14120,14114,14089,14113,14112,14116,14119,14115,14111,14121,14122,14161,14163,14160,14179,14157,14156,14154,14162,14164,14175,14177,14174,14176,14171,14170,14167,14168,14165,14158,14159,14153,14151,14152,14148,14147,14149,14135,14150,14155,24188,24187,14173,14172,14169,14178,14180,14181,14182,14183,14185,14187,14184,14186,14189,14192,14188,14190,14191,14193,14194,14196,14197,14198,14200,14201,14202,14203,14205,14204,14208,14209,14213,14216,14214,14221,14199,14166,14195,14124,14223,14224,14108,14227,14230,14228,14106,14233,14567,14234,14236,14235,14238,14237,14559,14557,14564,14566,14562,14563,14558,14555,14550,14556,14560,14553,14554,14551,14549,14533,14528,14535,14536,14546,14544,14545,14540,14537,14541,14542,14410,14409,14279,14276,14280,14282,14277,14278,14275,14272,14271,14270,14269,14274,14273,14268,14267,14265,14263,14264,14260,14266,14543,14262,14258,14548,14547,14552,14243,14252,14240,14239,14241,14231,14225,14232,14229,14222,14220,14226,14218,14242,14244,14248,14245,14246,14219,14217,14212,14210,14207,14211,14215,14206,14247,14249,14253,14254,14255,14256,14259,14261,14257,14251,14250,14281,14283,14284,14285,14286,14287,14289,14293,14288,14290,14291,14292,14294,14297,14398,14298,14302,14295,14296,14299,14300,14303,14307,14308,14313,14315,14394,14395,14393,14391,14399,14400,14401,14412,14396,14425,14392,14389,14386,14390,14426,14388,14430,14384,14374,14383,14381,14376,14380,14379,14382,14319,14385,14387,14316,14320,14331,14317,14318,14312,14309,14311,14306,14310,14314,14321,14326,14322,14323,14324,14327,14329,14328,14332,14335,14333,14330,14337,14343,14378,14377,14352,14348,14357,14367,14355,14350,14356,14353,14360,14361,14366,14368,14371,14369,14364,14375,14441,14657,14658,14656,14655,14445,14443,14447,14446,14457,14456,14650,14649,14653,14647,14637,14646,14651,14644,14639,14648,14685,14652,14675,14679,14673,14674,14669,14667,14654,14668,14664,14663,14661,14659,14660,14662,14363,3729,14362,14359,14365,3718,14358,14354,3716,3715,3713,3710,3714,14351,14349,14346,14344,14345,14340,14339,14338,14336,14342,14341,14334,14347,3709,3704,3711,3712,3706,3717,3719,3721,3720,3723,3728,3725,3724,3747,14666,3746,3741,14671,14670,14672,3743,14676,14678,14677,14680,14682,14681,14691,14695,14690,14687,14693,14692,14697,14698,14700,14707,14699,14694,14701,14709,3791,3787,3792,14714,3789,3790,3784,3779,3785,3771,3770,3766,3761,3767,3788,3763,3760,14689,14688,3758,3757,3752,3748,3753,14683,3750,3744,3751,3742,3739,3737,3738,3735,3652,3734,3657,3733,3732,3687,3731,3730,3693,3689,3698,3696,3708,3707,3727,3726,3722,3701,3697,3702,3699,3705,3695,3700,3703,3694,3692,3690,3685,3691,3688,3684,3681,3683,3682,3680,3677,3675,3672,3676,3673,3679,3678,3686,3674,3667,3664,3668,3660,3661,3485,3486,3481,3451,3479,3480,3476,3453,3474,3475,3463,3461,3471,3470,3473,3467,3469,3462,3460,3459,3457,3458,3454,3455,3444,3440,3448,3452,3447,3450,3441,3442,3439,3438,3437,3236,3436,3435,3228,3232,3222,3223,3218,3219,3211,3210,3209,3220,3203,3204,3200,3197,3199,3192,3201,3202,3205,3207,3206,3208,3212,3214,3213,3216,3217,3446,3666,3445,3443,3449,3665,3233,3669,3670,3671,3215,3190,3185,3191,3188,3189,3183,3182,3180,3181,3178,3176,3177,3179,3175,3174,3170,3169,3168,3172,3165,3164,3167,3161,3160,3163,3155,3154,3158,3152,3151,3156,3147,3146,3149,3144,3142,3143,3139,3137,3138,3136,3134,3133,3132,3131,3130,3078,3077,3085,3071,3070,3073,3068,3065,3069,3072,3066,3064,3067,3074,3046,3047,3043,3044,3039,3040,3036,3032,3035,3041,3030,3029,3027,3028,3025,3023,3024,3016,3014,3015,3009,3007,3008,3004,3002,3003,3001,2999,2858,2857,2854,2869,2996,2868,2871,2863,2861,2853,2851,2849,2848,2850,2852,2855,2856,2859,2860,2862,2866,2870,2867,2872,2864,2865,2874,2877,2875,2885,2886,2887,2891,2888,2889,2890,2892,2894,2896,2898,2908,2915,2907,2910,2904,2901,2905,2906,2909,2912,28718,2911,2913,2916,28712,2917,2918,2921,28708,2922,3944,3952,3943,3941,3942,3937,3933,3939,3938,3940,3935,3930,3936,3891,3946,3948,3945,3947,3949,3951,3950,4022,4027,4020,4021,4017,4016,4008,3954,4007,4010,4005,4006,4003,3959,4001,4002,3998,3997,3969,3970,3964,3965,3960,3966,3967,3968,3971,3973,3972,3974,3980,3975,3978,3979,3857,3853,3858,14779,3851,3848,3852,3849,3860,3861,14777,14776,14778,14775,14774,13764,13763,13761,13762,13758,13760,13755,13754,13751,13752,13748,13747,13733,13728,13732,13735,13729,13730,13726,13725,13580,13571,13579,13719,13577,13705,13573,13572,13570,13565,13564,13557,13553,13561,13567,13560,13555,13563,13731,13562,13558,13737,13739,13736,13734,13741,13743,13742,13745,13738,13750,13746,14789,14791,14787,14785,14788,14790,14786,12994,14784,14782,14781,14783,14780,13757,13756,12993,12990,12996,12995,12998,14796,14794,14797,14799,13002,13004,13013,13252,14801,13251,13249,13247,13245,13246,14802,14800,13257,13261,13559,13740,14798,14795,14793,13744,14792,12989,12986,13749,13556,13554,13265,13549,13548,13269,13267,13270,13539,13268,13271,13273,13538,13275,13288,13274,13531,13287,13290,13281,13530,13289,13534,13285,13286,13293,13535,13298,13542,13296,13291,13297,13301,13294,13292,13295,13299,13192,13193,13189,13186,13188,13300,13183,13182,13179,13180,13177,13178,13174,13106,13173,13170,13175,13176,13184,13315,13185,13304,13312,13303,13306,13305,13310,13311,13313,13317,13403,13322,13328,13320,13314,13321,13318,13325,13327,13326,13323,13329,13330,13332,13337,13331,13333,13335,13386,13341,13387,13385,13390,13383,13388,13389,13428,13433,13427,13426,13431,13423,13395,13422,13429,13419,13399,13418,13416,13430,13435,13439,13434,13432,13437,13440,13436,13438,13451,13450,13469,13468,13472,13475,13473,13474,13476,13477,13478,13479,13481,13484,13482,13487,13490,13488,13485,13493,13494,13495,13498,13496,13604,13603,13607,13610,13608,13489,13611,13614,13612,13483,13613,13620,13625,13619,13628,13630,13839,13631,13838,13841,13842,13843,13846,13851,13852,13854,13937,13855,13858,13934,13936,13940,13944,13941,14075,13938,14073,14076,14078,14077,14079,14074,14070,14072,14068,14063,14069,14061,14062,14065,14060,14059,14058,13954,13953,13948,13946,13949,13840,13950,13951,13832,13830,13833,13639,13837,13636,13643,13642,13663,13662,13659,13665,13666,13826,13825,13829,13952,13955,13957,13956,13960,13961,13966,13968,13969,14052,14051,14055,14056,14057,14064,14066,14067,14071,14080,14083,14084,14086,14087,14088,14090,14093,14094,14091,14097,14109,14098,14095,14101,14107,14102,14099,14105,14096,14100,14103,14033,14029,14038,14050,14037,14031,14040,14039,14048,14049,14092,14046,14047,14042,14041,14036,14043,14044,14045,13980,13975,13982,13981,13984,13977,13976,13974,13971,13970,13967,13963,13962,13827,13828,13822,13819,13824,13823,13959,13958,13821,13820,13671,13668,13674,13673,13809,13808,13817,13818,13814,13811,13835,13834,13836,13986,13812,13813,13810,13805,13804,13803,13798,13797,13676,13675,13679,13672,13677,13670,13669,13661,13660,13655,13652,13656,13667,13654,13648,13658,13664,13644,13649,13646,13647,13641,13606,13640,13638,13616,13637,13635,13629,13627,13615,13609,13602,13599,13595,13605,13594,13586,13585,13503,13584,13505,13506,13501,13502,13497,13499,13504,13508,13511,13507,13500,13509,13510,13512,13513,13518,13566,13517,13521,13514,13519,13524,13520,13522,13464,13463,13461,13454,13462,13523,13456,13453,13543,13541,13545,13425,13544,13307,13309,13410,13308,13409,13412,13407,13404,13408,13411,13405,13406,13402,13397,13396,13400,13394,13334,13393,13398,13391,13392,13401,13413,13414,13415,13442,13443,13441,13445,13448,13444,13457,13449,13446,13455,13459,13460,13465,13470,13467,13492,13466,13458,13491,13486,13471,13447,13452,13420,13417,13421,13424,13540,13533,13532,13528,13526,13527,13529,13525,13537,13536,13547,13550,13546,13551,13568,13569,13574,13575,13576,13578,13515,13516,13582,13581,13587,13588,13590,13593,13591,13596,13589,13600,13598,13597,13651,13601,13690,13693,13689,13691,13688,13687,13686,13657,13653,13650,13682,13692,13707,13706,13711,13712,13713,13714,13716,13720,13715,13718,13722,13717,13710,13708,13709,13704,13702,13703,13700,13699,13583,13697,13698,13694,13592,13695,13696,13701,13723,13724,13753,13759,13765,13766,13769,13770,13772,13776,13771,13767,13773,13774,13777,13778,13780,13784,13779,13786,13775,13685,13791,13681,13680,13683,13684,13794,13790,13796,13800,24216,24215,24217,13788,24221,24223,24222,24219,24226,24227,24225,24220,24228,14739,24229,14742,14746,14737,14735,14736,14738,14733,14731,14728,14727,14620,14618,14619,14616,14613,14617,14615,14610,24214,24213,14608,24211,24208,24212,24218,24209,24210,24207,24206,13995,13994,13815,13816,13806,13802,13807,13799,13801,13997,13989,13988,13992,13996,13993,13998,13990,13991,13985,13983,13979,13978,13831,14002,13987,14035,14034,14032,14004,14005,14015,14003,14001,14007,14006,14011,14016,14014,14018,14023,14020,14025,14019,14022,14582,14021,14586,14017,14009,14010,14008,14587,14590,14592,14591,14595,14588,14589,14585,14584,14583,14538,14539,14534,14531,14581,14579,14577,14578,14580,14576,14574,14575,14565,14561,14572,14569,14570,14571,14568,14573,14028,14027,14030,14104,14532,14529,14525,14530,14527,14522,14521,14502,14501,14497,14500,14491,14489,14490,14487,14493,14499,14492,14488,14486,14483,14481,14405,14406,14403,14402,14407,14408,14480,14479,14476,14482,14484,14485,14478,14477,14475,14470,14469,14420,14419,14418,14414,14413,14416,14411,14404,14417,14421,14415,14424,14428,14427,14431,14435,14432,14429,14437,14444,14438,14448,14449,14450,14451,14453,14454,14459,14458,14628,14632,14627,14625,14630,14634,14629,14623,14631,14641,14626,14622,14621,14612,14604,14611,14606,14614,14603,14605,14602,14609,14599,14601,14596,14607,14593,14594,14520,14524,14517,14516,14515,14518,14509,14508,14503,14504,14498,14496,14505,14506,14473,14507,14510,14466,14462,14465,14468,14460,14455,14461,14463,14434,14433,14423,14422,14464,14436,14467,14471,14474,14472,14511,14512,14624,14514,14513,14495,14494,14519,14523,14526,14600,14722,14640,14643,14642,14645,14705,14721,14720,14711,14724,14726,14723,14717,14729,14732,14730,14719,14734,14740,14745,14741,14743,14718,14715,3795,3799,3798,14744,14748,14747,14750,14752,3834,3801,3833,3832,3805,3803,3810,3830,3809,3804,3814,3813,3822,3826,3829,14760,14762,14759,14754,14761,13793,14763,14766,14768,14767,14770,14765,14764,3854,3850,3825,3824,3823,3827,3821,3818,3820,3816,3815,3817,3812,3811,3808,3650,3807,3777,3806,3769,3778,3774,3783,3782,3797,3796,3802,3800,3794,3793,3786,3781,3780,3776,3775,3773,3772,3768,3765,3658,3764,3762,3662,3659,3663,3638,3756,3759,3755,3641,3754,3749,3645,3642,3646,3493,3649,3745,3492,3489,3484,3653,3654,3482,3483,3478,3487,3488,3490,3472,3494,3491,3496,3468,3464,3465,3466,3421,3417,3422,3456,3419,3414,3420,3424,3412,3410,3427,3428,3429,3403,3433,3434,3258,3259,3249,3248,3227,3224,3226,3225,3221,3198,3195,3196,3193,3194,3187,3166,3186,3184,3171,3230,3162,3229,3231,3234,3159,3239,3235,3237,3157,3153,3238,3241,3244,3242,3150,3257,3256,3260,3261,3265,3268,3266,3263,3275,3274,3339,3348,3338,3322,3341,3343,3340,3335,3344,3346,3345,3349,3351,3350,3352,3347,3354,3353,3358,3359,3363,3364,3376,3386,3375,3372,3379,3385,3381,3380,3387,3389,3391,3392,3432,3431,3513,3512,3515,3516,3521,3522,3524,3523,3537,3615,3536,3530,3539,3538,3598,3611,3597,3607,3593,3592,3541,3542,3540,3550,3589,3551,3548,3572,3582,3571,3568,3574,3576,3573,3570,3577,3578,3585,3586,3587,3877,3575,3569,3925,3923,3924,3566,3922,3916,3920,3914,3921,3918,3917,3912,3909,3911,3907,3904,3906,3908,3903,3297,3902,3899,2965,3898,3895,2967,2963,2969,3897,2970,2955,3927,3926,3928,3931,2952,2948,2947,2946,2942,3932,3957,3958,3956,3953,28762,28760,28765,28767,28764,28761,28763,28766,28758,28752,28759,28756,28754,28755,28757,28753,28749,28747,3963,2926,3962,3961,2929,2925,2933,2943,2932,2927,2935,2939,2934,2930,2937,2938,2945,2944,2949,2951,2953,2954,2960,2961,2971,2973,2972,2968,2976,2975,3113,3112,3291,3290,3294,3296,3295,3302,3292,3293,3289,3300,3301,3303,3304,3306,3307,3564,3563,3560,3558,3559,3555,3562,3565,3561,3554,3567,3552,3553,3549,3545,3546,3543,3544,3533,3373,3532,3535,3529,3371,3528,3527,3382,3378,3384,3383,3520,3519,3525,3526,3370,3377,3368,3365,3367,3361,3360,3357,3362,3342,3337,3336,3334,3326,3328,3321,3319,3320,3323,3317,3315,3318,3277,3276,3279,3273,3271,3272,3269,3278,3280,3283,3281,3311,3310,3313,3308,3299,3309,3305,3312,3314,3557,3556,3329,3325,3331,3330,3547,3327,3333,3374,3332,3324,3298,3288,3284,3282,3285,3119,3286,3287,3117,3116,3114,3109,3115,3111,3118,3122,3126,3270,3125,3262,3267,3128,3120,3121,3123,3106,3103,3105,3099,3108,3107,3100,3098,2986,2982,2985,3110,2978,2977,2979,2974,2957,2958,2981,2940,2941,2936,2931,2902,2900,2903,2882,2881,2880,2992,2878,2873,2879,2876,2993,2991,2990,2988,2995,2994,2997,2998,3000,3005,2989,2987,3011,2983,2980,2984,3018,3017,3093,3091,3088,3092,3090,3096,3097,3094,3089,3095,3084,3101,3104,3102,3124,3129,3081,3083,3075,3082,3079,3076,3080,3042,3086,3087,3026,3031,3022,3019,3021,3020,3013,3010,3012,3006,2884,2883,2899,2928,2893,2897,2924,2895,2923,2920,2919,3127,3135,3141,3140,3145,3316,3366,3369,3356,3531,3534,3915,3913,2964,2966,2959,2962,2950,2956,28743,28748,28746,28751,28750,28745,28742,28740,28741,28739,28744,28736,28731,28737,28738,2914,28720,28730,28725,28734,28735,28732,28733,28728,28727,28726,28723,28724,28721,28711,28722,28719,28714,28717,28713,28715,28709,28710,28716,28729,3934,3900,3929,3901,3893,3892,3889,3894,3885,3890,3896,3886,3884,3910,3876,3883,3880,3875,3882,3881,3888,3887,3878,4033,4036,4031,4032,4034,4029,4028,4030,4025,4026,4023,4018,4024,4040,4013,4011,4000,3995,3999,3994,3993,3991,3977,3990,3988,3985,3989,3992,3986,3987,4059,3984,3872,3982,3981,3983,3976,3871,3863,3859,3862,3864,3856,3843,3855,3840,3844,3841,3847,3838,3842,3845,3831,3828,3835,3836,3640,3837,3608,3839,3865,3866,3867,3605,3604,3602,3603,3600,3596,3601,3599,3595,3606,3609,3612,3613,3637,3636,3643,3633,3617,3632,3648,3627,3623,3629,3651,3628,3624,3631,3656,3630,3626,3634,3635,3510,3511,3505,3502,3504,3639,3498,3499,3503,3500,3508,3509,3506,3514,3625,3518,3517,3621,3622,3620,3618,3619,3616,3614,3507,3501,3425,3426,3423,3418,3415,3416,3411,3413,3406,3405,3404,3409,3401,3398,3400,3402,3397,3393,3396,3255,3395,3394,3253,3252,3250,3251,3245,3246,3254,3243,3240,3264,3247,3355,3390,3399,3388,3407,3408,3430,3497,3655,3647,3644,3819,3610,3594,3590,3591,3588,3584,3583,3581,3579,3580,3868,3874,3870,3869,3873,3846,3879,4047,4054,4045,4046,4043,4049,4052,4048,4050,4044,4041,4038,4035,4039,4037,4051,4053,4056,4057,4055,4058,4042,3996,4012,4004,4015,4009,4014,4019,3905,3919,3148,3495,3477,14771,14773,14772,14769,13768,13781,13783,13782,13785,13789,13792,14757,14755,14758,13787,14751,14756,24224,14753,14749,14716,14713,14712,14710,14725,14708,14706,14703,14702,14704,14696,14684,14686,14638,14635,14633,14636,14452,14442,14439,14440,14373,14598,14597,14012,14013,13999,14000,14024,14026,13795,13727,13721,13678,13645,13965,13964,13973,13972,14054,14053,14110,13632,13947,13943,13942,13945,14082,13302,13169,13109,13114,13168,13104,13101,13097,13096,13094,13085,13095,13181,13083,13079,13084,13187,13081,13190,13191,13080,13076,13196,13197,13284,13552,3955,2847,3033,3034,3037,3038,3048,3050,3049,3045,3052,3053,3059,3061,3060,3058,3062,3063,3057,3056,3054,3055,3051,3173,3736,3740,14665,14370,14372,14325,14305,14304,14301,14397,14126,13350,13344,13363,13349,13362,13364,13356,13360,13880,13894,13901,13900,14141,24198,13119,13074,12887,13065,13060,13059,13203,13202,13207,13008,12988,13010,12981,12978,12955,12956,12953,12951,12952,12954,12949,12947,12945,12948,12950,12946,12944,12943,13019,12840,12923,16269,16267,16272,16270,16257,16278,16280,16277,16274,16279,16282,16041,16042,16037,16039,16031,16032,16029,16028,16026,16021,16027,16030,16023,16024,16019,16018,15960,15959,15949,15948,15947,15942,15951,15954,15952,15953,15751,15750,15723,15724,15720,15721,15717,15718,15714,15716,15710,15707,15709,15712,15705,15701,15706,15711,15703,15704,15693,15694,15689,15687,15690,15695,15356,15346,15355,15358,15349,15350,15344,15343,15342,15347,15337,15336,15335,15340,15333,15332,15329,15338,15339,15341,15345,16372,16373,16371,16370,16374,16367,16365,16366,16357,16356,16359,16353,16352,15965,15713,15964,15968,15962,15715,15961,15958,15966,15969,15967,15963,15972,15981,15976,15983,15975,15970,15978,15986,15977,15973,15980,16003,15979,16016,15974,16015,16020,16010,16008,16012,16017,16011,16014,16009,16006,16002,16007,16004,16013,16312,16313,16310,16311,16298,16291,16297,16293,16300,16308,16299,16295,16302,16306,16301,16296,16303,16211,16304,12083,16305,16307,16316,12078,12075,12079,12077,12070,12080,12081,12084,12085,12087,12086,12096,16208,12095,12093,16204,16203,16205,16206,16207,16212,16213,16214,16210,16225,16224,16218,16227,16226,16287,16294,16286,16232,16284,16281,16285,16292,16283,16256,16254,16233,16253,16250,16255,16259,16258,16249,16261,16260,16264,16268,16266,16265,22951,22949,22955,22960,22956,22952,22958,22962,22957,22954,22959,22964,16062,16050,16061,16064,16057,16058,16051,16049,16052,16053,15940,15944,15937,15934,15936,15938,15933,15743,15932,15931,15741,15742,15738,15739,15734,15735,15731,15730,15728,15727,15729,15732,15685,15677,15684,15697,15679,15676,15681,15688,15680,15674,15682,15683,15369,15370,15363,15364,15359,15361,15354,15353,15351,15352,15348,15360,15362,15365,15366,15368,15373,15367,15379,15374,15377,15371,15372,15375,15376,15378,15383,15387,15385,15391,15384,15386,15389,15390,15392,15395,15399,15396,15401,15393,15394,15402,15404,15403,15410,15647,15411,15416,15408,15405,15409,15412,15415,15413,15414,15417,15418,15419,15421,15424,15422,15426,15427,15429,15428,15432,15436,15433,15438,15430,15431,15439,15440,15441,15444,15445,15447,15446,15452,15458,15453,15456,15450,15451,15454,15457,15455,15463,15465,15464,15462,15470,15471,15476,15477,15479,15596,15478,15485,15474,15475,15480,15472,15473,15469,15468,15481,15483,15482,15488,15489,15494,15495,15497,15582,15496,15503,15492,15493,15498,15490,15491,15487,15486,15499,15500,15501,15504,15507,15505,15508,15511,15509,15517,15519,15518,15516,15524,15543,15525,15530,15522,15523,15526,15527,15529,15532,15535,15531,15533,15528,4088,15534,15536,4097,4090,4098,4108,4095,4092,4096,4101,4093,4094,4091,4086,4084,4081,4085,4089,4082,4080,4083,4087,4078,4079,4077,4099,4100,4102,4105,4109,4106,4113,4103,4104,4111,4112,4114,4117,4118,4121,4132,4122,4119,4140,4139,15562,15561,22937,22935,22938,12428,22943,22944,22946,12430,22948,12487,22947,22945,15868,15867,15866,15860,15865,15864,15857,15856,15854,15849,15855,15852,15853,22941,15846,15845,15843,15839,15844,15842,15838,15847,15848,15850,15836,15833,15835,15830,15831,15828,15816,15827,15822,15818,15824,15829,15823,15826,15820,15817,15819,15821,15808,15807,15804,15803,15801,15802,15798,15796,15799,15627,15800,15618,15617,15615,15610,15616,15613,15621,15622,15619,15625,15626,15628,15638,15787,15637,15640,15631,15632,15629,15635,15636,15639,15778,15780,15777,15776,15643,15644,15651,15641,15634,15642,15645,15435,15425,15434,15437,15420,15423,15646,15649,15650,15648,15652,15653,15655,15654,15658,15659,15753,15752,15756,15757,15759,15758,15772,15775,15771,15768,15774,15773,15781,15784,15779,15790,15785,15788,15782,15783,15786,15789,15793,15795,15797,15794,15814,15813,15898,15899,15897,15896,15825,15894,15900,15902,15905,15903,15907,15895,15892,15893,15888,15887,15885,15881,15886,15879,15878,15877,15873,15872,15851,15834,15875,15870,15869,15858,15859,15861,15863,15862,12491,12486,12494,12493,12489,12496,12495,15871,15874,15876,12499,15882,15883,15884,15889,12509,12510,12503,12500,12504,12507,12501,12502,12498,12481,12497,12492,12490,12480,12475,12478,12471,12470,12468,12469,12465,12464,12439,12438,12437,12442,12433,12431,12434,12435,12416,12417,12411,12410,12408,12409,12404,12402,12403,12400,12405,4153,4151,4148,4145,12415,12418,12414,12420,12419,12424,12423,12426,12421,12422,4141,12425,12427,4137,4142,4131,4138,4143,4135,4128,4136,4144,4133,4130,4134,4147,4146,4149,4152,4150,4154,4157,4155,4158,4159,4161,4164,4162,4166,12399,4171,4170,12372,12371,12395,12398,12394,12392,12401,12406,12407,12412,12396,12391,12397,12446,12393,12390,12453,12452,12541,12540,12544,12548,12545,12385,12546,12549,12377,12375,12376,12382,12373,12369,12374,12378,12381,12389,12366,12368,12363,12370,12359,4173,12358,12361,12357,12355,4179,4180,4175,4174,4172,4169,4167,4168,4165,4163,4160,4176,4177,4178,4181,4182,4184,4188,4183,12350,12348,12340,12349,12347,12353,12354,12356,12360,12364,12365,12367,12380,12362,12384,12383,12386,12387,12559,12558,12648,12647,12656,12657,12658,12388,12673,12675,12671,12663,12672,12674,12670,12669,12304,12299,12302,12303,12298,12295,12297,12292,12289,12291,12293,12284,4298,12282,12277,12283,12280,12276,12281,12278,12287,12290,12288,12667,12285,12279,12286,12245,12662,12665,12661,12659,12241,12242,12238,12239,12235,12236,12230,12233,12225,12222,12226,12228,12223,12220,12224,12227,12207,12204,12206,12202,12200,12208,12209,12213,12219,12214,12211,12217,12221,12218,12244,12215,12212,12216,12247,4596,4595,4594,4590,12252,12253,12251,12255,12257,12254,12258,12249,12246,12248,12250,12243,12240,12259,12261,12274,12260,12268,12271,12266,12267,12262,12263,12256,4582,12264,12265,4569,4323,4568,4562,4578,4577,4580,4566,4563,4567,4564,4573,4572,4579,4584,4583,4589,4592,4593,4597,4591,4586,4601,4603,4600,4602,4608,4604,4614,4622,4615,4612,4620,4624,4621,4626,4618,4613,4619,4616,4628,4629,4630,4632,4631,4617,4633,4635,4634,4641,4647,4642,4639,4645,4646,4649,4655,4650,4778,4779,4777,4775,4772,4776,4781,4774,4771,4773,4770,4812,4815,4811,4806,4813,4814,4817,4818,12149,12148,12154,12155,12157,12163,12156,12151,12159,12158,12165,12166,12171,12173,12175,12176,16163,16162,16168,16172,16170,16169,16187,16186,16235,16234,16238,16239,16241,16244,16242,16237,16245,16247,16246,16251,16252,16263,16262,22968,23048,22967,22961,22975,23057,22973,22963,22974,22984,22971,22965,22972,22980,22969,22966,22970,22976,16067,16063,16068,16073,16065,16066,16059,16060,16056,15929,16055,16054,15927,15924,15928,15935,15925,15922,15926,15765,15764,15767,15761,15760,15744,15740,15745,15733,15748,15754,15749,15672,15671,15662,15663,15660,15666,15667,15664,15669,15675,15668,15670,15665,15388,15381,15382,15673,15661,15398,15397,15400,15657,15656,15678,15755,15763,15762,15766,15769,15770,15917,15921,15916,15914,15919,15923,15918,15920,15912,15910,15913,15911,15906,15904,15792,15791,15901,15915,15908,15909,16071,16074,16070,16069,16072,16075,16079,16083,22978,22977,22981,22987,22982,22979,22985,22991,22986,22983,23052,23051,23056,23058,23054,23050,23053,23055,23046,23043,23045,23047,23035,23033,23034,23037,23030,23031,23028,23029,23025,16161,23024,23023,16156,16155,16151,16150,16148,16149,16145,16144,16143,12183,16142,16140,12189,12185,12193,12619,12191,12192,12186,12187,12184,12180,4810,12179,12182,12177,12168,12178,12174,12181,12170,16153,16152,16154,12172,16157,16158,16159,16160,16164,16165,16167,16166,16171,23032,16240,16243,16248,12169,12167,12164,12160,12161,4807,12162,4809,4802,4803,4797,4795,4799,4804,4798,4796,4801,4800,12188,12190,12194,12195,12198,12201,12199,12203,12205,12615,12618,12614,12613,12610,12611,12605,12608,12601,12600,12589,12586,12588,12591,12582,12581,12579,12580,12577,12542,12576,12575,12553,12554,12547,12543,12565,12551,12552,12557,12564,12561,12560,12566,12569,12567,12571,12574,12572,12570,12578,12583,12584,12585,12587,12609,12624,12622,12625,12628,12629,12630,12573,12631,12632,12633,12568,12636,12639,12637,12563,12641,12642,12645,12649,12643,12644,12640,12638,12634,12626,12635,12234,12627,12621,12229,12620,12623,12617,12231,12616,12612,12232,12646,12237,12650,12652,12653,12655,12654,12660,12664,12666,12651,12562,12555,12550,12556,12379,12537,12536,12534,12535,12533,12450,12532,12531,12461,12460,12449,12447,12448,12444,12451,12456,12454,12455,12457,12463,12458,12519,12518,12520,12521,12523,12527,12522,12525,12517,12511,12515,12506,12476,12505,12508,12513,12514,12516,16095,16099,16094,16097,16092,16091,16096,16086,16081,16088,16087,16093,16098,22996,22999,22995,16090,22994,22992,23002,23003,23005,22997,23009,23008,23016,23021,23017,23019,23014,23013,23011,23010,23007,23012,16137,16136,16126,16121,16127,16131,16124,16119,16125,16128,12604,12603,12602,12597,12606,12607,12599,16110,16112,16108,16105,16109,16111,16106,16107,16103,16104,16100,12524,12529,12530,12594,12593,12598,12592,12538,12539,12528,12526,12590,12596,16102,16113,16116,16114,16120,16123,16122,23001,23000,23004,23006,16117,16115,16101,16118,12595,16132,16129,16130,16133,16135,16134,16138,16141,16139,16146,16147,23015,23018,23020,23027,23026,23038,23039,23040,23042,23041,23044,23049,23036,23022,22998,22993,22990,22989,22988,16089,16085,16082,16084,16080,16078,16077,16076,15891,12512,12477,12472,12474,12466,12462,12467,12459,12473,12441,12440,12436,12413,12445,12196,4625,12197,4627,4623,4794,4793,4792,4787,4638,4786,4789,4783,4780,4785,4784,4788,4791,4790,4805,4808,4782,4648,4637,15930,16236,16231,16230,16228,16221,16229,16189,16188,16183,16191,16190,16194,16185,16184,16182,16179,16180,16176,16175,12146,12141,12147,12144,12152,12153,16174,16173,12150,12145,12142,12143,12138,12137,4834,4833,4822,4819,4821,4816,4820,4823,4765,4766,4762,4761,4759,4760,4756,4690,4755,4750,4758,4757,4753,4826,4829,4825,4824,4828,4827,4831,4832,4846,4847,4848,12135,12134,12139,12129,12128,12131,12126,12124,12125,12122,12127,12120,12130,12132,16181,12133,12118,16193,16192,16196,16197,16200,16199,16202,16201,16198,12104,12099,12103,12106,12100,12101,12098,12092,12091,4891,4890,4888,4883,4879,4882,4880,4885,12102,4884,12108,4881,4864,4862,12115,12116,12114,12112,12119,12113,12110,12117,12111,12107,12105,12109,4854,12121,12123,4840,4841,4837,4835,4838,4842,4830,4836,4839,4752,4751,4748,4745,4844,4845,4843,4851,4852,4853,4855,4849,4850,4741,4856,4858,4857,4871,4876,4870,4874,4868,4867,4872,4863,4860,4865,4869,4866,4733,4907,4909,4906,4912,4903,4902,4908,4899,4895,4904,4916,4905,4897,4925,4924,4927,4928,4931,4932,4935,4942,4936,4933,4943,4946,4944,12082,4938,4934,12089,12088,12090,4901,4900,4893,4892,4889,4896,4898,4894,4887,4875,4886,4878,12041,4945,4948,12035,4949,4952,4947,4941,4950,4953,4951,4955,4939,4940,4937,4930,4959,4958,4961,4962,4965,4966,4969,4970,4975,11935,4977,4976,11929,11932,11927,11925,11928,11930,11926,11924,11922,11923,11921,11919,4990,4989,4980,4979,4978,4982,4974,4973,4971,4972,4968,4967,4922,4923,4919,4920,4915,4917,4918,4913,4914,4984,4988,4983,4981,4986,4995,4987,4985,11915,11914,11916,11917,11912,11909,11911,11906,11913,5049,5048,5039,5038,5036,5037,5031,5030,5004,4994,5003,5006,4996,4991,4998,5001,4997,4993,4999,5000,5002,5009,5015,5010,5007,5013,5017,5014,5019,5011,5012,5008,4726,4720,4727,4911,4723,4718,4725,4729,4724,4721,4728,4732,4730,4731,4736,4859,4722,4734,4735,4740,4743,4742,4737,4738,4701,4739,4695,4744,4746,4747,4694,4749,4691,4687,4692,4689,4683,4676,4684,4681,4674,4682,4685,4491,4487,4492,4489,4495,4496,4499,4493,4490,4494,4497,4433,4432,4423,4420,4422,4418,4498,4500,4501,4503,4507,4502,4504,4662,4669,4661,4658,4664,4671,4663,4660,4667,4678,4668,4665,4768,4767,4764,4763,4659,4666,4656,4653,4654,4651,4644,4652,4657,4511,4508,4512,4509,4515,4516,4519,4513,4510,4514,4517,4400,4399,4397,4398,4391,4392,4385,4384,4382,4383,4378,4377,4380,4375,4373,4374,4368,4366,4363,4367,4365,4371,4372,4369,4376,4526,4381,4386,4379,4387,4522,4525,4521,4524,4518,4529,4523,4636,4520,4528,4533,4535,4547,4534,4537,4531,4530,4532,4527,4370,4359,4360,4355,4352,4356,4364,4353,4350,4354,4361,4362,4349,4342,4341,4338,4344,4351,4343,4340,4347,4348,4358,4357,4539,4541,4538,4536,4542,4543,4544,4540,4551,4552,4555,4576,4556,4611,4546,4610,4607,4605,4585,4606,4581,4575,4574,4571,4570,4565,4550,4548,4549,4553,4545,4554,4561,4560,4557,4558,4327,4559,4324,4322,4317,4314,4316,4313,4320,4321,4318,4311,4319,4309,12269,12270,12272,4310,12275,4307,4301,4308,4297,4293,4290,4294,4291,4288,4292,4299,4302,4300,4303,4306,4304,4312,4305,4287,4280,4279,4276,4282,4289,4281,4278,4285,4286,4296,4295,4283,12296,12300,12301,4284,12305,12306,12308,12307,12311,12312,12317,12676,12322,12321,12328,12327,12344,12343,12337,12346,12345,12341,12352,12351,12338,12342,12339,12336,12335,4192,4193,4189,4185,4186,4187,4190,4191,4194,4197,4195,4198,4201,4199,4202,4205,4203,4206,4207,4208,4209,4211,4210,4215,4214,4245,4244,12319,12318,12320,12324,12326,12323,12325,12329,4204,12331,12330,12333,12332,12334,4196,4200,12316,12315,4251,4250,4253,4248,4242,4249,4252,4246,4243,4247,4254,4255,4256,4258,4261,4259,4263,4266,4264,4267,4268,4270,4277,4269,4273,4274,4271,4265,4272,4262,12310,12309,12313,12314,4260,4257,4275,4241,4238,4236,4235,4237,4239,4234,4232,4224,4221,4223,4225,4220,4219,4218,4216,4213,4217,4222,4240,4212,4226,4227,4228,4230,4231,4229,4233,4315,4325,4326,4328,4330,4336,4329,4331,4332,4333,4335,4334,4337,4339,4345,4346,4388,4389,4390,4393,4396,4394,4401,4404,4402,4405,4408,4406,4413,4416,4414,4419,4421,4424,4425,4427,4426,4430,4431,4434,4435,4442,4441,4485,4484,4482,4452,4481,4480,4486,4488,4686,4483,4479,4699,4700,4698,4696,4702,4705,4703,4713,4712,4708,4710,4716,4709,4706,4714,4717,4715,4711,4719,5112,4707,4476,4474,4477,4475,4472,4464,4471,4466,4465,4463,4458,4457,4455,4450,4456,4460,4453,4454,4449,4446,4443,4445,4438,4451,4437,4436,4439,4429,4428,4440,4444,4447,4448,4478,4459,4461,4462,4467,4470,4468,5096,5098,5094,5095,5093,5092,5090,5089,5086,5084,5083,5082,5081,5080,5077,5076,5071,5079,5078,5085,5073,5087,5088,5091,5075,5099,5102,5097,5104,5103,5107,5100,5101,5111,5109,5113,4473,4469,5110,5108,5106,5105,5074,5066,5021,5065,5068,5062,5023,5061,5059,5064,5067,5063,5060,5069,5070,5058,5057,5055,5053,5054,5052,5056,5024,5022,5018,5026,5025,5029,5033,5032,5044,5046,5043,5040,5035,5041,5042,5045,5034,5047,5051,5027,5028,5016,5005,5020,5072,4704,4697,4693,4688,4415,4407,4411,4412,4417,4409,4506,4505,4410,4403,4395,4640,4643,4677,4680,4679,4675,4672,4670,4673,4910,4992,5050,11905,11904,11902,11901,11903,11907,11908,11910,11918,11920,11900,4921,4926,11931,11933,11937,11938,11939,11940,11944,11945,11946,11949,11950,11951,11952,11957,11961,11958,11959,11983,11982,11984,11986,11985,11988,11989,11991,11990,11993,11987,11981,11980,11978,11979,11967,11963,11972,11970,11976,11977,11974,12001,12003,11999,11994,12000,12005,11997,11998,12002,11995,11992,11996,12008,12009,12010,12012,12017,12013,16497,16495,16498,16499,16502,16501,16504,16505,16503,15251,15250,15253,15246,15245,15244,15247,15243,15248,15252,15256,15257,16469,16473,16467,16461,16468,16465,16487,16488,16485,16486,16479,16477,16466,16463,16458,16462,16456,16464,16413,16470,16471,16346,16472,12056,16475,16476,16478,16483,16484,16490,16480,16481,16482,12046,12045,12022,12023,12018,12015,12019,12016,12014,12011,12007,12020,12021,12024,12006,12025,12027,12032,12026,12043,12042,12051,12036,12040,12038,12049,12050,12052,12055,12059,12057,12065,12066,12062,12068,12072,16343,16345,16342,16336,16409,16407,16412,16404,16408,16405,16448,16447,16454,16455,16450,16446,16449,16452,16444,16442,16443,16440,16445,16451,15268,15269,15265,15264,15262,15263,15259,15258,15255,15261,15260,15254,15249,15266,15267,15270,15273,15271,15274,15275,15277,15276,15280,15281,16428,16432,16426,16425,16427,16429,16423,16422,16418,16416,16419,16389,16420,15321,16421,15315,15314,15296,15295,15292,16424,15287,15288,15285,15284,15282,15283,15279,15278,15286,15289,15290,15291,15293,15297,15294,15299,15302,15300,15301,15303,15305,15306,15308,15307,15309,15304,15310,15312,15316,15311,15313,15318,15320,15319,15324,15325,15327,16383,15326,15330,15322,15323,15328,15331,15334,16377,16379,16375,16381,16378,16380,16376,16384,16368,16369,16364,16362,16392,16395,16391,16393,16394,16396,16400,16410,16401,16406,16398,16399,16397,16363,16360,16361,16354,16355,16347,16344,16348,16349,15996,15982,15995,16341,15988,15984,15993,16335,15992,16327,15990,15985,15991,15994,15989,15987,15997,15999,15998,16001,16321,16005,16323,16319,16318,16317,16315,16314,16309,16320,12076,16322,16324,16326,12074,16329,16330,16331,16334,16332,16333,16338,16339,16340,16337,16351,16350,16403,16402,16328,12071,12073,12067,12069,12063,12060,12064,12058,12061,12048,12047,12044,12039,12037,16325,16000,16358,16415,16417,16414,16431,16430,16435,16437,16436,16433,16439,16438,16434,16441,16411,16390,16388,16385,16382,16386,16387,15317,15298,15272,16453,16459,16460,16457,12034,12031,12033,12030,12029,11973,12028,11975,11969,11968,4954,11971,11966,4956,11964,11965,11954,4960,11953,11947,11956,11962,11955,11960,11943,11948,11941,4964,4963,11942,11936,4957,12004,12053,12054,16489,16491,16493,16492,16496,16494,16500,16474,11934,4929,4873,4877,4861,4754,12094,12097,16195,16216,16215,16217,16220,16223,16222,16219,16209,12136,12140,4769,16178,16177,4609,12210,4599,4598,4587,4588,12273,12668,12294,4156,4129,4127,4124,4123,4116,4126,4125,4115,12429,12432,12443,12483,12482,12479,12485,12488,12484,15890,15880,15407,15406,15633,15630,15443,15442,15448,15449,15624,15623,15620,15459,15614,15461,15460,15467,15608,15466,15607,15600,15599,15602,15609,15601,15605,15598,15597,15595,15590,15589,15586,15592,15591,15594,15588,15587,15585,15593,15584,15583,15581,15578,15577,15570,15580,15579,15574,15840,15841,15576,15575,15573,15571,15567,15572,15569,15564,15563,15515,15514,15512,15513,15506,15502,15521,15510,15520,15560,15559,15551,15565,15566,15568,15553,22926,22928,22925,22927,22931,22939,22932,22934,22936,22933,22930,22929,15557,15556,15555,15554,15552,15548,15547,15545,15546,15540,15539,15538,15537,15542,15541,15550,15549,4110,4107,15558,4120,22940,15544,15832,15837,15604,15603,15606,15812,15811,15815,15809,15611,15612,15806,15810,15805,22942,15484,15380,15357,15686,15691,15692,15698,15700,15699,15696,15702,15708,15726,15725,15722,15736,15737,15746,15747,15939,15941,15945,15946,15943,15950,16025,16046,16047,16045,16043,16048,16034,16044,16040,22953,16036,16276,22950,16275,16038,16273,16271,16035,16033,16288,16289,16290,16022,15971,15956,15955,15957,15719,8771,8769,8774,8763,8762,8760,8761,8757,8756,8759,8753,8752,8755,8749,8748,8751,8743,8742,8746,8740,8739,8736,8744,8745,8747,8794,8863,8793,8849,8790,8750,8789,8785,8826,8824,8848,8847,8841,8851,8855,8850,8843,8853,8857,8852,8846,8854,8859,8861,8860,8873,8878,8872,8875,8864,8858,8865,8867,8862,8856,8870,8866,8868,8869,8871,8876,8741,8738,8737,8895,8735,8734,8732,8729,8733,8728,8727,8731,8730,8726,8991,8988,8993,8994,8995,8901,8997,8996,8907,8903,8914,8919,8912,8905,8913,8916,8908,8906,8909,8910,8887,8882,8888,8885,8886,8881,8880,8891,8883,8879,8884,8877,8889,8892,8890,8893,8894,8897,8896,8899,8902,8898,8900,8904,8874,8911,8915,8917,8920,8922,8921,8918,8926,8927,8928,8923,8933,8934,8992,8990,8989,8944,8943,8938,8946,8945,8959,8958,8987,8985,8978,8986,8982,8724,8980,8979,8977,8974,8976,8971,8960,8970,8964,8973,8972,8975,8984,8998,8983,9000,8981,8718,8714,8719,8716,8717,8711,8710,8708,8707,8712,8705,8706,8697,8696,8693,8699,8698,8703,8704,8701,9040,9038,9042,9045,9041,9039,9043,9044,9032,9031,9029,9030,9026,9024,9025,9022,9027,9028,9119,9120,9117,9034,9115,9113,9048,9046,9050,9053,9049,9047,9051,9056,9052,8700,9054,9055,9057,9060,9063,9061,9064,9059,8690,8691,8684,8679,8685,8688,8682,8681,8683,8686,8687,8689,8695,8694,8692,8680,8678,8673,8672,8670,8667,8671,8676,8668,8666,8669,8674,8677,8675,9069,9072,9068,9067,9070,9073,9071,9066,9094,9092,9096,9098,9097,9095,9099,9101,9100,9103,9104,9106,9105,9107,9108,9109,9110,9112,9114,9116,9118,9111,9102,9062,9058,9093,9091,9090,9088,9086,9087,9089,9084,9082,9074,9075,9076,9077,9080,9079,9078,9081,9083,9085,9065,8665,8664,8702,9021,9020,8967,8966,8969,8963,8962,8955,8965,8968,8954,8951,8957,8956,8953,8952,8950,8949,8947,8948,8942,8940,8937,8941,8939,8935,8936,8930,8929,8932,8925,8924,8931,9004,9003,9006,8999,9013,9005,9001,9008,9007,9011,9012,9009,9016,9017,9018,9014,9033,9035,9036,9037,8709,8713,9010,9023,9015,9019,9002,8715,8720,8723,8721,8722,8725,8961,8845,8844,8842,8839,8838,8837,8836,8834,8835,8833,8832,8831,8829,8818,8830,8827,8828,8820,8819,8817,8812,8811,8810,8814,8808,8806,8807,8809,8804,8802,8795,8792,8796,8797,8798,8800,8801,8803,8799,8805,8791,8778,8779,8775,8773,8772,8770,8768,8767,8766,8765,8764,8776,8782,8788,8783,8780,8786,8816,8787,8784,8823,8821,8822,8825,8781,8758,8815,8813,8777,8840,8754,43,46,42,41,40,39,37,38,34,33,32,35,27,26,29,23,22,25,19,18,21,15,14,17,11,10,12,6,5,7,4,3,2,1,0,13,8,9,244,252,242,241,246,239,238,233,247,248,249,228,224,229,226,227,217,215,221,220,216,222,223,225,230,231,232,234,235,236,219,237,240,245,243,270,273,271,279,272,274,282,285,286,288,287,291,290,294,297,295,299,292,293,298,206,205,204,203,301,199,198,187,186,189,182,180,181,179,178,176,175,177,171,168,170,172,165,164,162,161,160,159,153,152,149,155,163,154,157,151,150,148,147,146,145,144,112,111,138,107,106,109,103,101,102,86,85,88,78,77,80,73,72,75,68,67,70,65,63,64,62,61,59,58,60,54,66,71,83,81,84,52,91,93,92,89,94,98,97,104,114,105,120,96,31,95,44,36,45,47,30,28,268,24,269,265,260,266,263,267,264,262,16,261,257,256,254,253,255,251,250,258,259,126,127,122,121,119,124,117,118,113,110,100,115,116,108,129,132,128,131,123,125,130,211,210,212,213,209,139,141,134,135,137,136,140,143,142,158,156,193,194,192,190,195,197,196,202,207,208,200,201,318,322,317,315,311,316,320,313,310,312,314,308,283,306,307,296,289,284,281,304,305,309,302,303,300,280,277,278,275,276,321,319,191,188,184,183,169,185,167,166,133,214,218,99,20,90,87,82,79,51,49,50,48,53,55,57,56,76,74,69,174,173,4067,4068,4066,4065,4064,4063,4062,4061,4060,4069,4070,4072,4074,4073,4071,4075,4076,328,327,329,330,331,332,333,334,336,335,337,339,338,341,343,348,351,349,344,350,347,352,346,345,342,340,326,325,324,323)
   InputName=()
   OutputName=(StringValue="Cuffs_RenderSelection")
   RemoveIndices=()
   SelectionOverrideType=ReplaceAll
   SimTransferType=Use2DSimMesh
   Transfer=()
   TransferSelectionThreshold=0.95
  }
 }
 Selection_02 {
  type=FChaosClothAssetSelectionNode_v2
  guid=#62
  inputPins [
   Collection {type=FManagedArrayCollection}
   InputName.StringValue {type=FString}
   TransferCollection {type=FManagedArrayCollection}
  ]
  outputPins [
   Collection {type=FManagedArrayCollection}
   OutputName.StringValue {type=FString}
  ]
  properties {
   Group=(Name="SimVertices3D")
   Import=()
   ImportSecondary=()
   Indices=(10039,9987,9130,9129,8912,10967,10032,10033,9979,9986,9985,9982,9944,9943,9446,9444,9443,9445,9981,9141,9242,9978,9980,9983,9243,9984,10035,10024,10034,10036,10037,10078,10079,10125,9306,8911,9135,8872,9467,9466,8901,10821,8946,8874,9648,8947,9649,8902,9658,9028,10958,9661,9149,10840,9659,9605,9630,9398,9631,10797,9115,9116,9114,9292,9608,9633,9610,9634,9662,9397,9663,9431,9220,9664,9218,9682,9396,9660,9437,8910,8670,9281,9282,9827,9189,10884,10790,9825,9824,9262,8877,9261,9793,9763,9794,9764,9728,9762,8875,8876,9152,9730,10773,9701,9731,9729,9725,9727,8923,8924,8925,9700,9699,9427,9723,9428,9724,9726,9756,8746,9719,9759,9761,8747,9789,9082,9790,9791,9792,9451,9450,9821,9817,9819,9822,9823,9862,9863,9866,9449,9867,9865,10872,9905,14773,16392,15333,16393,15996,16197,16753,16752,16754,16823,15332,15383,16856,16397,16439,15573,15572,15574,14809,14808,14810,16792,16826,15382,15914,15444,15710,16794,15461,16779,16780,15632,15709,16798,16409,15633,16411,16410,16452,16884,16450,16449,16451,16480,16489,15398,15050,15561,15048,15562,15433,15432,15431,15437,15708,15305,15306,16440,16576,16575,16579,16751,14612,16578,16442,16441,14614,16573,16572,16535,16577,16612,16611,16580,16581,16538,14932,16539,16143,16142,15719,15377,15376,15378,16642,16643,16617,15776,15775,14931,14933,15334,15667,15335,15008,16495,16494,16496,15007,15006,15220,15885,14908,14906,16761,15218,15219,15749,16808,16697,16809,16730,14702,15134,14700,15325,16900,15315,15314,15313,16672,16671,15263,16879,16659,16675,16605,16676,16606,15671,15559,15118,15117,15119,14774,16728,14775,15483,15482,15484,15145,16726,15143,15144,16789,16817,16719,16720,16721,14732,16881,16718,16707,15105,16872,15106,15107,16905,15438,15439,15440,14807,14805,16865,16864,15059,16833,15060,14658,14657,14656,16708,15104,14907,15103,16378,15079,15078,14729,14730,14731,16654,16635,16613,15296,16634,16633,14601,15013,16653,16696,16694,14699,15196,14698,14697,14801,14806,14922,14921,14923,15187,15186,15411,15410,16690,16667,16692,16691,16706,15920,14733,14734,15467,16414,16705,15386,16713,15384,15385,16712,15947,15270,16703,16687,16702,15268,15269,16688,16686,16685,14976,15581,15580,15582,16665,16649,16866,16683,16670,16684,16695,16669,15921,15623,15622,15621,16666,16630,16668,16647,16645,16648,16644,16646,16650,16738,14901,16629,16604,16608,15605,15604,15603,14652,14597,14598,14599,15750,16802,15732,16886,15513,14765,15512,16806,16871,16610,16609,14602,16570,16534,16569,15730,15731,16533,16478,16845,16476,16475,15158,16477,15459,15159,16536,16537,16479,15615,15616,15617,14613,15460,16399,16398,16781,16394,15278,15157,16821,15368,15028,15280,14891,14890,14892,16460,15029,16257,15702,15703,16474,15782,15783,15784,16532,16793,16628,16566,16531,16565,16603,14902,16601,14956,14957,14958,16567,16750,16749,16530,15704,16413,16412,15234,16453,16420,16454,14856,16455,16418,16419,16396,14855,16438,15705,15706,16195,16196,14898,15171,15672,15674,16827,16788,14711,14710,14709,15454,16883,14970,16128,15697,15063,16375,15032,16349,14969,14971,15634,14727,14726,14728,15569,16401,16400,16402,16405,16444,16443,15174,15080,14761,15081,15173,15172,16481,16838,16585,16540,16583,16584,15494,15492,15493,16542,16614,14623,14943,15097,15096,15095,14944,14942,15262,15261,15480,14993,15481,14994,16607,16800,16656,16655,15560,15683,15682,15681,14776,15948,15946,15714,15715,16717,16724,16725,16722,16727,16723,16716,16714,15161,15160,14678,14676,14677,15691,16689,16682,16704,16681,16698,16711,16710,16709,16700,16699,16701,16680,16678,16679,16663,15491,15489,16765,15711,15713,16661,16660,16662,15077,15076,15075,16408,16407,14631,15696,15695,15644,15645,15328,16109,16110,16640,16294,15326,15327,16878,16810,15982,15983,15984,16592,15123,15321,16591,15324,15017,15016,15018,16876,16737,16888,15718,16348,16347,16346,15812,15813,14632,14630,16374,16373,14636,16403,16404,16406,14760,14759,16448,16487,16447,16486,16488,16482,16543,16545,15779,15608,15606,15607,15778,15777,16586,16618,16658,16657,16620,16619,14992,15781,15780,16582,16637,16636,16544,16715,16882,16887,16445,16446,14637,14638,14952,14951,14953,16822,15739,15738,15737,16268,16269,16270,16868,16271,15031,14900,14899,15030,16127,16263,15301,16853,16266,14617,14615,15466,15323,16590,15322,16589,16550,14835,15122,14834,14836,16291,16424,16425,16892,16901,15786,15787,15310,16459,16288,15388,16289,15389,14625,15475,15553,16596,16587,16595,14624,14626,15068,15133,16857,16839,16625,15067,15679,16840,16803,16899,16626,16627,16602,16600,16588,16554,16564,16548,15638,16547,15639,14848,14847,14849,16490,16491,16747,15311,16546,15476,15477,14903,16760,15678,15680,16072,16664,16071,14977,14975,15490,16641,16622,16621,16623,15712,16594,16624,15066,16593,14696,16553,14694,14695,16552,16551,15124,14622,15098,16310,16309,16311,16639,16638,16615,16674,16673,14701,16497,16499,16498,16616,15272,15271,15273,14621,16541,15399,15049,16745,14996,15279,14854,16571,15297,16574,14600,15046,15047,15045,14767,14766,16693,15185,15184,16652,16484,16651,16631,16632,16485,16483,15188,3250,3275,3929,4027,4110,4109,3603,3276,3604,3293,3814,3294,4026,3291,3290,3292,3273,3272,3274,4108,3713,3714,3715,3819,3060,3288,3269,3268,3267,2612,2424,2613,3061,11720,14316,12191,12190,12924,12867,12866,12328,11711,11710,11712,14251,13070,14208,14229,14228,14227,12329,12327,12214,12213,12215,14398,12502,13120,13113,13121,14301,12082,12080,13142,12630,11814,11813,11812,13112,13111,13137,13138,12253,14294,12946,14455,11848,13087,13086,13085,12501,11849,11850,13109,13110,13093,13088,13065,13067,13094,13089,12706,13082,13078,12705,12632,13077,12927,12262,12261,12260,12928,12943,13064,13063,13052,13061,13051,13062,13072,13050,13037,12244,13033,12009,12010,12011,13014,13016,13018,13012,13015,12993,12992,12994,12995,12996,13010,13011,12955,12954,12953,12801,12800,12799,12691,12692,12693,11747,11746,11748,12062,12893,12892,14220,12980,12981,12982,11984,11985,11986,12524,12523,12975,12249,14235,12474,12473,12250,12251,12965,12966,14349,12475,14042,22103,21253,21369,21368,22305,22104,22100,22482,21257,21256,21258,23497,23492,22682,23498,23499,23494,21760,21759,21761,21864,21372,23491,22099,21115,14462,21116,23559,14360,12968,12969,12971,12984,12983,12586,12585,14454,14282,23486,23484,23485,23487,23304,23305,23306,23420,23421,13303,23470,23469,13302,23475,23468,22330,22331,23474,23476,22953,21677,21676,21675,22335,22334,23520,23490,21479,21478,21477,22306,22562,21480,23493,21292,21291,21290,21482,21481,21281,21280,23601,21469,22640,21512,21511,21510,22639,21730,22446,21534,21993,21992,21994,21731,21729,20976,20975,21533,20940,23528,20942,20941,20885,20886,20887,20900,21277,21693,21465,21466,23606,21467,22272,21942,21941,21943,22188,21653,23024,21652,22353,22354,22355,21723,21722,23552,23530,23502,23495,21995,22473,23503,21307,21524,21306,21305,21345,21329,21330,21870,21869,21712,23263,21893,22115,21895,21894,23610,12438,12437,12436,14375,13223,13222,13135,12711,12823,12824,12589,12467,12590,12710,14382,12466,12468,13192,14244,14243,20871,20872,23603,21860,22199,21946,21526,21947,20870,22641,21181,22684,21654,22348,22230,11810,11809,11811,14242,13194,21656,21655,22452,23482,23533,20977,20979,23481,14421,12764,23599,22836,22835,22837,14439,13201,13199,14286,23473,23472,21705,21351,21706,12220,13225,13224,13226,13206,12413,23466,22196,22391,22392,22390,21350,22387,23513,22385,22010,21651,20858,20860,22009,21352,20963,20964,20859,20978,21082,21081,21080,21754,22357,22356,22694,20861,22349,20945,20944,20943,21903,22539,21573,20863,20862,21574,21036,21035,21037,22919,21815,22346,21814,21525,22329,22328,21721,22381,21658,21088,21087,21086,23003,20985,20984,20983,22261,22262,21657,21751,20997,21750,22394,22395,23489,23523,23524,22633,22561,21468,21279,21357,21356,21358,21772,23479,20938,22537,23478,23477,23488,23483,22554,21113,22555,22156,22157,22158,21384,21383,21385,23457,23460,21704,21259,21703,20935,20934,20936,23391,21261,21260,21527,21531,23439,23451,21416,21415,21414,23390,21448,23461,21446,23462,23394,23393,23308,23309,23547,23387,22012,22011,22013,23442,23443,23444,23571,23386,23052,23453,22547,23053,23051,21132,21131,21129,21221,21222,21223,22647,22646,23602,21114,21112,23463,23464,20939,20937,23538,23465,21579,21700,21701,21702,21580,20995,23480,20996,22902,22901,21519,21520,21518,21650,21474,21473,21475,21403,21130,21861,21404,22546,23537,22111,22048,21763,22130,22129,22131,23456,23454,23510,22542,22541,21796,21664,21764,21649,23373,21795,21794,23569,23471,23047,21666,21665,22080,23509,23455,22300,22299,22931,23272,23447,21148,23448,22132,20917,23446,21848,22545,21849,21850,21363,21362,21361,22454,21380,22827,23449,20916,20918,22604,22929,22930,21244,23609,21242,23436,12051,12753,12751,12834,12835,13236,12868,12231,12224,12225,13211,13210,12064,12066,13178,13177,13179,12226,11759,11758,11757,14432,14359,13176,11852,12032,14224,13208,13172,13207,13209,13234,13233,13232,14296,12168,12169,12167,12637,12638,12848,12076,14246,13171,13170,12236,12400,12399,12398,12486,13227,14241,13231,13230,13229,13228,12978,12393,12394,11845,12979,13247,13248,14324,14325,12480,11846,11847,14351,12610,11941,12608,12675,13250,13251,13253,12766,12075,13005,12869,13252,12074,12609,21879,21435,21878,21243,23585,21733,21734,23607,21530,21532,21445,21528,21529,13235,12035,12412,14274,14275,12033,12034,12644,13205,12643,11770,11769,11771,13547,11793,13154,12572,11792,14197,12573,13200,14467,12531,12533,14464,13195,13198,13196,12532,14414,12623,12944,13193,12723,12722,12721,12646,12645,12633,13167,13149,11729,11730,13165,13166,13134,13190,13189,13191,13188,12248,14211,12247,12246,13147,13148,13163,13164,14315,14423,12423,14434,13145,13143,13100,12422,11820,12889,12890,12536,13099,12614,11797,11798,14218,13096,13122,12111,12110,12109,14357,13097,13071,12173,12101,12174,13044,12541,12102,12103,12542,12181,12182,12777,12778,13007,12989,13009,13031,13013,12761,12760,13032,12939,14236,12061,12060,12059,12271,13046,12270,12269,13494,11799,14256,14298,12537,14279,14353,13095,12578,13102,13101,13103,13104,14237,12809,14405,12884,12911,12388,13048,13049,13076,12387,12373,12840,12839,12841,14248,14020,13106,12617,13079,12812,12615,13080,13136,11715,13107,12488,12487,12489,14258,11716,14391,14469,12092,12091,12090,11851,11853,12571,13139,13140,12785,12784,12252,12025,12024,11954,11953,11714,12188,13108,13090,14264,13081,12647,12639,13155,12499,13153,12676,12618,12616,12872,12871,12873,12948,12947,12958,14234,12754,13130,12593,12592,12591,14219,13129,12756,11872,11873,11874,13126,13168,14429,11892,12624,12622,13150,11891,12077,12078,12894,12628,12895,13124,13125,11964,11890,11965,11966,12822,12755,12272,12273,12274,12808,11821,11822,12629,11728,13146,12079,11794,13836,12238,12500,12237,12640,12641,12642,13047,13045,12988,12970,11640,11641,11908,11907,12363,12364,11642,13008,12186,13043,14448,14431,14404,14446,13098,14293,13123,13144,13197,12219,23459,23445,21732,22386,22004,22005,23452,23441,22427,23438,23430,23440,23428,23371,21974,12352,12087,12353,12088,23431,21975,21797,14297,21685,21684,14377,12838,11749,11750,11751,23450,23117,23458,21172,11840,11839,11841,13041,22000,12540,13042,14468,12548,12539,12538,13055,12720,12719,12490,13054,11902,11870,12552,12491,12553,12941,12940,12942,12599,11869,11871,11903,11856,13943,11855,11854,13003,13027,14452,13002,12584,13001,13304,14281,14276,14277,13030,13029,14447,13083,21173,22332,12985,13004,12845,12846,12847,12998,13023,13021,12997,12949,13019,13017,13020,13022,13040,14440,11627,13053,11626,11625,14302,12454,12455,12453,13066,13069,13068,13084,12492,14385,12922,12921,12923,14451,13039,12891,12986,12546,12547,12089,21447,23410,22444,22443,22445,21276,22347,22200,23504,13156,21331,21346,22420,22419,21881,21880,21882,22116,21692,23526,23527,23525,20974,21126,21127,21128,21254,21255,21198,21199,21200,21251,21252,22557,23500,22683,21373,12967,18288,18287,18289,3779,4014,4135,2020,1705,2021,2019,1445,1342,1341,3834,1706,3877,3279,3334,3335,3333,3299,3336,3048,3049,3050,4050,4051,1385,3599,3600,2893,3329,2891,3330,2746,2710,2745,3298,3296,3328,3501,3650,3923,3326,3327,3884,3889,3365,1490,1491,1489,3830,3828,3358,3357,3383,3356,3999,3361,3362,1363,2695,2696,3431,3998,3997,3432,3410,3411,1539,1540,1541,3802,3803,3804,3352,3351,3350,3353,3840,3355,2847,2848,2849,3360,3359,2082,4115,3325,2083,3651,3652,1382,3665,3664,1384,1383,3295,3297,3277,2709,3976,4191,1488,1487,1485,3248,3249,1486,3225,3251,1340,3801,3278,1962,1482,1483,1484,1465,1466,1464,3289,3323,1455,3324,3318,3317,3316,3322,3354,4179,4180,3835,3506,3313,3314,3315,4181,3265,3505,3817,3075,3312,3310,3309,3311,3321,3320,3319,3074,3073,3263,3985,4022,4034,4033,4015,3262,3240,3630,3628,2380,3242,2000,3243,2001,2002,3629,3264,3266,3287,3286,1456,1457,2738,4155,2737,4018,4138,20528,20501,20743,18648,18568,18743,18744,18519,18520,18182,18521,19325,19326,20581,20592,20584,18181,18180,20583,20574,20573,20582,19171,19948,19947,19949,20471,19172,19170,18605,18303,18604,18305,18304,19144,19191,19190,18414,19482,19481,19305,19304,18641,18646,18154,19225,18157,18156,18525,18526,18750,20591,20578,19906,18752,19343,19342,19341,19348,19347,19453,20718,20597,20719,19452,19454,19559,20701,18091,19557,20437,18907,18909,20634,20632,18984,19103,18985,19486,20646,20645,18731,19336,19335,19334,19485,18729,18730,19349,19350,18956,19428,20668,20669,18420,19096,19097,19098,20648,20210,19427,19426,19052,20636,20635,19053,19051,18229,18230,18228,20621,18929,18908,18090,18893,18092,20609,20608,20832,20612,19149,18141,20772,18640,18155,20599,18642,18139,18140,18413,18376,18377,18378,19357,19356,20596,19036,19035,19034,18176,18175,18174,20586,19658,18316,18711,18710,18799,18800,20603,19654,18375,19498,18564,18373,18374,18616,20018,20587,18315,20588,18560,18086,20778,18283,18084,18282,20572,18715,18703,18702,18704,20575,18567,18290,18291,19475,19474,20451,18121,18122,18123,18135,18884,19372,18085,20569,18450,18449,18448,18559,19528,19123,18116,19122,20568,3896,20570,4043,20756,19750,3303,3304,1704,4095,4146,3920,3897,3919,3338,3339,3340,1366,3372,1365,4090,4144,4113,18117,18933,20728,3895,20585,20703,18115,20594,20595,19500,4246,4112,20815,20601,20602,19184,19183,19182,20613,4070,3806,20614,3807,18809,19295,20697,18195,19443,18810,18399,20640,18398,4132,20637,20638,20639,20641,4131,3492,3510,3513,20649,20650,20651,18421,18423,18422,20670,20671,20654,20682,20655,20672,18534,18533,18532,19353,19573,19572,20660,20673,20661,18136,18320,18244,18321,19267,18680,19266,20600,20662,20663,20674,20675,19825,18243,18242,19370,18169,19039,19824,19360,19156,19507,19157,19508,20689,20690,20691,18168,18167,20738,19439,19438,18257,18137,18258,18138,19344,19345,20687,20688,20686,20684,20805,19117,19118,19062,19063,20683,19387,18540,18541,18542,4174,1454,19061,19369,20207,20505,4105,4104,4103,4224,4223,3462,3461,3460,3590,3589,3591,3593,3592,4066,1765,4056,3588,4044,1763,1764,3558,3557,1595,4199,1594,3126,3127,4128,3526,3525,3527,3553,4055,3280,3487,3490,3489,3488,4081,3845,4082,3281,3282,3737,3738,3465,1910,3466,3659,3418,3419,3671,1911,3420,3422,3421,3394,3414,2703,4236,3444,3440,2705,2704,3439,3438,3437,3436,3602,3090,3089,3413,3902,3433,1361,1362,3385,3386,3388,3387,4152,3392,3393,3395,3148,3367,3369,2892,3332,4158,3366,4040,3368,3601,3147,3146,3699,3391,3390,3389,3398,3399,3415,3416,3417,2626,2625,2624,3443,2966,2308,3447,3448,2307,2309,4069,3987,3493,3491,3511,1596,3512,1782,1781,1452,1453,3587,3586,18573,3690,3673,3396,3397,1364,3371,3370,1387,1386,3337,3434,3435,1380,3676,1381,3876,3458,3457,2480,3826,1379,3476,3475,3473,3477,3474,2821,3615,3503,3502,3504,3626,3614,3627,3546,3700,4106,3541,3961,3540,3539,3536,3831,3537,3518,3517,2578,1666,1894,1893,1519,1520,1521,3924,1797,3810,3809,3808,1799,1798,3746,3532,1369,1367,3533,3516,3515,3622,18356,20780,20679,20678,20680,18100,20681,18787,18101,18788,18918,18384,18385,20677,3773,3774,3775,4083,3875,18982,18983,18386,20644,20633,20631,19558,19594,19593,4177,20620,20629,20619,20630,4124,3402,3403,3658,3657,3423,3424,3425,3404,3375,3085,3084,3083,3376,4183,3706,3974,3379,3378,3380,3349,3348,3347,3344,3345,3346,4134,3341,3343,3342,3401,3400,3374,3373,20618,20826,20605,20606,19339,20607,19340,18134,18133,18132,18528,18527,18906,3308,3307,3306,4048,3305,3284,3283,18798,3285,4085,18986,20776,20589,18170,18171,18172,18173,20798,20800,20576,20577,19980,19638,19637,19636,19635,19639,20580,20579,19324,20813,20503,19188,20590,4130,3887,3407,3408,2482,3409,2481,2478,3412,2479,3459,3430,3429,3454,3453,3452,3428,3455,3456,3472,4078,4077,3469,2863,2864,2865,3426,3427,3405,3406,3449,4045,3934,4036,3470,3451,3450,3874,20666,3498,3499,3514,3743,1665,1664,3500,3616,2823,2822,18099,20667,19409,20685,19781,19780,19126,19125,18467,18468,18469,19346,18757,19226,18656,18654,18655,18431,18973,18418,18419,18980,19158,18432,18433,19582,20693,20692,20696,20695,20825,19358,19359,18674,18694,18672,18673,19025,18734,18735,18736,4087,4086,4088,1368,2936,2935,3495,3494,3496,3745,3564,3565,3535,3761,3916,4210,4167,4211,3568,3597,3567,3538,3566,3870,3569,3571,3542,3543,3545,3544,3572,4038,3547,3550,2274,2273,2275,3548,3549,3519,3869,3520,3868,1970,1971,1972,3883,3606,3480,3605,3479,3740,3464,3478,3463,3482,3442,3441,3483,3484,3445,3468,3508,1421,1509,3529,3528,3507,1507,1508,3792,3793,1493,3996,3583,1492,3551,3552,4011,3926,3927,3928,3585,3584,3948,3949,3731,3730,3556,3555,3554,3837,3530,3581,2027,2028,2029,3578,3575,3579,3576,3577,3573,3574,3570,3582,3531,3912,3521,2679,1420,1426,3523,3522,3524,1494,1422,3481,3485,3486,3467,1912,3446,3534,18873,20694,20729,18358,18355,18359,19469,19470,18955,18954,18190,18357,20498,18830,20676,20665,20664,20785,19270,20643,20647,20209,20626,20628,19414,19415,19413,20721,20802,20610,19767,18147,19148,18146,18145,18281,18999,19000,20611,20622,20623,20616,19901,19237,20625,20624,19620,20627,18514,18513,18515,18681,18682,20656,20659,20658,20657,20653,20717,20472,20642,20698,18193,20652,18194,18875,18877,18876,19444,20604,19882,19238,19236,19294,19499,19527,19952,18751,20598,20439,10812,10813,8970,9448,9447,9904,9490,9489,9864,9861,9903,9142,8959,9140,8960,8961,9267,9268,10909,9394,9395,9230,9705,9163,9704,9162,9061,9231,9941,9975,9976,8979,8980,8981,10026,10022,10923,10809,10808,10810,9216,10771,10770,10030,10894,10978,10888,10031,9270,9269,9271,10145,9286,9902,9520,8767,8766,8768,8944,8945,10905,10804,9853,10805,10765,9157,8815,8816,8817,9087,9086,9088,9156,9854,9813,9812,9811,9788,8748,9083,10694,9859,9816,9818,9820,9858,9860,9815,10749,9857,9856,9855,9898,9852,8689,8688,8687,9229,9942,9755,9758,9720,9718,10854,10853,10795,9757,10794,10722,9698,10723,9697,10966,9695,9300,9299,9298,9692,8698,9676,9693,8648,9424,10858,9650,9625,9653,9626,9623,9651,9652,9679,9678,9647,9677,8675,8676,8677,9102,9654,9619,9645,9673,9103,10011,9646,9643,9641,9644,9621,9617,9616,9596,9622,9602,10216,9618,9620,10785,9603,9624,9953,9426,9604,9629,9425,10714,9362,9361,9363,10713,10706,10738,10863,10824,9364,9586,9585,9255,9254,9256,9554,9570,9568,9558,10727,10914,9553,9559,9560,9959,8895,9548,10718,9288,8897,8896,10422,10970,10896,9556,10736,9557,9555,10027,10025,10023,10029,9060,9062,10944,10028,10075,9561,9562,9563,8710,8709,8708,10733,9599,9598,9597,10077,10128,10127,10076,10938,8903,9606,9358,9359,9360,10734,9587,10702,9293,9588,10741,9289,9287,9462,9592,9594,9593,9595,9607,9609,9612,9611,9613,9638,9637,10698,9666,9219,8753,9092,9091,10891,8754,8906,8908,9012,9668,8725,8723,8724,8806,9751,9752,9753,10469,9580,9581,9582,10845,9675,9674,9104,9717,9716,9688,9667,8752,9642,9640,9639,9615,9614,9579,9126,9127,9128,9571,9564,9572,9601,10739,8867,8870,8871,8868,8869,9569,9565,9378,10719,9376,9377,9203,9547,9214,9215,10913,10959,10895,9552,10737,8697,8696,10705,10707,10908,8646,8647,9694,9429,9696,9241,8671,8669,9232,9423,9233,8805,8807,8808,9075,9670,9669,9749,9748,9750,10865,9804,9801,9805,9802,9785,9786,9747,9010,9009,9008,9715,8830,8829,8831,9740,10957,9714,10782,9713,9686,9741,10695,9738,9797,9737,9352,9169,9168,9167,9739,9784,9519,9800,9517,9518,9781,10974,9284,9495,9497,9496,9494,10772,8913,8914,8915,9521,8772,9522,9285,9283,10911,8771,10969,9380,10751,9379,9381,10752,10916,9926,10922,9924,9181,9925,9927,10870,9887,9888,9886,9929,9928,8883,8882,10003,9930,9511,9509,9510,9958,9184,9183,9182,9393,9391,9392,10760,10761,10007,8656,8655,8657,8934,10008,9319,9318,10744,10061,9250,9264,9263,10103,9297,9296,10762,10141,10085,9026,9025,9024,10142,10104,9575,10940,10102,10144,10934,10174,10172,10171,10166,10167,10168,10799,10215,9901,9900,10173,9538,10217,9537,9536,9266,8749,8750,8751,10220,9145,10223,9144,9143,10222,8819,8818,8649,10309,10308,10686,8650,8651,10300,10301,10956,10350,8774,8668,8769,8854,8667,8666,8992,8728,8727,8726,8855,8856,9948,10764,8653,8952,10423,9095,10424,8765,8828,8826,8827,10871,10883,10882,8991,10420,10421,10233,10419,10458,10457,8799,9123,10789,10232,8732,8733,8734,8866,9054,9055,10412,9346,9345,9344,10227,10390,10385,10391,10392,8785,10418,9146,10318,8770,10393,8786,10347,10348,10349,10344,10343,10345,8784,10917,9411,10961,10298,9412,9077,9076,8849,10297,10299,10244,8850,8851,9526,9527,10243,10245,10828,10303,10246,10247,10753,10163,10213,9304,9316,9317,9239,9469,9468,9303,10214,9584,9176,9177,9178,10084,10086,9742,10164,9744,9743,10098,10976,10138,10975,10165,9240,9238,9329,10139,10161,9350,9349,9348,9956,9957,8879,8878,8880,8794,10211,10818,10293,8888,10292,10291,10830,10294,10296,10775,8804,8802,8803,9131,10295,10290,10340,10341,10342,10384,10383,10382,9334,10381,10284,9336,9335,10283,9337,10408,10448,10449,10791,9333,10925,9632,8899,9015,9208,10949,10511,10477,9312,9311,9309,9977,10479,10480,10930,10931,10514,10946,10482,10481,10483,8743,8742,8741,10516,10515,10485,10844,10450,10453,10454,10416,9528,9530,9529,9910,10456,9228,10452,10742,10451,10935,10852,9354,9355,9356,10851,10787,10521,10800,10543,10542,10540,9002,9003,9004,9247,9248,10555,8793,8792,8791,9212,9213,9151,9150,8955,10562,10841,10798,10919,10747,10920,10712,8935,10597,9161,9031,9029,10560,9483,9482,9481,9920,9874,10242,8936,8937,10640,10054,8714,8715,8716,8661,8660,8662,8933,10926,10683,9492,9019,9491,10668,10669,565,599,563,561,562,559,560,10670,9023,9022,558,557,556,555,554,10681,10682,10906,10677,10756,9472,9471,9470,10241,10666,10665,10647,10678,10679,602,550,512,549,551,552,10680,553,604,605,606,607,608,609,17619,17604,17601,17618,17047,17607,17309,17605,17606,17608,17621,17620,603,513,511,7492,7493,7442,5610,17058,17059,17060,16926,17310,17573,17560,17561,17586,17587,7790,7494,7445,5609,7791,7843,7444,7822,5777,5775,5776,7446,7447,5664,5663,5665,7495,7496,7488,7395,5838,7399,7397,7398,7438,5766,5767,5768,7794,7336,7832,7808,7329,7275,5931,5930,5929,7274,7229,7167,7228,7166,7239,7878,7241,7096,7164,7165,7223,7224,7221,7278,7280,7279,7751,7276,7277,5681,7388,7387,7391,7334,7390,5755,5752,7331,7332,7389,5754,5753,7392,7441,7443,509,506,508,510,548,547,10676,10663,10662,10661,10648,10633,10667,10632,10634,10629,9133,10630,9134,9132,10038,10586,10596,9873,10594,10593,9508,10595,9374,9172,9171,8673,10538,9310,9373,8674,8672,10512,10478,10476,8810,9457,9138,9458,9416,9136,8809,8811,9139,10726,8926,10443,10846,9067,9066,9068,9185,10847,9079,8962,9402,8963,10310,8943,10717,8941,8942,10404,10405,10768,10767,10843,10842,8859,10692,10330,10376,8857,8858,10329,10286,10285,10287,10326,9036,10278,9034,9035,10279,10947,10275,10274,8965,8966,8967,10827,10826,8790,10238,10237,10239,10874,10950,10766,10186,10234,9341,9340,10270,10269,9196,10267,10266,9000,8999,9001,9463,10265,10106,10268,10105,10107,10322,10366,10365,10323,10398,10955,8886,9387,9386,9388,10435,10100,8884,8832,8833,8834,9290,9272,8692,10400,10399,10401,10436,9273,10437,10402,8700,10441,10403,8701,9217,10440,8840,10442,8839,8838,10484,10473,10474,8836,10508,10507,10517,10532,8905,533,10530,532,10531,10505,10506,531,10529,10504,10503,10472,530,529,476,478,477,479,480,481,482,483,484,485,486,487,534,536,535,8904,10539,10535,10534,10520,10518,10519,10510,8837,8835,10170,10509,9005,9400,9399,9401,10169,9504,9505,9506,10409,10410,10927,10750,9084,9085,8699,8919,8920,10370,10371,9332,9331,9330,10839,10324,10367,10368,10369,10271,8789,8722,9195,8720,8721,10206,10207,10325,9197,8885,10968,8964,9089,9090,10012,9007,10533,10558,10537,10589,10588,9101,10637,10628,10635,9461,9459,9460,10587,10585,10584,10626,10625,10627,10636,10658,9434,10631,10659,9100,10674,10675,541,540,10664,539,10638,9099,10590,10536,10639,10592,10591,537,10559,490,493,492,538,495,494,496,497,498,516,515,499,500,501,7163,7782,7835,7222,5680,7220,502,503,504,505,5682,507,518,544,545,546,543,9433,9436,9435,10657,10660,542,7162,7091,7090,7039,7040,5846,5845,7097,7095,7041,6966,6967,7042,7098,7100,7099,6580,5592,7218,7217,7273,5591,5590,7219,7272,7271,7383,7858,7330,7328,7381,7382,7327,7384,7417,7418,7771,5627,7326,5628,5629,7315,7314,7260,7195,7259,7732,7731,5867,5866,5868,6057,6058,7366,7365,7420,7462,5760,7421,7367,7465,7422,7370,7423,6112,6111,6110,7368,7316,7270,7262,7261,7211,7155,7154,7214,7213,7212,7215,7157,7156,7082,7081,7153,7145,5647,6368,6961,6960,6959,7035,6962,7036,7037,7038,7083,7158,7159,7144,7197,7196,7143,7194,7193,7142,7141,7071,7072,7024,5624,7073,5623,5622,6965,6964,6950,6877,6888,6887,6886,6814,6815,5749,6065,6805,6804,6876,6874,6872,6875,6803,6740,6741,6806,5577,5571,6807,6818,6873,6889,6948,6947,6946,7020,7023,7069,7070,5559,6951,5561,5560,6892,6893,6823,5658,5660,5659,6820,6968,6831,6821,6900,6899,5847,6972,6970,7789,7086,6971,7085,7087,7088,7089,7788,6969,7043,7044,6291,491,514,488,489,6897,6825,6827,6826,6828,6898,7742,6830,6822,6756,6712,6757,6711,6710,6709,6829,6755,6383,6754,6751,6817,6704,6703,6702,6816,6655,6649,6604,6650,6606,6651,6550,6549,6551,6496,6494,6493,6432,6495,7819,7842,6433,5871,6386,5872,5869,6431,6429,6428,6430,6385,5800,6325,5802,5801,7876,6294,6255,6253,6213,6212,6214,6252,6171,5740,6170,6169,5741,5742,6168,6207,5805,5804,6246,5540,6290,5811,5539,5812,6254,6335,6330,6329,5870,6328,6388,6387,6389,6390,6437,6440,7766,7767,6333,7737,6332,7738,7829,7830,6414,5511,6364,6360,6318,5830,5829,5831,6282,5512,5513,6362,6363,6378,7865,5548,5549,5547,6288,6287,6289,6286,5746,6279,6239,6240,6241,6232,6197,6187,5878,5877,5876,6233,6234,6188,6185,6186,6190,6148,6151,6147,6158,6155,6152,6109,6154,6153,6143,5892,5891,5890,6094,6092,6093,6059,6095,6079,5694,5692,5693,6036,6031,6032,6033,6081,5570,6080,6082,6052,6053,7188,6012,5734,6011,6009,6008,5735,6010,5995,5981,5984,5980,794,5994,6007,798,671,797,674,842,800,799,676,6034,6035,5736,803,6054,6056,6055,802,680,679,804,683,841,805,6089,6090,6085,6084,6083,6091,5568,5569,6135,6138,6088,6173,5807,5806,5808,6172,6218,5837,6221,6220,6219,6175,6222,6265,6262,6266,6302,6303,6304,6342,6450,6301,6341,6299,6396,6395,6447,6448,6449,6451,6510,6511,5863,5864,5865,6397,6398,6399,6405,5554,6400,5520,5521,5522,5553,5555,6461,6460,6576,6577,6292,6293,6624,6623,6575,6574,6573,6625,6622,6626,6684,6685,6726,7736,6727,5875,6782,6734,6732,6733,5937,6784,6785,6426,6686,5797,6424,7774,5796,5795,5947,6682,6683,7875,6681,6620,6680,6724,6571,5822,5910,6570,823,824,717,719,825,826,6722,6723,6772,6847,6848,6849,6777,6725,6776,6851,6852,7901,6853,6671,5860,6780,6312,5850,6314,6313,6779,6778,5946,5948,6781,6856,6857,6858,6425,7804,7801,6696,6786,6695,5938,5936,6699,6930,6929,6931,6697,6932,5531,5530,5529,7884,6652,7007,6653,6654,7008,6071,6070,6072,7053,7054,7055,7058,7057,7056,5712,5711,5710,7802,7756,6860,7748,7747,5849,6997,6855,6926,6998,7800,7000,6999,7849,7885,7059,7856,7120,5927,7119,7825,5347,7824,5452,5098,5099,5097,5128,7855,7857,7002,7827,6995,6923,5311,6996,5129,5130,5382,5436,4332,4331,4330,5127,5384,5147,5095,5094,5380,5341,5103,5102,5104,5096,4328,5079,5077,4329,5049,5054,5051,5053,5057,5080,5078,4408,4409,4410,4413,5056,5076,5071,5072,4433,5093,5073,5081,4516,5387,4992,4993,4517,4515,5046,5359,5375,5043,5024,5064,5044,5063,4521,4520,4519,5092,5116,4991,5337,5074,5142,5125,5126,5101,5100,5131,4319,4498,4497,5143,5141,5145,5144,5146,5151,4318,5184,5176,5175,5174,5185,5186,5200,5212,5213,5187,5215,5216,5247,5248,4381,5251,5278,5266,5277,5281,5304,5303,5305,5306,4379,5307,5283,5320,1220,1221,1184,1186,1222,1188,1223,5284,5326,5308,5282,5321,5310,5309,4414,5285,5280,5286,5221,5256,5255,5252,5253,5254,5250,5279,4380,5149,5249,5150,5218,5217,5219,5191,5192,5223,5220,5222,5235,5236,5233,5193,5205,5180,5178,5234,5208,5238,5181,5179,5207,5206,5154,4469,5211,4468,5156,5210,5209,5157,5155,5182,5183,5377,5379,5466,5475,8389,8383,8382,8384,8514,8326,8327,8328,8511,8386,8385,8387,8388,8336,17952,17965,17123,17124,17125,17986,8324,17998,17322,17321,17320,17653,17462,17967,17985,17647,17646,17996,17342,18005,17264,17263,17265,18007,17183,17182,17181,17419,18006,8351,8377,18057,18011,8378,18012,18013,17774,18015,8374,18016,8375,17775,17776,854,852,768,773,851,879,878,18017,770,769,766,853,877,856,18014,17398,17186,17185,17184,18003,18002,18004,17266,17092,17267,17399,18010,17043,858,17044,17946,18009,857,761,762,875,860,859,760,18034,16993,16992,17945,17445,17995,17444,18001,17090,17091,17403,17404,17984,17983,17981,17343,17341,17645,17429,17982,17303,17302,17401,17400,17402,17453,17421,17420,17295,17137,17940,17138,18020,17939,17938,17937,17927,17926,17928,17922,17905,17904,17925,8523,8527,17906,8509,17902,8396,17069,17070,17015,18025,18026,17845,17881,17866,17013,17208,17079,17080,17081,17815,17813,17840,17867,8404,17839,8405,8339,8340,8358,17814,17788,17789,17787,17782,17790,17346,17344,17345,17210,17209,17816,18052,17812,17838,17278,17277,17279,17464,17465,17048,17396,17050,17865,17879,17864,17880,17899,17863,17877,17876,17898,17897,17895,17894,17896,17893,17892,17916,17915,17914,17917,17432,17934,17919,17918,17237,17236,17235,17481,17936,17949,17428,17951,17304,17239,17238,17935,17296,17297,17923,17924,17920,17049,17921,18023,17903,17901,18054,17900,17950,17964,17963,17962,17405,17978,17979,17980,17994,17478,17477,17476,18008,17023,17024,861,17082,17083,866,865,18035,753,873,754,755,756,778,758,8551,8550,8643,8593,8552,8592,8549,777,752,751,8548,8547,749,8585,8582,8644,8588,8637,8586,8557,8556,8579,8581,748,745,8576,8575,743,8570,8571,8577,8572,8565,8568,8569,8567,8566,8564,740,739,645,741,742,744,870,871,872,738,849,17988,17987,17148,6764,5954,5949,7811,6272,6273,17147,17149,17199,17407,17406,16999,18041,17957,17956,17440,17358,16968,17359,17955,18043,17943,16998,17000,7895,17944,17931,17329,7851,5976,5959,5978,17330,17910,17911,17930,17907,17139,5979,5537,5536,5535,6004,7916,17319,17408,17313,17053,17885,17882,17884,18036,7913,7860,6006,6005,7861,6028,6027,6029,6050,6068,6069,5744,5743,5604,6051,5602,5603,6076,7807,6073,6128,6125,6077,6124,6127,7910,6129,6123,6122,6119,6074,6120,6121,6126,6165,6164,6166,6167,6198,6199,6200,5803,6242,6231,6202,6183,6203,6229,6230,6238,6237,6276,6278,5625,5626,6317,5538,6359,6334,6283,5884,6284,5883,5882,6281,6235,6236,6277,6191,6192,5722,6182,6181,5723,5721,6103,7897,6104,6163,6114,6113,6063,6146,6096,5798,6097,5783,6098,6102,6101,6099,6100,7844,6061,6043,6044,6060,6041,5785,5784,6039,6023,6017,6040,5998,6000,6022,5715,5714,7795,6001,5944,5788,5787,5786,5789,5790,5791,5945,5990,6021,5713,6025,6062,6024,6047,6046,6045,5773,5772,5774,5510,5509,5508,6048,5576,6049,5881,5880,5879,5575,5574,6064,7841,6161,6160,6162,6201,6159,6115,6067,6066,6026,5661,6003,5993,5662,5977,5934,5933,5932,5991,5992,7846,5975,5971,5965,5717,5935,5964,5716,5958,5955,5951,5956,5970,5972,5957,5963,5968,5962,5969,5967,5966,5989,7864,6002,5997,5763,5764,5765,5985,5986,5987,5983,796,5982,795,669,843,792,664,736,793,7854,844,790,663,737,791,845,788,660,658,789,846,786,656,654,787,847,784,652,650,785,5952,5953,5950,783,648,646,782,848,643,644,647,649,8128,8129,8121,8120,8119,8069,8261,8294,8122,8116,8115,8111,8276,8114,77,8113,8117,74,8084,8083,8085,8118,8125,8124,71,73,72,75,76,4258,4259,4260,4261,19656,19655,19631,19629,19176,19175,18614,19651,19650,4257,18615,18435,18379,18380,18370,19646,18371,18372,19625,18499,18500,4256,4250,4249,82,79,20730,78,80,81,83,8110,8112,8068,8070,8259,8533,8532,8543,8559,922,8580,8574,8573,8583,8584,8640,8578,8558,8587,8563,8562,8639,918,917,916,919,920,921,84,942,442,443,444,456,453,438,446,445,447,943,915,914,913,909,912,448,925,926,2776,2831,2777,2832,449,441,439,436,440,2311,2310,2833,2778,2780,2734,2781,2782,2783,2834,2835,426,2786,425,421,433,428,429,427,430,417,419,409,410,418,415,416,450,451,413,412,406,19181,19657,19652,19653,18601,18833,18832,18831,20702,18479,18979,18488,18487,18489,18599,18600,18434,19223,19041,422,19040,414,423,424,455,19644,4255,431,434,435,437,432,18446,18447,19662,18478,18477,18083,18239,18238,18237,19135,19136,18082,19232,19441,20754,19440,18821,18817,18816,18818,19321,19322,19690,19691,19703,19704,19702,19705,19721,19314,19706,19316,19315,19373,19722,18518,18988,18987,18989,19723,19751,18917,18916,19753,18475,19771,19060,19059,19754,19770,19943,19793,19796,19776,18251,18253,18252,18593,19794,19828,19249,19856,19857,19287,19878,19200,19199,19883,19900,19863,19907,19908,19264,19230,19231,19229,19457,19934,19221,19916,19455,19456,18825,18824,18823,19280,19961,19960,18812,19979,19978,19512,18773,18763,18774,18764,19981,19579,19998,18811,19087,18578,18577,18576,18223,18222,18224,18626,20046,20045,20026,20043,20044,20078,20042,20028,20075,20077,20076,20110,20119,20118,20117,20102,20103,20120,20121,20148,18329,18328,18330,20149,20122,18212,18213,18214,20124,20105,20123,18202,18203,18204,20154,20151,20150,3301,20152,20153,3772,1578,1637,3855,1638,18383,18382,20201,20180,19140,19139,19138,20200,20199,20218,18921,20219,18598,18597,18596,18920,18919,20242,19478,19479,19480,19619,19741,20241,18741,18740,18742,19536,20272,20293,20735,20346,20317,20347,20319,20790,20361,2103,19077,2102,1988,1987,20320,20318,20294,20321,4145,1700,1701,20295,20273,20274,3561,20833,20243,3709,3708,3562,3857,20275,3962,3598,1695,1754,1755,1811,1813,1812,1814,1879,1746,1745,1815,1816,1880,1817,1757,1760,1818,1819,1881,1820,1821,1882,1883,1887,1886,1824,1762,1822,3849,1889,4035,1825,1826,1823,1767,1750,3678,1827,1828,1535,1534,1536,3674,1829,1774,1772,1773,1790,1834,1833,1830,1900,1903,1831,1904,1902,1901,1950,1951,1898,1897,1899,3785,1967,1966,1965,1963,1964,2049,2051,2050,4049,2128,2127,2131,1653,1652,1649,1650,1651,2194,2195,2197,2198,2264,2263,4169,2196,2124,2126,2125,2266,2267,1478,1477,1476,2346,2347,2344,2439,2431,2538,2525,2524,2433,2539,2540,2541,2535,2534,2533,2537,2438,2343,2437,2434,2531,2532,2536,2600,2529,2528,2530,2651,2650,2649,1429,1428,1427,3132,2769,2715,2714,2716,2768,3797,2816,2817,2813,2814,2815,2860,2859,1668,4114,2808,4004,2898,2046,2852,2048,2804,2805,2806,2807,2809,3377,3906,2707,1753,3749,3750,2702,2698,2697,2699,2701,4010,1810,1809,1808,2700,1601,3471,3980,2642,3981,2761,2762,1602,3842,2759,2639,2760,2640,2638,2584,2641,2587,2585,1614,2589,2588,2427,2428,2426,2429,2516,2517,2518,1358,1359,1360,2060,2061,2335,2336,2337,1461,2258,2257,2186,2334,2333,4089,1615,2423,2331,2330,2332,1662,3799,4187,2108,3742,2109,2107,2185,3864,3851,3850,1959,1958,1957,1947,1949,1948,2033,3271,1744,1952,1481,1953,2034,2113,2114,3967,1532,1462,1463,2187,2188,2259,2260,2192,2261,2338,2339,2340,2432,3836,3744,4220,2526,2521,2522,2523,2527,3965,4168,2596,2595,2602,2598,2597,2706,2603,2604,2656,1752,1751,2652,2594,2590,2643,2349,2350,2262,3798,2193,2191,3726,3725,3727,4047,2122,4046,2129,1646,2130,1644,1645,2045,1961,1960,1566,4107,1890,1891,1567,1568,1888,3384,3753,2043,1956,1955,1479,1480,1954,2041,2042,2117,2116,2115,2118,2123,4170,3784,2044,1892,4057,1878,1875,1876,1877,1946,1944,1699,1945,2069,2918,2184,1663,1661,2104,2256,2255,2183,2328,2329,2420,3208,3209,2511,2421,2422,2513,2514,1616,2515,2586,2583,2581,2582,2579,2510,2512,2580,2636,2691,20459,20460,20468,20469,20470,20445,20443,20426,19256,19272,19271,19105,20401,19106,20397,20364,19534,20395,20396,19329,19330,19093,20359,20358,20360,20367,19095,19094,19531,20384,20418,20385,20419,20420,20762,19078,20386,19074,19076,19075,20344,20345,19303,19302,19301,20315,20313,20316,19009,18128,18127,20292,20277,20297,20314,20322,20296,20276,20282,20767,19626,20271,19627,19628,20220,19081,19080,18120,18118,18119,20240,20239,20203,20204,20205,20779,19740,19995,20769,20181,19996,20145,20143,20144,20176,20467,20109,20147,20125,19243,20126,19244,19245,20107,20108,20084,20083,20061,20085,20050,20051,20027,20024,20029,20011,20023,20049,20059,20058,19246,20060,18430,18188,18189,19042,19043,19859,19832,20092,18098,18097,18096,18187,20090,20091,20087,20088,20089,20784,20057,20056,20112,18412,20020,19551,19550,19552,20054,18544,18850,20055,18775,18205,18545,18543,18206,19248,20008,20007,19989,20009,20010,18165,18164,18166,19084,19982,19963,19955,20006,20005,19962,18218,18220,19468,19467,20615,20757,18887,19923,19924,19922,18695,19895,19893,19894,19896,19921,18856,19892,18981,18855,19127,19618,18885,18886,19129,19128,19990,19971,19967,19970,19993,19992,20012,20791,19688,20751,18310,19969,19308,19307,19306,18309,18311,18801,18802,20710,18739,18662,18493,18494,18495,19585,19584,4062,18661,18663,2288,4063,2915,2916,2917,3947,2147,2144,2145,2146,2289,4126,3990,3988,3989,19945,19058,18455,18456,18457,19057,19929,19010,19011,19012,20737,19972,19968,19944,18854,20753,19918,19919,19920,20792,19599,19891,19600,19868,19843,19869,19613,19866,19844,19867,18697,18696,19917,18219,19942,19562,19561,19560,19890,19865,18106,18107,18105,18760,19954,19936,19909,19937,19938,19789,19083,19082,19790,19791,19957,19958,19959,19983,19622,19621,19623,19985,19984,19487,19608,19987,19986,20025,18813,19279,19647,19648,19941,19940,19939,19914,19912,19910,19911,19913,20704,19563,19564,19565,19888,19887,19860,18363,18108,18362,19840,19864,19886,19802,19861,19837,19782,19801,19838,19833,18709,18708,18417,19797,18416,18415,19798,19834,19835,19836,19884,19885,19862,19858,19219,19220,19889,19915,19831,19388,19800,19777,19775,19799,20221,19774,19772,19773,19589,19586,19587,19645,19588,19738,19736,19725,19737,19724,18822,19537,19406,19405,19407,19328,19624,20816,19693,18776,19692,19694,19708,19707,19275,18508,18507,18509,18510,19712,19710,19711,19727,18368,18367,19592,19591,19590,18698,19055,19054,19056,19784,19783,19804,19803,19785,19807,19839,19805,19806,19163,19162,19159,19160,19161,18550,19814,19813,19841,19842,19601,19847,19848,19872,19048,19870,19871,19897,19898,2376,19926,19928,19927,19925,4123,4122,3683,3642,3847,2375,2374,3643,3960,3735,3736,2294,3959,3381,2295,2296,2385,2384,2383,2301,2223,2297,2382,2386,2298,4237,4101,2413,2411,1608,2412,2415,2414,19786,19787,19808,19809,19812,19811,19810,19758,19956,19759,19033,19032,18317,19743,1609,19189,19757,18458,2634,18459,18460,19700,19744,18845,18319,18318,18552,18365,18549,18551,19598,18366,18364,19726,18844,19701,18369,18352,18353,18354,18323,18322,18324,18665,18666,19378,18523,18522,18524,19685,18781,19110,19380,19379,18302,19431,19695,18300,18301,18639,18081,18777,18978,19323,19670,18782,18975,18976,454,405,19178,18144,18143,18142,19679,394,395,393,458,452,459,460,2687,2688,1412,1413,2708,2689,2690,3891,2631,2509,2576,2574,2573,2630,1470,2629,1472,1471,2996,3741,2321,2322,2406,2405,2506,2504,2572,2505,2503,2685,2623,2502,2317,2404,2571,2568,2622,2570,2683,2682,2741,2684,3935,2748,2749,2216,2215,2214,2750,2686,2628,2627,2789,399,398,396,400,403,402,411,2747,2743,2787,2788,420,2744,2742,2785,2739,2740,2681,2680,4037,2476,2784,2736,2735,3975,2477,2475,2494,4235,2495,2560,2561,2559,2621,2678,2677,2779,2733,2672,2732,2671,2731,2730,928,927,911,924,906,908,929,904,930,2040,2038,2670,2039,2615,2614,2676,2617,2557,2487,2486,2484,2616,2556,2397,2306,2488,2235,2489,2558,2492,2491,2493,2400,2401,4228,2208,2209,2207,2233,3866,2232,2205,2206,2142,3865,2143,2141,4129,2062,2063,2064,2081,1863,1862,2085,2005,2004,2003,2084,2165,3649,3647,3648,2007,2006,3702,3703,2009,2168,2086,2008,2090,2089,2087,2088,2091,2012,2011,1937,2010,1682,1864,1865,1866,3704,3705,3943,1860,1931,1932,1929,1926,1927,1928,1930,1976,1977,3930,3815,3901,2947,3644,3933,4079,1835,1836,1837,1840,1841,1842,1895,1896,1838,1550,8017,7973,7948,7947,7944,7946,7945,7998,7974,8032,8034,11683,11684,14092,14091,14081,14082,14083,14084,12797,12068,12798,12069,14090,11606,14100,14099,11608,11607,12094,13951,13341,13342,12139,12138,12256,12712,14119,14120,12259,12258,14121,14138,14135,14139,14137,14136,14118,14336,14134,12427,12587,12426,12588,13340,14110,12095,11697,13605,13606,12782,14117,12783,12456,12457,12458,12428,12429,12430,14132,14133,14116,12897,14115,14131,12476,14166,12278,12279,12280,14130,14145,14163,14146,11944,14164,14165,14177,14176,14178,14190,11938,11939,11940,1256,1255,1254,1257,1258,1259,1260,14175,12917,11739,11738,1262,12918,12919,14160,14159,14158,14174,14305,14142,13259,14393,14141,983,975,972,989,990,991,984,997,996,998,993,992,995,994,1267,1268,1269,999,1270,5489,5490,5491,5492,14578,14575,14579,14482,14481,14480,14570,14569,14571,14576,14592,14520,14488,14519,14487,14489,14583,14580,14581,14521,14584,14497,14496,14495,14476,1173,1171,14586,14587,8253,8252,8251,733,8249,8053,8055,8054,1,14585,14590,14510,14522,14523,14506,14582,0,2,4,3,5,6,7,8072,8071,8073,8240,8245,8244,8239,8241,8242,8243,8303,8231,8096,8089,8233,8232,8230,8228,8297,8225,8299,8223,8224,8219,16,15,17,18,14,13,11,12,10,8229,8236,8235,9,8,87,1303,1304,1273,1274,1305,22701,22703,22699,22278,21314,21313,21312,21315,22702,22707,22706,22705,22704,22708,22724,23584,23583,22718,21226,21228,21227,21699,21698,22717,22722,22709,22501,22710,22721,22723,22725,23580,23579,22657,23514,1013,22730,1009,1008,966,1010,969,982,968,22711,967,981,1015,977,978,976,1011,961,962,959,963,971,970,1007,958,14313,14112,12124,12123,11931,12550,1021,14103,12551,1022,11929,11930,12564,12563,14094,12345,14095,12344,11695,11694,11693,12685,12684,12683,14087,14104,12420,12421,12419,14086,14088,13631,14089,14105,11700,12038,11698,11699,13955,11975,11974,11973,13630,13629,14064,14065,14062,11638,14063,14061,11639,11637,11665,12245,11663,14021,14043,11682,11681,11634,11635,11636,12112,11614,12774,12775,12776,12519,14007,12047,11781,11613,11612,12297,12230,11633,11628,11879,11878,11630,14031,14010,13992,11782,11783,13598,13597,13599,13970,11718,11719,12842,12209,11679,14322,14323,14331,12679,11724,11722,11723,11832,12707,12708,12709,11678,11677,12627,12625,14339,14361,14460,13993,11717,13994,14450,14449,22207,22208,22832,22498,22819,22497,20972,22083,20971,20973,21824,21311,20896,22270,22342,21386,20895,22271,22269,22877,22477,21388,22027,22026,22028,22476,22475,22474,22588,21375,22894,22895,22893,21376,21316,23307,23546,21038,21040,22095,22564,22565,22096,22917,22932,22180,22179,22181,22880,22879,22915,22370,22190,22916,22874,22191,22875,22863,22878,21429,22862,21387,22552,22593,22594,22536,21430,21428,21095,21050,21051,21052,21578,22535,22336,22873,22084,22085,22596,21971,22913,21970,22914,23553,23518,22926,22950,23573,22951,22970,23578,22952,22927,22693,22692,22189,22928,22496,22074,22491,22277,22276,22275,23001,23000,22907,22986,22075,23002,22985,21006,23026,21004,21005,21443,21444,21800,23031,23362,23596,22146,23593,21804,22589,22144,22145,22595,23604,23109,23531,23060,21688,22258,21939,21938,21940,20891,20892,20893,22236,22235,22234,23120,23118,22003,22470,22001,22479,23142,22322,23030,23575,23517,22503,22502,23541,23180,23179,23554,22364,23181,22691,22580,22511,22581,23568,22695,23224,22136,22135,22137,23223,23225,22055,23608,22896,22559,22057,22056,22020,22558,21238,23270,21335,21336,21337,21678,21679,21680,22656,21194,22685,22030,22029,22031,22402,22401,21535,23283,21026,21028,21284,21283,21282,21454,21452,21455,23237,22613,20991,21190,21189,21191,21958,23269,23257,23255,23268,23256,21453,21027,23282,23267,23266,21237,21239,21423,21424,23231,23226,23253,22398,22399,23254,21780,21781,21779,21792,21793,21490,23228,21554,21303,21555,23227,23229,23238,23167,21830,21304,21959,21828,21829,23277,22090,22089,22088,22591,22590,22047,20990,20989,22036,22019,22018,22017,20869,22037,21672,22226,21842,23299,21521,21817,21671,21673,21816,22487,21999,23313,23565,23297,23298,23315,23314,21606,22599,22598,22597,22281,23318,22629,23316,22716,23303,23302,21536,23301,22175,22174,21178,21179,21177,22616,22586,22585,22587,21193,21835,21834,21268,21952,21953,21192,21919,21171,21030,21170,21728,21875,21262,21727,21726,21874,22289,22288,23300,23287,22954,22955,13719,13718,13716,13720,14260,13770,13769,13771,22527,22528,23281,23280,23597,23262,23246,23247,23249,23265,13791,23264,13823,23248,23219,23250,23217,23251,13841,13075,13073,13888,23218,23196,23221,13889,23222,23198,23197,23199,12742,23191,23190,23172,23171,23163,23178,23173,23162,23139,23114,23138,23136,23135,23161,23160,23137,13965,12743,12097,13966,23113,23097,23591,21391,22631,22232,22632,23115,23098,23116,23099,21412,23140,21410,20958,20957,20959,23176,23177,23200,22302,23175,23195,23220,23215,23194,23216,23213,21141,21140,21139,23174,23164,22301,21933,21932,21432,23563,21433,21431,22504,21710,21709,21711,23134,22481,22480,23133,23102,23101,22296,22303,23093,23062,22107,21470,21353,22603,21738,21739,20953,23061,23055,21687,22259,20866,23054,20865,20864,21735,21736,21737,21749,22655,23036,23023,22689,21906,23025,21802,21801,21907,22906,22969,22982,21402,21401,21400,22367,22368,22999,22283,22284,22968,21066,22947,22948,22924,22949,22923,20875,20874,22945,21912,21067,21065,22922,22921,21911,22946,22941,22963,21484,21483,22980,22979,22488,22489,22494,22493,22492,22965,22966,23590,23206,22978,21471,21472,21149,22635,21151,21150,21493,22636,22533,22667,22996,22997,22998,23015,23014,23016,22669,22668,23012,23013,23040,23041,20951,20952,23042,23017,23008,22662,22663,22661,22642,22690,23367,22981,22282,22167,22994,21419,21914,21354,21915,23057,23056,23076,22788,22313,21440,21439,21441,22311,22312,23011,23039,22403,22404,22549,23059,23038,20106,19965,19964,19966,19602,19603,19604,20081,20080,20074,20731,20760,23010,23009,21286,23037,23043,23020,21371,23058,18232,18231,18233,20041,20099,21370,23019,23018,23007,22434,22433,23006,23004,23586,22991,23005,18963,22990,20082,18964,20039,20809,22989,22977,22976,22975,22974,22890,22891,22892,22960,22938,22937,22936,22959,20954,21413,20955,22905,22550,22551,22457,22314,21886,21220,21883,21884,21885,21901,21902,22458,23555,22856,22205,22523,22524,22525,22637,22308,22638,21600,21599,21598,22764,22763,22846,22845,21668,21549,22362,22363,22233,22206,21548,21550,21230,21231,21229,22825,22815,22814,22813,21552,21551,22570,22569,22568,21667,22619,22618,22816,22817,22607,22794,22805,22608,22795,22063,22062,22061,22606,22426,22160,22425,22424,23594,22462,21979,22463,22857,21936,21934,21624,21623,21622,21991,22859,22858,22870,22860,22015,22872,22871,23252,22889,22876,20873,22087,21913,22888,22887,21924,22910,22903,22908,22904,22909,22939,22592,22940,22967,22964,21219,21218,22961,22962,21033,21032,21034,22869,21921,21922,21923,23577,22466,22467,22465,21935,22911,22912,21969,22016,22086,22847,22849,22850,22337,22831,21648,21646,20968,20969,20970,22435,22007,22414,20894,22359,21745,21195,22808,22807,20903,20901,22298,21321,21323,21322,21397,21610,22799,22008,21609,21608,21647,22093,21064,21120,21119,21121,21062,21063,21301,22830,22848,21300,21302,21838,21837,21836,22792,22517,22152,22151,22153,22780,22758,22759,22791,22405,22754,22755,22756,22757,23540,21054,21053,21055,22044,22045,22046,22680,22679,22043,22681,22041,22719,22726,21908,22245,22042,22744,23539,22753,22625,22776,22624,22512,22789,22790,22793,22775,22760,22762,22778,22073,21812,21758,21757,21756,21813,21811,22774,21071,21072,21073,21074,22769,22784,22783,22801,22785,22786,22802,22804,22803,22583,22584,22582,22112,22113,22114,22810,22809,22811,22720,22529,22530,22800,23535,22782,22674,22448,22675,22770,22771,22752,22772,23534,22449,22751,22750,1289,22737,1290,22,24,1291,1292,1293,19,22738,22739,21925,22672,22673,21927,1294,1296,1295,1297,22622,22621,21926,22714,1298,22713,22715,1299,22712,1301,1302,1306,1300,1307,22731,22742,22743,22740,22741,22092,22072,22773,22734,22761,22615,21910,22244,22733,22623,22745,22732,21909,23550,21,20,23,8207,8213,8218,8272,8082,8290,8216,8270,8285,8212,8208,8080,8081,8199,8265,8204,8200,8203,8202,8104,8103,8102,8198,8193,8092,8194,8098,8058,8094,8292,8192,8278,8075,8277,8074,35,36,37,34,33,32,85,1280,1279,1277,1278,1281,20912,21785,21784,21782,21783,21021,21019,21020,21973,22823,22822,22821,22222,22854,22853,22212,22214,22213,21997,21109,21108,21110,22054,21341,22839,22415,22416,21340,22840,22841,22842,22824,22601,22826,22257,21364,22855,22866,22865,22864,19931,19930,20911,20734,20910,20732,20733,19877,19873,19876,18948,18946,22838,18947,18967,19086,18965,19505,18767,19506,19634,22833,22221,22220,1276,22834,1275,42,40,4276,19826,45,4275,46,18768,18769,19769,18990,19768,19748,19420,19419,19102,19731,19091,19090,19092,19715,19069,19716,19070,19068,4254,4271,54,4273,4272,50,48,49,51,53,4274,52,55,8044,8043,8086,8087,8078,8077,8079,57,56,58,86,61,59,60,62,8105,8106,8107,8149,8154,8152,8288,8279,8153,8280,8287,8286,8257,8298,8163,8100,8284,8099,8166,8161,8162,673,675,677,678,681,8172,8296,8050,8173,8171,8274,8049,8048,8169,8293,8065,8066,8040,8041,8039,8052,8160,8159,8101,8051,8164,8165,47,8167,8168,8268,44,8183,8067,8295,8178,8184,8186,8185,39,8275,8179,8188,8062,8063,8064,8263,8264,8190,8057,8189,8187,8045,8046,8047,8042,8056,8095,8195,8076,8196,8201,701,700,702,704,703,705,706,8210,8209,708,707,709,710,8271,8211,8214,8215,712,711,713,714,715,8220,8222,8221,8109,718,720,8108,8226,8234,8097,8088,721,722,723,724,725,8090,8237,8238,8038,728,729,8246,8093,8247,8250,8248,731,730,732,734,735,832,850,1172,1174,1214,1215,5290,5291,5312,5289,5260,5257,6992,6990,6991,6921,6920,6919,833,831,830,726,834,727,828,829,6917,6918,6850,827,7727,7728,6672,6668,6922,6987,6986,6988,6989,5224,5225,5194,5258,5226,5259,5227,5262,5261,5230,5263,5292,5267,5294,1217,1176,1216,1178,1218,5314,5313,5293,5319,1219,1180,1182,1181,1183,1185,14566,14567,14564,14563,14502,14500,14501,14503,14504,14561,14593,14594,14562,14556,14560,14565,5494,5496,5495,5497,5498,14595,14557,14555,14554,14515,14553,14484,14505,14498,14499,14550,5499,14545,14549,14494,14508,14516,5500,1252,5503,5501,5502,14507,14517,14543,5504,1246,1243,5505,14596,14491,14538,14477,14537,14479,14478,14492,14539,14542,14544,14591,14588,14547,14546,1193,1195,1194,1196,1197,1198,1199,1200,1201,14540,14541,1203,14473,14475,14471,14472,14470,14532,14512,14528,14529,14530,14527,14534,14535,14490,5506,1241,1238,5507,880,14526,14525,14513,20844,20846,882,20841,20835,20837,20836,20839,20840,20843,20845,20838,20847,14514,20848,20849,20851,20842,776,4300,20850,8620,8619,8621,8624,8622,8641,8638,8629,8627,8631,8630,885,890,888,889,891,892,893,894,8626,8541,8628,8540,8542,8534,8561,8636,8623,8617,8618,771,772,8614,8544,8545,8546,767,765,781,763,764,8608,8612,8607,8635,8634,8604,8609,8530,8602,8611,8610,8536,8616,8535,8554,8555,8531,8529,8615,901,900,898,902,903,8553,8560,8642,8603,8645,8598,8538,8597,8633,779,757,759,780,874,862,8537,8539,8606,8605,8613,8601,907,905,910,8599,8600,8595,8596,8590,8632,8591,8594,8589,897,896,899,923,931,932,2555,2554,2485,2669,2483,2565,2566,2611,933,934,935,2562,2550,2563,2564,2399,2369,2368,2370,2398,3829,2305,2234,2287,2286,2284,4143,2363,2362,2364,4000,2466,2468,2467,2469,2465,937,895,936,938,939,2463,2464,2460,2462,2461,3721,2365,3871,2366,2367,4207,8009,7977,7966,7976,7918,7949,4117,8012,7919,7917,4248,8011,7967,7962,7963,7964,8010,12355,12354,14173,14172,14156,14187,12356,14185,14186,14170,14171,14169,14168,12170,14149,14151,14150,14152,14153,14154,14155,14157,8027,8035,8033,8025,8019,8018,8008,8007,4222,3617,8005,8006,8016,7950,8003,8028,14140,14122,8004,14124,12257,8015,8014,8002,8000,7951,7952,7965,7939,7940,4133,7921,7920,3859,8001,8020,8021,7999,8022,7922,7928,3636,3635,3637,2056,3860,2945,2946,7929,8031,7997,8024,14102,14101,14111,14123,2057,2058,3982,3931,3957,3956,2140,2285,1622,3941,3942,4071,1621,1620,2283,2282,13726,13725,13617,13616,12172,12171,13602,13603,14182,14180,14191,14183,1247,14181,11737,1251,1250,1249,1248,1253,14179,14167,14147,14148,1245,1244,14109,14106,14184,14107,14108,14192,14193,1242,1240,13458,12086,8013,1239,883,881,886,887,4247,941,940,2549,884,8625,855,876,775,774,4298,4299,4303,4306,4305,4304,4302,4307,1235,5329,8373,8513,8322,8321,8320,8517,5318,8376,8521,8354,8353,8310,8352,8379,5460,5275,8380,8309,5456,8381,5426,8325,8323,8308,8350,17997,5245,5385,5378,5241,5243,5244,4467,5270,5271,5242,5272,5273,5240,5269,5268,5239,5295,5296,5297,5299,5298,5288,4465,4464,4466,5287,4416,4415,4855,4856,5411,5327,5315,5068,5324,1230,1229,5328,1232,5325,5069,5070,4501,4500,4499,5302,5316,5317,1233,1234,1210,1209,1208,1211,1213,1231,1204,1205,1206,1207,1236,14536,14474,14524,14531,14533,1212,1202,5301,5274,5276,5300,1228,1237,1226,1227,1225,1192,1224,5323,1189,1187,1190,1191,14551,14552,14485,14589,14558,14559,14511,14483,14518,14548,14493,5237,4301,20834,5488,1261,1263,1264,1002,1265,5493,1266,14486,14188,14574,14568,14572,1179,14573,1177,14577,1175,14509,5264,5232,5231,5265,5214,5204,5201,5203,5199,5202,5229,5228,5196,5198,5172,5195,5170,5168,5169,5171,5383,5173,4424,4423,4425,5421,5197,5339,5340,4317,5167,6994,6993,7001,6924,6854,6670,5848,6669,6927,6985,5618,5617,8269,8227,8217,716,821,822,835,820,819,6452,6515,6402,6455,6568,5823,5909,5908,5907,6569,6456,6572,7792,6458,6459,6457,6403,6349,6404,6309,6350,6346,6344,6345,7675,6771,6343,6268,6225,6267,5525,6306,6307,6269,5524,6271,6270,814,813,837,811,812,695,698,6226,6178,5523,6179,6180,6177,6223,6176,6224,6174,6137,6140,6087,6086,6136,6139,6141,807,808,689,690,839,809,694,810,696,838,697,693,692,8260,8182,8289,688,691,8180,8181,8176,8175,8170,686,685,684,687,806,840,682,8174,8177,699,8191,815,816,6305,6308,6310,6311,6348,6347,817,818,6401,6454,836,6453,6621,38,41,43,672,670,8158,8157,8156,8155,8151,8300,667,668,666,665,8301,8281,8254,8147,8262,8060,8059,8061,8146,8267,8266,8291,8150,8256,8255,8142,8141,8145,64,63,65,66,67,69,8140,8273,8133,8282,8136,8134,8132,70,68,4262,4263,19663,19668,19667,19674,19630,19672,19671,19673,18938,18941,18940,18942,4252,4251,4265,4264,4267,4266,19203,19202,18977,19687,19031,19030,4268,4269,19699,19698,19718,19029,19017,19018,19016,18072,18073,18074,18079,18078,18080,18939,19681,19682,19677,19676,19675,19683,19375,18109,18110,18111,19131,20818,19130,19132,18888,19259,19719,18263,19717,18407,19567,18409,18408,19566,19730,19089,18952,18953,19088,19495,18547,19765,18847,19792,18548,19820,18846,18848,19819,19852,19851,19875,19874,19821,19827,19823,18546,19749,18966,19899,18937,18936,19932,19933,18563,18561,18428,18427,18429,19022,18562,19023,19905,18675,19277,19951,18677,18676,18227,18226,18225,18587,18327,19950,18326,19975,18588,18586,20001,18785,18783,19999,19946,20015,20022,20016,20038,18784,22973,23562,22934,23592,20956,22900,22935,22899,22882,22883,22485,22484,22486,22897,22920,20822,19976,22957,22958,19977,18592,19278,20823,22898,22918,20793,22340,22339,22867,22844,20853,20852,20854,21996,22843,22868,22204,22341,22881,23211,20000,20036,20014,20037,20073,18617,18619,18618,19241,20098,19242,18851,20017,18871,18756,19045,19046,19044,19997,19497,18765,19974,19973,19265,18314,19935,18313,18312,18400,18401,18402,19235,18325,19269,19285,19881,19283,19286,19829,19250,19855,18061,19284,18062,18063,19377,19311,19830,19795,19816,19310,19817,18935,19818,19822,19397,19788,19815,19597,18671,18609,18610,19595,19596,18829,19752,18474,18476,18828,19251,19213,18611,19734,19761,19745,19733,19732,19760,19762,19746,19747,18265,18264,19729,19720,19260,20700,18612,18517,18613,18516,19735,19374,19268,19578,18872,18852,19385,20047,20048,18853,19545,19544,19546,20079,20104,19376,18241,19669,19659,19666,19632,19633,19660,18240,19661,19649,4270,4253,8127,8126,8138,8135,8137,8144,8143,8139,8302,657,659,655,653,8131,8123,8130,651,661,8148,662,22602,22812,22053,22143,23515,1283,1284,22781,22447,1285,1286,28,1287,27,29,30,1282,31,8037,8091,8283,8197,8205,26,8206,25,1288,22787,22406,22161,22806,21697,20855,20856,20857,21888,21889,22304,22670,22671,22766,22749,23598,22729,22735,22746,22747,22650,22210,22209,22211,22765,21986,21436,21437,21438,21396,22767,22768,21508,22518,21509,22297,23582,22796,21985,20902,22798,22797,21185,22544,14416,22543,22225,944,945,946,11673,14052,14067,14050,12735,14049,12574,947,11670,14073,1025,12567,11672,11671,13929,14074,14085,14033,952,14093,953,954,955,956,957,12836,14032,12165,12164,14075,12161,12162,12163,13749,11680,12265,12264,14076,14078,14077,12166,14060,11631,13928,11632,14059,14041,11629,1006,1005,960,1018,964,1017,1023,1019,1024,1016,1026,22736,20946,20947,22748,950,951,949,1004,1003,948,22779,22651,1020,22728,1014,1012,965,14011,14009,14370,22224,22223,22818,21601,23021,21285,21287,22995,21442,20156,4163,20155,20925,20926,20927,23095,8026,7933,7938,12098,12613,12611,13169,23094,21719,21720,21755,23096,12100,12612,12660,12659,14433,13949,13922,13924,14245,13944,13963,12071,12072,12073,13961,13945,13946,13962,13921,12661,12662,12663,13985,14005,11912,11909,11910,11911,11662,11616,11615,8036,11623,7960,11624,14058,7956,14019,14004,14001,14039,11660,14038,14017,14018,14003,13999,14000,11741,14262,14016,11923,11740,13300,11925,13982,14257,13981,13980,14412,11906,13960,13006,12206,11932,12204,12205,14418,14443,13977,12445,11588,11589,11859,11858,11857,13203,13204,13942,13941,13174,13897,13938,13918,13917,13915,13512,13511,12849,13936,12850,13937,12910,13976,11590,11980,13975,14259,13974,12687,11979,11978,12790,11604,11603,11605,13995,12788,12569,12104,12568,12570,14028,14051,11609,11610,11611,11647,13481,12929,12930,12084,14014,14034,12083,14015,11924,12116,14035,14054,11651,14055,11649,12118,12117,12843,12844,14036,11829,11830,7936,12870,14056,7988,7927,7924,7923,7925,7987,12582,12581,11742,14057,7986,14037,7985,11661,7932,7931,7930,7959,7984,7958,4073,4072,1538,1558,1569,1570,1571,3964,4241,7953,7954,4202,3950,3758,1784,1783,1785,1626,1585,1562,1563,1559,1537,1586,1557,1587,3938,1561,1560,1564,1565,1575,1576,3892,3729,1593,1588,1589,1590,3846,1583,1618,1633,1634,1659,1656,1617,1619,1635,1657,1660,3718,1597,3812,1347,1346,1348,3777,1647,3890,1654,1370,1655,4118,1371,1648,1643,1642,1758,1759,1702,1641,1703,1756,1697,1639,1581,1434,1698,1696,1579,1636,1580,1433,3057,3056,3716,3781,3782,1548,1435,1546,1547,1604,4233,1553,1606,1605,1610,1554,1552,1577,1598,1599,1600,1584,1582,1555,1556,1543,1528,1542,1530,1529,1514,1515,1513,1516,1517,1518,4100,3983,7969,7968,7970,7971,7975,7980,3838,1905,3839,3984,1544,1906,1545,4203,4162,7981,7982,7979,13964,12904,12903,7978,8023,7934,4161,3300,3302,11913,7983,7957,7955,1531,1525,7961,1526,1527,1533,1640,1372,1761,1766,1768,3719,4176,3811,4054,3687,1712,1711,1713,1771,1769,1770,4219,1735,1736,4200,3656,1503,1851,1501,1502,1846,1845,1847,1852,1908,1909,1984,1980,1979,1978,1983,1981,1982,4232,3752,3904,3778,1853,1791,1922,1859,1832,4102,1920,1921,1925,1923,1924,1968,1856,3946,3796,1969,1857,2035,2036,2037,3944,1858,3822,3881,3882,1497,4076,1677,1678,2318,2132,2133,2053,2052,2134,2189,2200,2201,2202,2203,2204,2304,2190,2270,2269,2277,2276,2359,2358,2360,2396,2279,2278,2231,2303,2229,2302,2395,2394,4206,2458,2455,2459,4029,4028,4061,2030,3786,3787,2031,3878,3879,3903,3594,3596,4039,3872,2457,2456,2610,2592,2547,2546,2545,2452,2593,2591,2665,2666,2667,2668,4226,4225,4230,20215,20216,20235,20234,20236,20237,20238,20269,20267,3654,20268,20287,4195,20288,18151,20289,20787,20309,20290,18152,20311,20312,19450,19451,20291,20270,18992,18820,18819,18566,18565,18077,18075,20308,20286,20307,20336,20812,20750,20749,20820,20830,20786,19494,19493,18076,18531,18405,18163,18406,18529,18530,18296,18608,20406,18591,19523,19522,19524,20408,19515,18590,19514,19513,20409,19254,18126,18124,20343,20002,19436,20777,18724,19020,19021,19019,20431,20450,19433,19434,19435,20461,18838,18396,19361,18968,20455,20454,20453,20465,20449,20417,20416,20415,19680,20246,20245,20244,19489,18857,18858,18690,18689,18691,18932,18504,19079,18862,18861,19204,18255,20475,20720,19484,19483,20497,20489,20352,18442,20351,20353,18572,20814,18971,20496,18261,18260,18259,18815,20512,20513,20511,20510,4214,4240,18497,18496,18498,3939,3940,3095,3094,3091,2819,3509,2820,4075,4209,4140,3763,3093,3092,3640,3641,3046,3068,4172,3038,3037,3036,4041,3764,3747,3065,3064,1354,2430,1624,1884,3954,3783,1352,1353,2837,3900,3688,2836,3689,3668,3666,3088,3063,1396,1395,1394,3098,3667,4060,4059,3795,3149,1392,1393,3932,4142,3177,3179,1416,1415,1414,3609,3610,3180,4024,3195,2121,3823,3197,3824,3958,3166,1714,1715,3169,3168,3167,3142,3204,3165,3164,3121,4198,3205,3202,3201,3200,3203,3952,4125,4215,4171,3707,3224,3226,3724,3723,3722,3223,3196,3194,3247,1613,1611,2120,4193,3260,3261,3259,3234,1612,3235,3236,3237,3222,3221,3219,3233,3190,3189,3191,3193,3192,3175,3861,2119,3178,3176,3162,3161,3160,3163,3886,1405,3955,1404,1802,3945,1410,1409,3159,1573,3139,1800,1801,3613,4023,3559,3611,3612,3951,4025,3918,3216,3607,3245,3246,3762,3217,3220,3258,3218,3244,3917,3270,2425,2378,3186,3187,3185,3994,4006,3560,4005,3158,3154,3155,3669,3156,3157,1574,1572,3885,3111,3140,3141,3112,1411,3114,3113,3115,1935,1934,1933,3645,1374,1375,1373,2942,3116,1400,1401,1402,3071,3070,3072,2943,1376,3069,3039,3030,1377,1378,3029,3028,3004,3003,2964,2965,3005,2976,3014,3040,3044,3016,3015,3734,4205,3042,3732,3681,3041,3031,3018,3017,3019,3733,3848,2979,2980,2981,2977,2978,2975,2929,2951,2952,2953,3805,2928,3936,3972,3973,3691,3937,1669,1667,3692,3776,2926,4196,2896,2895,2889,2857,2371,2373,4185,2858,2812,1500,2767,1498,4116,2372,1350,2766,2764,2157,2158,2156,4188,2894,2856,2855,2923,2890,2888,1430,2925,4216,3844,1431,1432,2971,3907,2920,1850,1848,1849,2948,2949,3011,3013,3012,2982,3672,2954,1423,2984,2983,2985,1424,1425,20434,20452,19064,3979,18934,18859,18256,3978,3977,3047,1451,4243,3067,18254,3986,4173,19574,18440,18441,1450,1449,4042,3020,3009,3698,1625,1623,2673,2674,2675,3010,3789,3035,1885,3034,3032,2838,3033,4120,3925,3006,2967,2969,2974,2973,2972,2970,4111,2924,2922,2944,2927,2416,2950,2419,2968,2418,2417,20433,20432,1915,1914,1916,2955,3921,20383,20258,20414,20413,20257,20705,2904,2903,2902,4092,2869,4234,18306,20357,20379,18307,18308,18161,18162,19516,20342,20341,20310,20339,3816,20340,2829,4064,4065,4013,2775,2772,2774,2771,2773,3720,2830,2828,2861,2862,3633,3634,2770,4157,2854,2811,2824,2810,2853,3992,2931,2932,2933,3618,2919,2899,2868,2900,2901,4201,4127,3854,2921,2930,2763,1351,2726,2724,2723,2722,2729,2728,2727,2662,1522,1524,1523,2607,2660,2661,2657,2609,2608,2548,1709,1708,1710,2354,2353,2381,2453,3825,2357,2355,2445,2446,2444,2543,2352,2348,2111,2110,2112,2356,2361,2454,2544,2351,2268,4227,2199,3632,2265,2345,1390,2342,1389,1388,2436,2435,2440,2441,2599,2542,2442,2443,2601,2605,2606,2653,2658,2659,2725,2655,2654,2644,2765,1349,1694,1693,1691,2646,2645,4019,2712,2711,2713,1692,1499,2341,1707,4190,18606,18607,18294,18295,18589,2047,2897,4053,4156,1872,1871,1870,2450,2451,2941,2272,2962,2963,2472,2471,2470,2958,2939,2618,2940,2619,2620,2912,2911,2913,2490,2271,2886,2885,2887,2914,2883,2880,2881,2884,4119,2803,2850,2851,2802,2801,3382,1603,2648,2647,3827,2846,2843,2882,2845,2844,2799,2878,1805,1807,2879,2874,2875,2872,2906,2908,2907,3653,1806,3880,2957,2956,3909,2989,2993,2995,2994,2999,2998,2997,2961,4151,3000,3001,3002,3619,3364,3363,3059,3624,3087,3062,3625,2866,3086,3082,3110,3079,3080,3081,3993,3108,3608,4007,3137,3135,3107,3109,3106,3134,3133,3105,3136,4009,3739,3138,3693,4008,4184,1442,3153,3182,1437,3183,1438,1436,1443,3213,3211,3210,3212,3675,2054,3184,3677,4084,3206,4052,3241,3239,4016,4217,18958,19038,19037,18667,19253,19252,18894,18895,18892,18890,18891,20561,19124,18668,3207,20040,4160,3152,20794,20707,19538,19211,20706,19429,20708,19210,18915,19085,19169,18211,19309,20491,19212,20532,20534,20533,20545,20821,19111,19609,20781,18772,19521,19209,19207,19113,19112,19503,19517,19393,20726,18250,19392,19501,19502,18470,19477,19476,18472,20507,20506,18266,20517,19208,18720,20758,18719,18718,18438,18208,18439,19496,20535,20526,20525,18602,20752,18392,18394,18393,18603,20763,19026,19028,18160,18159,18158,18503,18502,18501,18645,18748,18684,19421,18660,19422,18658,18657,18659,20514,19186,19185,19173,20759,19174,18215,19408,20500,20804,19678,19617,19616,19401,20806,18217,20724,19402,18271,18269,18360,18069,18361,18473,20716,18270,18397,18722,18721,18403,18071,18404,19177,18150,18395,18149,18148,18723,20003,20004,19014,19013,18808,19015,18931,19142,20412,20411,20410,19141,19143,19333,20430,20429,19424,19332,18883,20428,20403,18882,20404,20427,19298,20394,20389,20392,20390,20086,20387,20388,19569,18759,18753,18758,19610,19611,19612,19430,18755,19581,20376,18114,18725,18726,18727,20337,19764,18693,18692,18095,18766,20325,18093,18754,20348,20326,20709,19568,18791,18790,18789,20302,20334,19417,19418,20354,20355,20370,20369,20375,20374,20391,20393,20402,20421,19300,19299,19533,19532,20746,20448,19423,19425,20436,19137,18089,18868,18878,18088,18651,18653,18652,20435,20447,20446,20425,20424,18463,18462,18102,19449,20441,20442,20440,19257,19850,20457,19258,20466,20486,19849,20739,19180,20783,18747,18746,18745,19179,2751,2694,2692,2752,3331,20479,2754,4147,2755,3821,2637,2693,2758,2757,2794,2797,2798,2800,2795,2841,2840,2796,2756,2792,2753,2790,2793,18221,18191,20782,19196,19466,19465,19464,18192,19195,19194,18390,20831,20768,18103,20458,20715,19460,18279,18278,18280,18391,18338,20477,20476,20478,20480,20481,20482,20483,19490,20492,19904,19903,19902,18993,18994,18995,19024,18183,4067,4242,4244,2987,19192,18961,19193,20350,20725,2988,2986,3873,3021,18928,18927,18926,3710,3711,3712,18064,18065,20101,18902,18901,18900,20520,19684,20727,18342,3662,4159,3100,3099,3101,4058,3078,3660,3077,3052,3051,3053,3661,3748,3022,4192,3055,3054,4148,2721,2720,2717,3767,3768,3820,4121,3104,3103,3102,3497,1356,3131,3766,1355,2519,2520,3151,3150,1357,20553,18581,18580,18582,20593,3130,18343,19399,19398,19400,20714,20713,19394,20617,20509,20100,20508,18962,19390,19391,20349,19879,19547,18814,19880,18131,18130,18129,20488,20487,18482,18249,18248,18471,20493,18706,18481,20495,20499,20494,19282,18705,19281,19201,18864,18267,18268,18863,19371,20473,18678,19432,19027,19535,18216,18679,18197,19640,19492,18196,18198,19491,18070,20807,20803,18347,20740,20485,20474,18707,18348,18349,20464,18620,20463,18762,18761,18930,18622,18621,18842,18843,18870,18869,19714,18867,18087,18201,20773,20774,19504,20462,20828,20766,18480,18199,18200,18337,18579,19008,19072,18786,2055,3181,2718,2719,3058,2867,4149,2991,4154,3025,3024,3026,2992,2990,2553,2552,2551,3238,3841,3858,1459,4012,1458,1460,3888,2937,2938,2910,3639,4186,2909,2876,2934,2877,2870,2873,2871,2791,2842,2839,2905,2449,2448,2447,19607,19442,2825,2826,2827,3023,3910,3027,18104,19395,19396,20400,20373,20764,20398,19583,20372,20399,20423,20368,18959,20371,20329,20335,19416,20328,20331,20332,20333,20305,20283,20304,19151,18635,19150,18778,18779,18780,20255,19615,20256,20284,19614,19543,20253,19542,19541,19198,20250,20254,20281,20249,20248,18805,18806,18807,20222,20224,20223,20279,20278,19605,19606,20227,20226,20230,20228,20229,20519,20208,18553,18555,18444,18443,18445,19367,19366,20191,20192,20712,19555,19554,19556,20064,20190,20189,19168,19519,19520,19518,20699,20133,18465,19153,19152,19154,20114,19362,18426,18425,18424,20789,20136,20137,20134,19363,20142,19580,20141,20140,19410,20139,20116,18068,18066,20135,18700,20113,20338,20065,20072,20071,18998,20067,20066,19228,18881,18880,18879,19227,19368,20810,20811,20035,20795,20034,20033,4098,20736,20069,2387,18996,18997,20095,20070,20094,2473,18351,4094,2389,4021,2474,2032,2390,3898,2392,2391,2393,2227,3863,2228,3853,3894,1419,1418,1417,3953,4189,2388,4137,4097,4096,2293,2377,1973,4099,3198,3199,3915,3694,4068,1974,1975,2217,4221,3832,4080,4003,3852,3991,1495,3856,2139,2138,2137,2135,1676,2136,2230,3790,1496,2076,2075,2074,2070,2071,2072,2073,1989,1990,2154,2155,1985,1986,1398,1399,2153,2149,3701,2210,1447,2211,1448,1446,3623,3963,3968,3908,4074,3791,3695,3696,4020,2148,3638,1505,1506,3682,2290,2291,1873,3966,1504,3867,2218,2212,1467,1468,1469,1874,1732,1733,2159,2161,2160,2213,3620,3621,2067,2066,2065,2150,2151,2152,1397,3905,4032,3911,4030,1994,2068,1995,1510,1511,1512,2059,2162,2219,2220,2221,2222,2226,2300,2224,2181,2225,2163,2180,2182,2178,2098,2078,2080,2280,2281,2077,2164,1996,4178,4139,3697,1919,1907,1854,1913,3684,4150,1787,4194,3769,1731,1734,1729,4208,1674,1786,4164,3770,1673,4231,4175,4218,3771,1730,3914,1658,1680,4229,4166,1630,4165,1592,4001,1690,4002,1632,1631,1688,3788,1742,3913,1679,3755,4017,3754,1675,1747,1748,1749,1804,1803,3813,3899,3922,1940,3971,1868,3970,3760,2017,3759,4153,2014,2015,2016,2018,2093,2094,2172,2023,2096,2174,2173,2248,2320,2319,2247,2316,2245,2244,2315,3663,2313,2243,2170,2242,2166,2241,2314,2312,4197,2500,2403,2501,2498,2497,2496,2567,2569,2499,2106,1993,1992,1991,2105,2236,2237,2240,2402,1861,2239,2238,2167,2169,2171,2092,2013,1938,1936,1681,1683,1869,1867,1743,1741,3680,3686,1689,1687,1685,1684,1686,1740,1737,1739,1440,1439,1441,1727,1726,1728,1738,1725,1672,1722,1721,1720,1776,1775,1777,1779,1839,1844,1843,1792,1780,1793,1795,3794,1794,1796,3679,1549,1551,1778,1724,4136,1719,7989,7990,7943,7942,7941,7972,7995,7994,7996,8029,7993,7992,7991,7935,13487,11653,11654,11652,14072,14070,14071,11650,14068,14069,11887,11888,11886,12945,12301,12067,11744,11743,11745,12302,12303,11689,13092,13091,11658,11593,11592,11591,11659,11669,13957,13956,13958,14027,12099,12021,12022,12023,14045,12105,14025,14026,14044,14008,12832,11664,13347,12820,13346,12821,12686,12688,12789,11775,11776,11777,13990,13349,13954,13953,13777,13952,13933,13967,13934,13927,11587,14372,11585,12479,13969,13968,12049,12048,13986,14006,13989,13988,14426,12833,12478,12678,11586,13925,13127,12157,12232,14356,13926,12936,12320,12442,12920,12444,13930,13931,13932,13972,14310,13971,12694,13776,12935,12934,12696,12695,13510,13973,13913,13901,13879,13878,13877,13876,13874,13873,13872,13900,13898,13896,13899,11784,13880,13881,12309,11991,11990,11992,13852,13853,12415,13515,13875,13516,13517,13991,14383,12603,13871,12605,12604,12158,12443,13895,13891,12321,13892,12159,12160,13847,12583,11899,11900,14209,14226,14225,11823,11825,11824,12012,12739,11901,12740,12288,13846,12377,12378,13831,12447,14334,14261,13828,12287,12562,12561,12560,12973,12972,12974,12448,12449,13832,13813,12518,12517,12516,13833,13849,13835,12296,13811,13854,13850,13851,13848,13830,13810,13809,12294,12295,13834,12414,14409,13815,12916,12045,12042,12046,12912,12746,12522,12043,11802,12382,12381,12383,12745,12747,13787,13786,11582,13764,11583,13783,13785,13781,13816,13814,14358,12254,12014,12255,11690,11691,11692,12013,13808,13779,13812,12179,12180,13778,11837,13756,13202,14233,14232,13758,13755,12803,12802,11836,11838,12396,12395,12397,13733,12744,12465,12433,11998,12483,12484,12485,13699,12000,11999,13732,14376,13728,13731,13729,12141,13730,12030,12029,12028,14202,12142,12140,14309,14295,23085,23086,23087,23071,23070,23050,21977,21976,21978,22184,23072,22413,22664,22412,22411,21167,23091,21612,21611,21359,21168,21169,22169,22065,21498,21500,11803,11804,11805,11889,13667,23521,11971,11970,11972,12026,12027,12771,11951,13639,12558,12804,12283,12282,12281,12976,12557,12977,13608,12737,12736,13609,12601,13610,13613,13612,13611,11988,13640,13644,13669,11989,11987,12360,13700,13620,13701,13619,11960,11959,11958,13698,14402,12770,13666,13618,11917,11918,11919,13704,13703,13735,13706,13702,13736,12401,13737,13738,12527,12525,12526,13330,13329,13328,13784,13762,13763,12607,11584,12606,12772,13790,13819,13789,13788,11914,13822,11915,12313,14345,13744,13765,13750,12417,12418,13773,12314,13827,12311,12310,12312,12512,12305,12304,12306,13056,13826,13797,13798,13799,13824,13128,13825,13844,13615,12956,11916,12957,13614,13840,13766,13862,13863,13767,13864,13865,13309,13886,13887,12907,12347,12346,12348,12494,12493,12495,13987,13947,13948,13818,13817,12741,13907,13026,13025,14285,13842,13890,13843,12699,12698,12697,13074,13793,13794,12520,12933,14453,13795,12521,13807,14214,12183,12184,12185,13866,13024,13885,11920,11921,13796,13748,13747,13746,11922,11948,12773,11949,14348,13689,13688,13690,13398,13397,13399,13687,13686,13685,13684,13683,14422,13654,13651,13655,13650,13503,14250,13662,13663,13664,14249,13595,13594,11875,11876,11877,13502,12950,12951,12952,13894,12657,12658,12057,12056,12058,13593,12544,13571,12543,12545,14326,21007,22246,22250,12825,21744,21743,21742,21604,22154,22155,21607,22280,22279,20888,23570,20889,23334,23335,23336,22022,22021,22614,22315,21799,23326,13665,23340,23346,13638,13604,13637,13661,13696,13697,13717,23327,21798,23339,21263,23317,21264,23347,20890,21603,21605,23261,21101,23332,21487,21099,23337,22611,23325,21855,21854,22612,22060,22059,13457,13480,13478,12508,12507,12506,12959,12960,12877,13443,13444,13418,13416,14395,12876,12878,12879,14200,14199,13386,13413,13412,13414,13415,13438,13439,13455,14265,14417,13456,11726,11725,11727,12509,12196,12964,13114,13498,13499,12915,13530,13531,13212,13501,13479,13500,13213,22295,23338,22777,12826,12913,12914,13772,13521,13520,13497,13549,13522,13496,11754,13523,13524,11806,11807,11808,13550,13551,13572,13573,13596,13575,13574,14335,13576,11961,13553,11835,11834,11833,14410,13628,14415,14444,13652,13504,13653,13221,13656,12888,12887,12885,12937,13632,14343,13693,13658,12886,13566,13694,12384,14189,12765,13724,12425,13692,13722,13721,13723,13751,13691,13219,13220,13657,12424,12511,13708,12416,13727,13635,13710,13743,13742,13709,13715,13741,13740,13357,13712,13739,13713,13711,13707,13676,13673,13671,13677,13678,13679,13680,13714,14411,13682,13675,13674,13636,14427,13623,13624,13622,13649,13627,13592,13621,12859,12858,12856,13645,13646,13586,13647,13587,13588,13670,12114,13672,12113,12115,13668,13643,13642,13641,12857,14239,12750,12748,12749,12600,12602,12999,13000,13563,13561,11946,13562,13541,13538,13545,12192,11843,12193,13544,13543,13542,12926,12925,13508,13509,13514,13513,13570,13565,12332,13548,12781,12333,12780,12779,13568,13569,11667,11668,13590,13589,13591,13625,13626,13660,13634,13633,13648,14300,12386,13695,12385,13659,13601,13577,11962,13600,13567,11884,11883,11885,13546,11963,13528,13529,12001,12597,14217,13519,13518,14231,11996,12598,13464,13463,13462,13465,11967,11968,11969,13466,13450,13492,13493,12155,12154,12153,13449,11731,11732,11733,12715,12714,12961,12962,13452,12648,14390,11997,11762,13429,11760,13426,13427,13428,13403,13451,12875,12874,12505,13760,12039,12040,12041,12242,12243,13395,13396,12862,14198,12504,13274,13273,13275,13394,13393,12175,12178,12177,12239,12240,12241,12121,13356,13373,13374,13375,13343,12726,12596,12595,12594,12122,11926,11928,11927,13321,13318,13319,13320,13294,13288,13267,13272,13287,13289,13268,13290,14338,13322,13324,13325,13291,12300,12299,12298,13270,13269,13292,12498,12497,12496,13276,12358,13293,13117,13118,13119,13238,13239,12793,14012,13216,13215,12452,14098,13278,13243,13244,11880,13279,11882,11881,12880,12881,14287,14401,14430,14210,14040,11868,11866,13332,13331,13333,14424,13305,13307,13350,12330,11867,12380,14344,13358,13380,13336,12318,12319,12134,12135,12136,11896,12902,12901,11787,14284,14283,14379,13298,13131,13132,13133,14380,12565,13348,11819,11818,11780,12566,11778,11779,12689,12690,12738,13430,13407,13431,13404,13405,13377,13681,12515,13152,12514,12513,13151,14321,12725,13401,14419,13400,13402,12268,12266,12267,12851,13406,11761,12852,12055,11995,13467,12002,13472,11656,11657,12054,12053,12818,12819,13471,13468,14461,13433,13432,13434,13469,12854,12855,13453,13435,13437,13454,14311,14312,13474,12275,13473,12853,12277,13470,14255,13526,13525,13527,13495,12276,11752,11753,13359,13476,13477,12197,14413,13475,13440,14314,13441,13442,14328,14420,12717,13436,14384,13408,12728,12718,14269,13409,11788,11789,12729,11898,11897,13378,12716,13379,13381,13363,13411,13410,13385,13382,13383,13384,12763,12762,11937,13360,13361,13362,13365,14270,14271,14346,13364,14437,13417,14436,14347,22260,21621,21620,22186,22187,20987,20986,20988,22407,22556,21777,21775,22984,22983,22987,23564,23380,23588,22285,22286,22287,21890,21891,21367,23396,23395,23385,23384,23397,21125,13369,21107,13370,11755,13371,13387,11756,11601,11602,11600,14445,13337,12331,11936,11935,13306,13338,12682,13419,13420,13301,14079,14080,12680,12681,13314,13315,13316,13339,12371,12372,14223,14222,12529,12365,11705,12366,13312,13313,11620,11622,14364,23407,23408,21900,21863,21899,14330,20999,20998,21000,21862,22515,22421,22521,22423,22422,21308,23401,22171,23400,23506,21297,23507,22024,22058,21298,20898,20897,20899,21310,22345,22344,23398,23544,22350,21188,21982,21981,21951,21187,21186,21930,21931,21768,22600,23542,23377,21344,22567,21324,22566,21158,21157,21156,20905,20904,20906,21325,23106,21153,21154,21155,21275,21274,21769,21296,21299,22620,22519,21366,21365,22025,22023,21950,21871,21918,21916,21917,23516,13368,13367,13366,14381,14299,14221,11704,11706,14342,22563,21872,21873,22034,23409,23399,21713,21068,23372,23360,21069,21070,21582,22033,21135,22032,21134,23361,21571,21343,21342,22267,22820,23382,21084,22377,22378,21085,23381,22351,22318,22317,22316,22379,23388,20933,20932,21858,22320,22319,21661,22380,21083,22219,21117,22993,21787,22218,21740,22217,22066,21990,21789,23319,21516,21515,21517,22644,21725,22643,21851,21852,21853,21724,21164,21786,21166,23368,22472,23359,22091,21089,21091,22471,22164,22165,21045,21046,21044,21206,21205,21204,23351,22548,23556,23558,23349,23350,14386,14387,13335,12017,14319,23519,23529,21409,14320,21213,21214,23366,21645,21425,22483,21090,21426,21427,21165,21643,21644,21056,23549,22128,22251,23375,22373,14463,12446,11772,11791,12655,22120,11774,12656,12016,12015,12575,12576,12323,11795,11645,11644,11643,11796,12194,12036,12195,13271,12580,12579,12791,12534,12792,13266,12535,12367,13868,13241,13240,11790,14288,11773,12128,12129,12130,12671,12202,12665,12669,14206,12668,14230,13184,13183,13180,13185,13186,12769,14394,14392,23209,23208,23210,23240,23293,23389,22147,22149,23532,22241,22242,23239,23241,22126,22127,21058,22202,22201,21057,22240,21459,21461,21460,21406,21405,21013,21349,21348,21347,22163,22162,21788,22309,21660,21659,21857,21558,21856,21556,21807,22538,21805,21806,22109,23392,20931,22110,22108,22183,22134,22182,21215,21216,21217,23088,23089,23406,21987,21820,23414,22139,21505,22138,22070,22123,21506,22540,22125,21092,22513,21716,21694,21715,21714,21507,21810,21640,21642,23595,21821,21819,21641,22371,22514,22168,23405,23404,23403,21625,21627,21626,21011,21583,21012,21584,21123,21122,21124,21568,21686,22148,21570,21569,23077,13158,13157,23402,14292,22553,22516,20994,20993,20992,22534,21635,13116,21617,14291,12795,14400,12861,12796,13141,12620,12807,12860,12189,11976,11721,11977,11702,12392,12619,11701,11703,12827,12070,14278,13038,12634,13057,12653,11895,11894,11893,12651,12652,13182,12990,12991,12290,13214,12730,12732,13181,12667,12786,13028,12664,12368,12201,12203,12666,12654,13035,13034,13036,12635,12768,12767,12794,12636,12369,12621,12787,13257,12441,13258,12909,12938,12672,12673,12674,12731,12733,12734,13914,12221,12814,11981,11982,11983,12222,12223,13242,12292,12291,12293,13217,12227,12459,12228,12229,12987,12350,13159,12351,12349,13160,13161,12963,14306,12828,14238,14304,12081,12631,12554,12556,12555,13162,12325,12326,12324,12316,12315,12317,12649,12805,12806,12308,12702,12307,12126,12125,12703,13245,13246,12906,12905,12864,12289,12127,12865,12018,12019,12020,12137,13265,13280,13281,11764,13256,13264,12391,13255,13286,11765,13285,13262,12390,13263,12145,12144,12143,11943,11942,12724,12482,13249,12829,12830,12831,14303,13311,12200,12199,12198,13284,12528,11993,12530,12370,12477,13950,12150,11763,12151,12152,11621,11994,13105,14329,23417,23416,23427,21752,22572,23437,21965,21753,13283,21963,22439,22372,21434,21161,21162,21163,20962,20961,20960,21289,21602,21288,22450,21382,21381,22203,23425,22071,21637,22461,21420,21421,21422,21590,21591,21196,21197,22388,22389,23576,21476,21241,21240,22140,21904,21905,22451,22237,22238,22239,22384,22141,22142,21309,23512,21587,21892,21097,21096,21098,21762,22122,23426,21803,21174,21175,21176,21964,22121,23415,21588,21589,21636,22658,22124,23435,21633,21632,21634,23075,21180,21093,21094,21948,21968,21966,21967,23035,23379,23378,23434,23422,23423,23424,21695,21696,23412,23413,22035,21495,21494,23411,23418,23419,22560,21876,21877,23312,21618,21928,21929,23311,23370,14350,12549,13115,23429,22428,23432,23433,22006,13261,12481,13282,12389,12752,12052,13254,14396,12263,12050,13237,14403,12065,12863,13187,13218,12882,12883,12451,12450,13260,12908,12439,12440,12650,22307,21557,21988,21989,22456,22697,22455,23383,22252,14465,12284,12670,12285,12187,12404,12322,12405,12461,12460,13334,14267,14307,14308,23589,22506,22505,23342,21048,23341,22150,23022,23320,21049,21047,21597,21491,21492,21596,21572,21339,21791,21338,20921,21224,21790,21328,22686,21327,21631,21023,20981,21024,21025,20980,22081,22082,21980,23330,23331,20913,20915,20914,21586,23344,21581,22052,22231,21133,22067,21003,22049,20919,22050,23352,23355,23356,20920,22166,22459,23357,21002,23358,21001,22338,23501,23545,22634,21741,21118,22266,23369,23511,23348,21225,21585,21567,22173,21847,23345,22648,22666,22649,22409,22410,21102,21103,21104,22195,23324,22438,22437,22436,22727,23295,21832,21831,21833,21523,21522,22610,21394,22361,21393,22194,21547,21546,21545,21670,21669,22432,21577,23353,21295,21138,23467,21575,21576,21137,23566,22077,21841,22522,21159,20908,22464,20909,21136,21106,21105,21682,21681,21294,21293,22105,22628,21077,23505,23354,22193,22617,22106,22256,22247,22192,22248,23600,21683,23374,22290,22468,21619,22064,21840,21278,21776,20907,21160,21485,22453,21486,21079,21078,21100,23333,22268,21998,21392,21808,21809,22014,21565,21566,21846,21839,21390,21843,22499,22294,22500,22333,21844,21845,22079,22078,21564,21389,23276,23286,22418,22417,22264,21540,21541,21538,23259,22263,22265,21497,23560,21318,21319,21320,21496,20884,23235,21867,21866,21865,22469,22327,21183,23234,21184,21272,21945,23258,23232,21944,23329,23274,23273,23543,23260,14253,14333,14332,14195,23284,14368,23275,23289,23288,23290,14428,14196,14263,13389,13388,13390,14363,22198,21960,21962,23605,23192,22119,22117,23292,23291,20967,23285,21152,20965,20966,23271,20882,22441,20883,22442,21707,22654,23343,21553,21708,22478,22383,21825,21826,21827,22382,22510,22509,23321,21326,23322,23323,22688,22687,22609,20982,23310,21144,22358,21143,21142,21562,21563,21937,21859,22249,21818,21595,21559,21560,21561,21961,22526,22520,22118,23294,22197,23328,14352,14397,12339,12340,14201,12338,13352,13351,14442,12106,13355,13353,13354,12108,12107,12210,12212,12211,13317,12176,14240,12119,12007,12120,13372,12063,12006,12008,13422,13421,14204,14205,14194,14268,14272,14273,13447,14371,13423,13392,12411,13391,12409,12410,13445,13425,13424,13734,13461,12503,14247,12408,13505,11955,11956,11957,12837,13448,14203,13446,11598,13459,13460,11597,11596,11595,11594,13483,13484,13485,11766,12577,11768,23536,23207,23204,23205,21182,21273,23233,23236,23187,22326,22942,22944,23170,22431,22429,22430,23188,14290,13582,14289,13581,14407,11767,13533,13556,13555,13534,12376,12375,12374,12379,13578,13580,13579,13607,13560,13559,13535,11947,13536,13532,12208,13506,12713,12207,13537,13507,13491,13489,11844,11842,11945,13490,13488,13486,11599,12559,14266,12131,11952,12133,12132,14374,23150,23152,23151,23154,23149,21499,23130,23153,22177,23157,23581,23129,22532,23155,22039,23156,22040,22038,22531,21360,21248,23112,23111,21250,23127,23128,23131,23132,23084,21628,21629,21630,23148,23147,23126,23165,20950,23183,21270,21269,21271,21146,21145,21147,21488,21489,21765,21766,21767,21778,23083,23159,23158,22310,23169,22508,23185,23574,22176,23189,21203,21202,21201,22360,23193,22228,22227,22229,23166,23168,23203,23186,22507,23202,22943,22178,22829,22400,22460,23201,23184,22292,21594,22291,21593,21592,21983,21984,23230,21377,21378,21379,22365,21771,21770,22366,20922,20923,20924,23182,21042,21043,21059,21076,21075,23143,21897,21896,21210,23122,23105,23121,21211,21212,23104,23079,23080,23065,23064,22573,23063,23048,22626,22577,22627,21638,22216,22215,21266,23078,21265,23108,23107,23103,22828,23119,22002,21898,22321,21395,21267,23044,23045,23046,23567,23029,21639,21017,21501,22159,21245,21246,21746,21333,21747,21690,21689,20929,20930,20928,21018,21022,22578,21016,22579,22576,22630,22575,22574,23049,23068,23069,23081,23548,22170,21208,23066,23067,22886,21955,21954,21956,23082,23508,23124,22884,23125,23146,23144,20948,20949,23145,22393,20880,21060,21061,20881,20879,20876,23123,20877,20878,22885,23092,22172,21249,23074,23090,22665,23073,21513,21417,21514,23551,21209,21207,21418,22097,22098,22992,23587,23034,21502,22396,21504,21503,21614,22051,21233,23027,21232,21234,22352,21014,21615,22133,21748,22325,22571,21009,22971,21008,22324,22323,22972,22254,22255,22988,21543,23364,23363,23365,21450,21451,13558,13845,13557,11864,14406,21449,22253,22956,22068,14327,22851,22677,22676,22678,22094,23557,23110,22933,21462,21463,21464,22652,22274,22653,23376,21949,21399,21332,21334,21010,21616,21015,21247,21398,22273,22369,22440,22076,21236,21235,23028,23572,21041,22069,14373,14458,14459,14365,11828,11827,11826,12677,11863,11865,13869,11860,11861,11862,11618,11617,11619,13829,12286,13805,13806,13804,14389,14388,13802,13803,13801,12899,12898,12900,12811,12810,11707,13774,12759,12758,12757,13753,12217,12431,12434,12432,13752,13775,11709,11708,12234,12235,12031,14435,14355,23522,22185,22343,21613,23033,23032,14354,11713,14466,14425,21718,21717,12149,22925,21542,21544,12148,12147,23561,12218,12216,13979,13757,13978,14280,13754,12896,12816,12815,12817,11817,13867,11816,13908,13909,13910,14340,13911,13912,13935,14456,14457,12626,22605,22102,22101,21957,21662,21374,21663,22852,21822,21823,22861,21887,21674,22495,21317,14367,14366,21039,13893,14362,11831,11815,12156,12233,22645,23245,14252,14254,14317,21539,14378,11655,13552,13554,12727,14408,13344,14369,13376,13327,12359,12357,13345,13326,13323,13295,13296,13297,13299,13277,14337,14207,11666,11674,11675,11676,13585,13583,13584,13564,13745,11950,12932,12931,13792,12096,14215,13923,14318,14216,13906,13920,12342,13905,12341,12343,13882,13904,12406,13857,12407,12146,13916,13919,12813,14399,13705,12463,12462,12464,13940,13939,12700,11736,11735,11734,12704,12701,11904,13983,13984,14002,13959,11905,11934,11933,13173,13175,13903,13902,11786,11785,13870,13856,13855,12044,12510,11801,11800,12362,12361,13838,13837,13858,13860,13859,13861,13768,13839,13821,13820,14341,13308,13310,13884,13883,13761,13759,13782,12402,12403,13780,14441,14030,14029,14046,14047,14048,11646,11648,13482,14053,14066,12472,12471,14096,12470,12469,11687,11688,11685,11686,11696,12093,14097,12335,12334,12336,12337,12037,14438,14114,14129,14128,14022,14127,14212,14143,14144,14162,14161,14126,14023,12005,12003,12004,13540,13539,13800,13059,14125,13060,13058,14024,12435,14113,7937,7926,8030,1717,3757,3756,3843,1723,1718,1671,1670,1629,1628,1627,1475,3728,1591,1473,1474,2246,2175,2251,2250,2325,2252,2323,2407,2409,2507,2508,2408,2575,3646,1607,2577,2632,2633,2635,392,19713,19728,20819,19696,20817,19697,390,391,19686,2410,2327,2299,2253,2254,2326,2324,2176,2177,2179,2097,2249,3893,2095,2024,2022,1943,1939,1941,1942,1917,3563,2025,2026,2079,3580,1918,3995,3685,1855,1788,3655,2292,3833,20013,18067,18701,18350,20115,19411,19412,20138,20172,20171,20173,19296,20175,20174,18835,20742,20741,18554,20212,20211,19337,20213,20214,20232,20231,20233,18436,18245,18437,18247,18246,19364,19365,18969,18486,18970,20298,20299,18897,20327,20324,20301,18898,18896,18899,20300,20303,20306,20285,20280,18699,18094,20261,20260,20259,19763,20266,20264,20265,20263,18991,18945,18944,18943,19338,20262,20217,19276,18836,18834,20197,20196,19297,20195,4031,2663,20194,4093,3595,2664,18153,18803,19445,19355,19354,18713,18714,18712,19187,18464,20068,19240,19239,19461,18737,18262,19549,20052,18466,20053,20169,20170,20131,19641,20132,20160,20183,20182,19167,20186,20185,20184,20161,20129,20162,20163,18177,19222,20164,20130,18461,18410,18454,20127,20128,20111,20019,20021,19642,19643,19548,20093,20062,18452,19553,20063,18059,18594,18058,18060,18738,19403,20031,19404,19689,20032,20030,18595,18451,19526,18453,19991,19447,19446,19448,19525,18849,19766,19988,18411,19109,19327,20165,20166,20167,18574,18575,18299,18179,18297,18298,18178,19050,18643,18922,18923,19664,18333,18331,19318,19317,19319,20251,19530,20252,20247,19529,18687,18688,18686,19289,19290,19291,20225,19570,19197,20206,19351,20187,19352,19166,18332,19065,18874,20202,20188,19994,20168,20159,20158,20157,20797,20761,19313,18634,18633,20323,20330,20775,20366,20365,20801,20363,19472,20796,18795,18796,20824,18797,18960,19473,19471,19104,20362,20356,18112,20377,18733,20378,20382,20407,20380,18732,20381,18841,20405,18125,18839,18840,18113,18804,18837,18484,18483,18485,18506,18974,18505,18624,18623,18625,20484,19073,19007,19006,19004,19005,18570,20504,20490,18860,18571,20096,18972,18792,18793,18794,19488,19218,20097,20516,20515,18381,18345,18346,18344,19217,19003,18637,18638,20808,20799,20518,20723,18911,20747,20748,18912,18910,20523,20521,3174,4238,20522,4245,19224,18950,18949,18951,20765,20539,18925,18636,19462,19463,20538,20537,20745,20771,20531,20544,20540,18536,19459,20549,18286,18388,18389,18387,19458,18492,18491,18490,19510,19509,19511,18728,18535,18924,18539,19247,18277,18276,18275,20555,3780,20554,4091,18650,4239,18649,3232,19066,18538,20755,18537,19067,1345,20536,1343,1344,3173,3172,3120,3170,4182,3231,3229,3122,4212,3230,3227,3228,2100,3253,3252,3800,3256,3254,3257,3255,3862,1444,2099,2101,3143,3171,3145,3144,3129,2818,4213,20829,3125,3124,1406,1407,1408,4141,3818,3123,3119,3117,3066,3118,1998,1997,1999,3717,1716,3188,20562,20556,20563,18184,18185,18186,19312,18717,18716,19320,19571,18630,19133,19134,19165,19164,20550,18632,18631,18903,20551,18904,18865,20560,19147,19146,19145,19100,19099,19101,18556,18557,18558,19155,19234,20529,18889,18511,18285,18512,18284,19853,19854,19233,18905,18770,18771,18685,20770,18683,20530,20193,18293,18272,18273,18274,20543,18340,19206,19205,19273,20542,18341,18827,20548,19114,19274,19263,19575,18234,19576,19389,19262,19115,19116,19577,19384,19383,19381,20547,20546,19292,20565,19293,20564,20558,19288,20711,19107,19216,19953,18957,18669,18670,19215,19214,19108,18236,18235,19255,20571,20502,18664,18336,18335,18334,19121,19119,19120,19261,20552,20559,18584,18583,18585,18569,18644,18866,19437,20566,20567,18647,20557,18914,18913,19382,19001,19071,18210,20541,18209,18207,18628,18627,18629,19002,18826,18339,20722,20527,20788,18292,18749,20744,20524,19386,3631,3214,4204,3215,2379,2959,2960,3670,3969,3008,3007,3043,3045,3076,3096,3097,3765,1391,3128,1403,1789,3751,14213,12085,13997,13998,14013,13996,21691,23100,22397,21355,21868,21411,21456,21457,21458,22408,23141,22490,22696,23243,23212,23244,21408,22375,23214,21111,21031,21029,21407,21972,22293,23242,21920,22374,23278,23279,22376,21537,22659,21773,21774,22243,20867,22660,23296,20868,23496,974,973,979,980,986,985,987,988,1000,1001,22698,1272,22700,1271,401,457,407,404,397,408,19665,19709,19846,19845,19049,19047,19331,19778,19539,19779,19756,19540,19755,19739,20146,20178,20177,20179,20198,19742,20438,20422,20456,20827,20444,8258,5718,5960,5961,6020,6016,6018,6042,5656,5655,5657,6019,6013,6037,6038,5902,5901,7796,6108,6106,7845,6105,6107,6150,6149,6189,5999,5799,6331,6130,6211,6209,6249,6248,7821,6210,6208,6244,6247,6323,6324,6381,465,6251,7836,6327,6326,467,466,469,468,470,471,472,473,474,475,527,528,10501,10502,10470,10471,10438,10439,10500,8691,8690,10468,10467,10101,10434,10433,9404,9403,9405,10497,10496,522,525,10272,10273,524,520,521,9147,8718,8717,8719,10526,597,10495,10465,10673,10672,10684,596,595,594,517,10656,10655,10654,10653,10624,10952,10577,9535,10576,10552,10549,10551,10574,10618,10619,8987,590,592,10650,10651,591,593,637,638,639,17256,17257,17791,17767,17792,17817,7803,17818,17819,17794,17793,17337,17820,17821,17336,17822,7778,6206,6205,6204,7872,17823,17071,17824,17841,17452,7873,17258,16937,16936,16935,16938,16939,16944,16943,17413,17412,17455,17769,17446,17362,17795,17213,17204,18030,17361,17771,17132,17131,17133,17796,17772,17076,17759,17797,17798,17799,17773,17159,17338,17075,17077,16916,17416,16918,17351,17738,17701,17157,17369,17370,17371,17717,17716,17715,17700,17697,17372,17699,17683,17684,17698,17178,17703,17718,17722,17739,17198,17188,17760,17761,17187,17804,17806,17805,17807,17808,17831,17271,16934,16930,16933,17214,16932,17778,17269,17270,17835,17810,17809,17811,17836,17861,17834,17833,17862,17860,17858,17859,17874,17891,17875,17873,16950,17484,17485,17486,17913,17948,17275,17276,17431,17430,17004,17005,17006,17356,17482,17764,17196,17195,17197,17975,17977,17976,17366,17365,17364,17878,17368,17244,17409,17410,17411,17999,16908,17974,16910,16909,17993,16907,16906,863,867,864,868,750,747,746,869,17298,17299,17991,17969,17990,17992,18000,17973,17971,17972,17959,17968,17970,17958,17234,16970,16969,17215,17216,17231,17230,17232,17233,17245,17243,17367,17947,17027,17025,17929,17912,17247,17026,17246,16975,17908,17888,16976,16974,17870,17871,17784,17422,17889,17423,17933,18049,17932,17262,17260,17261,17206,17205,17207,17248,17055,17887,17669,17668,17667,17785,16971,16972,16973,17872,17471,17084,17085,17086,17868,17869,17886,17041,17720,17719,17721,16948,17890,17909,17042,16949,17856,17467,17855,17828,17857,17830,17829,17832,17466,17802,17800,17803,17826,17801,16945,17827,16947,17851,17852,17848,17825,17099,17846,17847,17853,16946,17854,17468,17100,17135,17850,17849,17056,17017,16941,16940,16942,17057,17126,17127,17054,17883,17595,17397,17282,17300,17301,17339,17844,17843,17842,17225,17226,17340,17212,16985,18047,17134,17098,17211,17259,6075,17158,17961,17960,17954,17989,17781,17777,17780,17779,17742,16931,17747,17746,17744,17743,17741,17726,17707,17705,16921,16920,16919,17745,17763,18042,17729,17748,17783,17751,17749,17752,17728,17314,17374,17375,17727,17750,8316,8315,8314,8345,8415,8411,8416,17710,17711,17708,17693,17694,17691,17021,17315,16929,16928,17022,17037,17687,17685,17672,17686,17334,17335,17018,17020,17688,17689,17674,17675,17679,17678,17443,17677,17657,17636,17658,18056,17655,17637,17676,8432,8363,17451,17441,8431,8430,8420,8424,17442,17690,17692,17709,8423,8368,8422,8347,8348,93,8421,94,95,328,330,329,8419,8425,8418,8426,8427,8435,4832,5422,4831,4336,4335,8417,8412,91,92,370,371,8413,8414,5439,5470,4337,4894,4893,4368,4367,4369,5464,5465,4926,4925,4951,4976,5468,5007,8317,8410,8346,8343,8409,8360,8408,8362,8361,8407,5009,5008,4978,5059,5040,5038,5412,8403,8406,8338,8341,8367,8366,8400,8512,8401,8397,8399,8398,5471,5362,5448,5397,5361,5084,5112,5111,5110,5372,5350,8395,8370,8331,8371,8393,8313,8329,8508,8342,8312,8311,8391,8335,8524,8337,17941,17942,17136,17966,17953,17454,8390,8392,5450,5403,4371,5404,5473,5349,5474,8394,5348,4372,4370,5109,5108,5107,5408,5136,5415,5152,5153,5138,5137,5435,5357,5334,5454,5158,5159,5160,5177,5166,5165,5164,5188,5457,5123,5139,5121,5140,5122,5120,5089,5090,5088,5091,5124,5335,5115,5085,5114,5134,5358,5392,5391,5135,5106,4355,4356,4357,5083,5082,5060,4444,5039,5012,4980,5011,5013,4982,5017,5016,5015,4445,5430,4703,5431,5105,5113,5086,5062,5061,4504,4503,4502,4702,4316,4315,4837,4835,5018,5019,5021,5020,4983,4954,4984,4956,4986,4985,4987,4955,5022,5023,5025,5045,4988,5026,4434,5027,4989,4359,4990,4959,4958,4936,4957,4932,4907,4906,4933,4934,4935,4908,4909,4871,4872,4842,4839,4870,4869,4903,4905,4904,4931,4930,4902,4929,4836,5014,4900,4928,4314,4479,4477,4478,4952,4953,4382,4927,4867,4378,4377,4376,4865,4866,4864,4862,4834,4863,4860,4859,4861,4833,4512,4857,4470,4471,4472,4858,4895,4896,4897,4898,4924,4948,4949,4383,4977,4979,4950,4981,5010,4310,4308,4805,4806,4785,5428,4784,5407,4789,4435,4309,4324,5360,4404,4403,4405,4323,4325,4838,5427,4810,5398,4793,4809,4808,4807,4811,4812,4813,4841,4814,4815,4816,4873,4876,5462,4843,4393,4817,4844,4845,4874,4875,4883,4854,4853,4910,4920,4921,4912,4911,4913,4943,4944,4960,4964,4937,4360,4358,4966,4965,5032,5033,5423,5434,5133,4432,5034,5075,5035,4407,5055,5058,4406,5052,5047,4346,5029,4804,5050,4452,4453,4443,4442,4441,5036,4420,4421,5458,5037,7647,7182,7888,7180,7245,5545,7246,6519,7299,7300,5001,7894,7306,7350,5003,5002,5000,5004,5005,5006,7349,7348,7347,7305,7346,7344,1032,1031,1030,1085,4974,1168,1167,1165,1091,4975,4973,5338,4972,4923,1164,1163,1162,1161,1160,4891,4890,4884,4885,4886,5443,5444,4970,4940,4945,4939,4938,4922,4916,4880,4889,4917,4915,4941,4942,4962,4919,4918,4882,4881,4851,4850,4849,4852,5356,4847,4848,5451,5355,5354,4796,4819,4411,4362,4412,4363,4361,4794,4751,4533,4750,4818,4395,4394,5405,4460,4459,4450,4449,4451,4769,4510,5366,4768,4892,4657,4706,4704,4481,4480,4482,4766,4765,4656,4458,4790,5417,4791,4792,5388,4437,4436,5246,4739,4738,5394,5440,4511,4491,4492,4489,4493,4701,4736,5446,4447,4448,4734,4473,4509,4508,4507,4737,4733,4732,4660,4698,4697,4699,4658,4659,4640,5367,4623,4622,4485,4484,4483,4655,4679,4680,4700,5396,5433,4731,4729,8461,8462,8460,8467,8468,8474,8473,8472,8525,8477,8489,4643,8480,8478,4644,4642,4641,5344,4611,4610,4428,4598,4349,4347,4348,4427,4588,4587,4577,4575,4576,4476,4374,4600,4601,4579,4580,4581,4573,4475,4373,4574,5418,4440,5365,5351,4546,4545,4540,4558,4351,4438,4439,4354,4353,4350,4596,4592,5399,5322,4339,4340,4619,4618,4617,4650,5330,4597,4689,4686,4687,4692,4653,4631,4654,4621,4338,4632,4630,4629,4633,4673,4635,4671,4670,4669,4672,4715,4713,4666,4667,4668,4714,4390,5406,4771,4710,4711,4712,4775,4774,4760,5442,4758,5472,4387,4388,4389,4311,4313,4312,4334,5479,1142,1141,1140,1144,1143,1145,5480,1146,1147,1148,5481,4759,4776,4777,4778,5429,1127,1126,1125,1149,5482,4780,4779,4781,4797,1151,1150,1152,1153,5483,4820,4821,4798,4822,1155,1154,1157,1158,5484,5437,4824,4888,5342,5485,4887,4878,4877,4879,4823,4800,4826,4825,4801,4803,4846,4402,4401,4400,4802,4799,4795,4391,5336,5353,4392,4770,4772,4773,4742,4707,4740,4708,4741,4767,4709,5030,4705,4664,4636,4663,4675,4674,4625,4627,5455,4662,4490,4661,5413,4624,4446,4626,4628,4605,4603,4602,4604,4634,5410,4578,4572,4474,4639,4599,5414,5445,1156,1159,15804,15803,15811,15823,15817,15649,15287,15650,15814,15810,15245,15286,15285,16804,15246,15244,16786,15836,15654,15653,15222,15223,15224,15772,15771,14744,14745,14746,15852,15853,15215,14878,14877,14879,16785,16784,15850,15592,15591,15589,15626,15627,15664,15820,15496,15802,15590,15419,15420,15421,16844,16903,16508,16509,15578,16510,16511,16512,16513,14863,15855,14860,15856,16841,15849,15851,15870,15764,15763,15765,16870,15458,15457,15288,15289,15502,15501,15499,15599,15598,15600,15915,16805,16896,15500,14866,14867,14868,15890,15214,15523,15260,15277,15524,16825,16895,16889,15868,15854,15519,15518,15516,14979,15517,14980,14978,14874,14875,14876,15834,14966,14967,14968,16787,15835,15357,15833,15832,15837,15838,14783,16854,15840,15859,15414,15412,15860,15880,15861,15879,15878,15877,15907,16851,15908,15881,15909,15935,15937,15938,15965,15343,15091,16904,15092,15963,15964,15687,14813,15930,15928,14812,16894,15875,15904,15093,15876,15874,15873,15413,15154,14651,15155,14650,15906,15902,15901,15900,14965,15511,14961,14643,14644,14645,14962,14963,15061,15062,16852,16758,15913,16874,15894,16796,16370,15735,15869,14964,15736,14649,15515,15514,15448,15447,16830,15520,15893,15535,15534,15533,15521,15193,15916,15917,15292,15191,15895,15918,15293,15919,15943,16811,14920,14919,14918,15951,15952,14754,15274,15275,15276,15972,14830,16010,15115,14755,15116,14793,14792,14794,16011,15316,15317,15318,15773,15165,15774,15167,16763,16762,16064,16764,16782,16829,16746,15240,15656,15375,15238,15746,15753,15748,15487,15733,15734,15611,15610,15609,15450,15069,15070,15071,15663,15662,14960,15239,15657,15035,15036,15034,16824,16139,15474,15473,15801,15800,14959,16151,14735,14736,16776,16777,16226,16807,16194,16051,15755,15756,15754,16050,16156,16155,16108,15058,15449,15056,15057,15026,15025,15027,16079,16037,15486,15485,16036,15000,15002,16039,16890,15985,14829,14828,15052,15051,14666,14667,14668,14827,14826,15612,15635,15829,16769,15613,15614,15177,15176,15175,15927,15922,15760,16568,15957,15955,15958,15959,15767,15961,15536,15537,15538,15962,15960,15929,14811,15677,14939,14938,14937,15992,15689,16867,15164,15989,15463,15163,15990,15991,15265,16834,15690,15264,15266,15565,15564,15563,16018,15688,16017,15084,15083,15082,16095,16094,16093,16097,15576,16052,16123,14648,16739,15698,15019,15021,15231,15229,15228,16122,15723,14646,15818,16880,16201,16200,15724,15725,16116,16117,15434,16167,15954,15953,15443,16743,14684,15038,15255,14817,15256,15136,15135,15037,16237,16328,16329,14833,14832,14831,15628,14818,15147,15146,15232,15665,15666,14882,14777,16283,15236,14779,15235,15237,15455,15365,15366,15367,15456,15337,15472,15338,16230,16828,16231,16285,16232,16233,16286,16287,16324,16325,16284,16733,15371,16272,14738,15373,15362,15361,15360,16360,16361,14740,14739,16741,16385,16384,15249,14945,14946,14947,15247,16426,15248,16427,16768,14712,14714,16772,16771,16383,16842,16815,15126,16422,14889,16423,15848,16501,16421,14803,16464,16502,16500,16837,16527,1090,1095,16528,16557,1094,16561,16560,1092,16562,16563,16559,15847,15127,16558,16506,14841,14842,14843,16597,15797,15495,15796,14800,14799,14798,15799,15828,16742,15798,14608,14973,14607,14606,15806,15423,15422,15805,16766,15815,15816,15827,15826,1166,14972,16898,16507,16897,16504,16505,15577,16503,15125,14713,16465,16466,14703,14704,14705,16846,15694,14986,14985,14987,14673,14674,14675,14718,14719,15088,15089,15090,15267,15658,16387,16432,15112,16433,16434,14913,15012,15011,15010,16280,16281,16467,15101,14720,15102,16468,15872,15871,16519,15198,16518,16522,16521,16520,16524,16523,15197,15478,14884,15479,16740,16077,16034,14851,14852,14853,16035,16525,16526,15752,14655,14654,14653,15353,16529,15355,15354,15751,16463,16816,16055,16462,16461,14844,16473,15652,15651,15465,14846,15403,15404,15405,16458,16891,15642,16456,16849,16850,16755,14845,14894,14893,14895,16770,14790,14791,16862,16801,14982,15179,15178,15180,14983,14984,15181,15023,15430,15429,15183,15182,14616,15300,16223,16222,16189,15190,15189,15022,15195,15194,16190,16144,16140,16059,16193,16091,16191,16058,16057,15131,14915,15132,15137,15138,15139,14990,14988,15252,15253,15254,15374,14686,14763,14762,14764,14749,14748,14747,16812,14991,14687,16181,16180,15925,15934,15497,15498,16102,15216,14904,15743,15166,15744,14781,14780,14782,14924,15200,15284,15074,15072,15073,15004,15005,15003,16138,15446,15445,16183,16185,16219,16184,16218,14927,14928,14661,16859,14660,15395,15397,15396,15415,16216,16215,16217,16253,16214,16252,16251,16020,16019,16021,15980,15979,15981,16300,16301,15673,16734,16343,14771,14772,16303,16302,15442,15441,14864,16365,14659,14865,14929,15009,15281,16256,16308,14820,14825,16307,16255,15149,15568,14909,16344,14823,14824,15120,15121,15707,16395,16372,16345,16379,14871,14872,14873,16354,15629,14850,14925,14926,15033,15631,15630,15944,15669,15503,15945,14633,14634,14635,15108,15109,15110,15453,15452,15451,15643,16312,16275,16352,15759,15757,15758,15686,14750,14751,14752,16229,14954,16228,14955,14610,16317,16319,16320,15370,14936,14934,16318,15586,16356,15243,15242,15241,16382,16381,15369,14935,16875,16893,15640,16377,16376,15141,15142,16351,16350,15507,15728,15505,15506,15792,14627,15793,15211,14628,14629,14789,16388,16290,16492,16549,16493,15346,15345,15347,15785,16416,15312,14618,15427,15426,15428,14619,14620,15794,15795,15140,16306,16305,15566,15567,16415,16855,15641,16457,16380,15251,15339,15341,15340,15407,15408,15409,16147,16146,16145,16148,16471,16056,16150,16149,15100,16472,16470,16469,15099,14721,14722,16435,15014,16436,15620,15618,15619,15745,14769,14768,14770,16316,14605,14604,14603,16364,15015,14897,14896,16023,14914,14912,15659,15660,16362,15259,15364,16363,14837,16327,14840,14838,16236,16315,16314,15741,15740,15742,16278,15585,15471,15042,14609,16227,16235,16164,15044,15043,15717,16157,16154,16152,16153,15933,15932,15931,16115,15510,15509,15508,15570,15571,16799,16081,16045,16041,16043,16038,16013,15543,16014,16042,14940,15542,14669,15987,14941,14716,16040,16012,15986,15988,16015,14857,16046,16083,14717,16082,16048,16047,16084,16832,16831,16417,16731,16732,16677,16159,16158,16160,16118,16791,16790,16120,16119,16166,16165,15436,16774,15435,16198,14881,15233,14819,15094,14880,15788,16748,15790,15789,16161,16162,16113,16114,16112,15899,16767,15336,15791,15380,15207,15898,16111,16757,15648,14797,14795,14796,15655,15857,15858,16756,16044,15406,14715,16885,15379,15381,16163,16860,16861,16863,16819,16820,16234,15202,15294,14839,15295,16326,15257,15258,16847,15668,15201,15203,15769,15770,16121,15889,16088,15887,16087,16086,15392,16049,16085,15169,15424,15390,15391,15425,15700,15699,15701,15722,15393,15768,15394,15766,15168,15152,15151,15153,16016,14859,14858,15594,15156,15595,15593,15761,15956,15716,16053,14647,16836,15888,16089,16090,16080,16843,16279,14611,15584,15583,16869,15250,16797,16274,16224,16225,16273,14815,14814,14816,16258,16250,15729,15221,14997,14998,14999,15646,15647,16259,16261,16260,16220,16873,14930,16254,16304,15150,15148,14910,16276,16277,15114,15113,15504,16313,14911,16353,16355,14821,14822,15995,15363,16342,16877,16262,15809,15807,15170,14981,14743,14742,14741,16213,15661,16221,15808,16212,16211,16182,15924,14685,14989,15923,14974,15416,15417,15541,15540,15539,16267,16192,15670,15217,16073,15213,16033,16007,15636,16008,15053,15054,15055,15064,15065,16009,14753,15291,15950,15531,15192,15532,15977,15978,16858,15637,14662,14663,14664,14665,15720,15721,16000,16001,16002,16027,14886,14888,16003,16004,15976,16813,16814,16199,16783,15290,15897,15896,15526,15974,15973,15975,15527,16775,15199,15331,15528,15529,15530,15525,16103,16104,15685,15684,16516,16515,16431,16430,16429,16848,15552,15692,15693,14869,16514,14861,14862,16759,14870,15579,15111,15554,15555,15556,15387,16517,15040,15997,15998,15999,16026,16028,14887,16030,15624,16070,16029,16006,14757,16032,16031,14758,15727,15212,14905,16744,16005,14756,16069,16141,15625,15544,14917,14916,15545,15303,16076,16067,16066,16068,16065,16074,16075,16773,14883,15039,14885,16105,16078,16106,15302,15304,16107,15597,15596,15024,16555,16556,15762,15949,15522,16599,16598,1093,1033,1034,1027,1089,1028,1029,1036,7237,7292,7294,7293,7295,7729,7345,7032,7033,7034,6520,5546,7749,7734,7733,7735,7787,7242,7243,7244,7726,7236,7238,5733,5731,5732,5782,7291,7290,1035,1039,1038,1084,1041,7289,7288,7235,5988,5781,7877,7110,6285,7173,7172,7171,7108,7109,7107,5670,5616,5669,7014,5810,7046,7104,7105,7106,7816,7101,7026,5973,5842,5974,5844,7168,5841,5839,5840,7232,7233,7285,7286,7342,7343,7744,7338,7283,7284,5889,5888,5887,7230,7741,7140,7139,7138,5843,7018,7068,7019,7066,7136,7135,7137,7187,7249,7899,7190,7189,7191,7251,7250,7252,7253,7307,7254,7255,7258,7309,5640,5641,7310,7311,5533,5534,5532,5705,7257,7256,7231,5706,5704,7340,7339,7341,7401,7763,7449,7448,1051,7450,7414,1053,1052,1055,1054,1056,6280,7413,7415,7459,7461,7416,5585,7458,7460,7502,7503,7504,1057,1058,1059,1060,1061,1062,7546,7543,7545,7544,7542,1064,7883,7584,1067,7583,7540,5918,5894,7581,7582,6408,7626,7579,7580,6014,5919,5917,6015,7902,7541,7539,7538,7536,7578,7537,7618,7567,7528,7566,7619,7565,7650,6938,6939,6940,7564,7526,7480,7768,5707,5708,7481,7479,6369,6370,7613,5690,5691,7525,7524,7522,7563,7776,5899,7775,7523,5793,5703,5701,5792,5794,7021,7022,7471,7476,7433,5666,5667,7474,7432,7475,7517,7518,7519,7473,7472,7557,7556,7605,7606,7561,7607,7779,7673,7780,7637,7636,7638,7498,7499,7500,7671,7672,7674,7753,7798,7907,7670,7669,7635,7914,7906,7915,11526,11403,11395,11394,11036,11035,11034,11138,11385,11137,11136,11381,11367,11371,11368,11369,11372,11532,11362,11084,11345,11344,11226,11342,11533,11380,11358,11359,11357,11343,11370,11355,10994,11217,11218,11337,11228,11227,11346,10986,11539,11538,11075,11315,11530,11158,11188,10988,10987,11152,11153,11147,11148,11347,11348,7680,7912,11083,11363,11085,7640,7826,7639,7643,7612,7611,7676,7677,7678,7642,7641,7609,7679,7681,4518,7695,4522,4523,4524,4525,7682,6775,6774,7696,7170,7169,7685,5459,4567,4534,4568,7684,7683,7688,7687,7697,7698,4569,4570,5419,5461,7702,5425,4593,7699,7700,7703,7701,7704,7715,7716,5673,4694,7717,7708,7718,5671,7705,7691,7706,7693,5922,7692,7690,7646,7689,5709,7396,7649,5920,5921,7617,7660,7694,7661,7651,7616,6407,6406,7662,7663,6885,7664,1072,1074,1075,1073,1076,1077,6883,6884,7713,7712,7710,7709,7711,7764,7707,7719,7720,7721,7722,1079,1078,1080,1081,7723,1083,7725,4754,4753,4752,4755,4728,7724,5672,4693,4691,4495,4494,4496,4455,4514,4513,4722,5386,4386,5333,5332,4385,4717,4681,4716,4684,4683,4647,4685,4651,4614,4612,4615,4646,4645,5132,4648,4649,4589,4590,4591,4571,4564,4565,4555,4613,5453,4595,4566,4594,4616,4620,4652,4690,4554,4543,5368,4542,4544,4526,4530,4529,4532,4541,4539,4556,4352,4557,4536,4537,4549,4553,4322,4550,4320,4551,4375,4430,4559,4552,4585,4586,4429,4609,4584,5376,5409,4419,4418,4417,5374,5424,8304,8497,8491,8502,8503,8501,8479,8487,8496,297,8498,294,291,292,295,276,8499,8500,293,8504,8526,8505,4638,4608,4637,8506,8507,274,147,261,332,11279,11327,11285,11284,11270,11328,11305,11212,11323,11088,11072,11070,11303,11302,11301,11211,11213,11013,11060,11061,11062,11014,11267,11151,11549,11026,11028,11112,11113,11304,11287,11272,11271,11444,11567,11515,11286,11283,11282,301,277,146,306,342,299,298,283,265,284,349,350,351,378,263,234,256,232,231,257,258,259,260,145,275,253,387,254,229,230,199,196,195,194,192,191,193,226,255,323,227,296,8488,228,189,188,187,190,185,184,183,316,8456,8455,8454,8457,8459,368,327,116,117,8446,8458,8453,8445,8448,4783,8449,8357,8356,8355,8434,5373,8439,8528,5343,4764,4827,8436,4829,4828,4830,4786,4787,4763,4762,5401,5402,4788,4735,4782,4761,4730,5420,8447,8433,100,359,324,325,99,98,96,103,354,102,101,326,337,341,120,376,336,105,104,106,112,113,127,118,377,125,311,124,126,312,358,164,138,360,137,136,119,115,317,313,139,159,158,160,161,141,333,186,140,163,162,166,173,172,345,375,208,197,200,198,90,235,233,89,88,278,266,279,280,281,282,264,262,379,23652,23618,23730,23673,23674,23675,23688,23689,23721,23723,267,238,361,207,171,205,380,322,170,202,356,355,236,268,23703,23704,23714,23713,23715,23733,23630,23731,23719,23732,23631,23620,23621,23622,23707,23645,23644,23616,23676,23677,23729,23706,23700,23682,270,23658,4291,23659,23681,23699,23698,23697,4289,4288,4287,4284,4294,4295,4297,4293,4292,286,272,4290,271,287,353,334,319,247,340,384,285,331,246,288,289,273,248,290,250,252,389,251,224,237,315,209,204,352,223,220,335,150,177,176,388,221,216,385,210,213,357,157,180,156,149,369,343,155,344,373,362,338,148,152,151,174,175,178,179,144,143,168,167,203,201,374,318,364,122,165,121,123,142,314,382,310,128,321,114,339,111,135,134,133,367,366,132,153,131,154,8463,8372,8332,8334,8464,8515,8465,8466,8518,17574,17588,17589,17590,17591,17575,17564,17576,17280,17457,17063,17064,17065,17456,17066,17067,17068,17424,17532,17547,17549,17548,17550,18045,17449,17450,17448,18028,18046,17388,17390,18027,17552,18044,17533,17534,17093,17254,17253,17252,17520,17519,17517,17518,17469,17061,17535,17554,17555,16988,16987,16986,17569,17570,17579,17557,17558,17436,17571,17578,17572,16922,18019,17437,17438,17556,17151,17152,17380,17536,17521,17522,17379,17381,17523,17537,17538,17539,17559,18022,17480,7298,17524,17525,17512,7297,7296,7783,7823,17503,17513,17192,17417,17360,17504,7810,7577,7535,7890,17268,17193,17489,17488,18038,17494,7569,17385,17384,17386,7893,7622,7652,7653,7654,7621,11468,17483,11185,17487,11184,17427,17493,17490,17491,17496,17501,17499,17500,17498,17497,11201,11466,11086,11080,11081,11082,11087,11178,11560,11534,11465,11457,11456,11021,11459,11122,11460,11121,11120,11139,11019,11020,11022,11115,11172,11173,11174,11531,11522,11451,11242,11073,10984,10983,10985,11074,11543,11450,11544,11523,11001,11189,11454,11453,11455,11190,11119,11118,269,244,245,11469,11058,10999,11575,11573,11485,11006,11005,11004,11000,11471,11470,11195,11194,11193,11472,11240,11486,11475,11520,11241,11214,11477,11215,11216,11479,11177,11176,11175,11480,11478,11481,11489,11482,11493,11490,11491,11492,11499,11250,241,11059,11484,11497,11488,11498,11249,11487,11496,11263,11476,11474,11473,11262,11264,239,242,214,243,8495,8486,8485,8483,11494,11495,17542,17540,17527,17526,16955,17528,16951,17529,17541,17543,17544,17107,17545,17562,17563,17281,17106,17105,17546,17530,16952,16953,17221,17001,17220,17439,16954,17128,17003,17002,17514,17508,16980,17509,17507,17016,17292,17291,17290,17333,17505,17515,11044,11525,11046,17129,11045,11483,11110,11111,11123,11528,11571,11463,11464,11553,11569,17506,16965,16966,16967,17104,17492,16984,16983,17194,17289,17288,17495,17078,17502,16963,17150,16962,16964,17119,17470,17510,17511,17114,17113,17115,17389,16981,17353,17516,17531,18048,16982,17479,17293,17294,8471,8470,8475,8516,8484,8476,8493,8481,8492,8494,240,225,372,211,212,182,181,8469,8482,215,219,218,217,249,222,11563,11146,11145,11016,11018,11017,11225,11224,11064,11063,11065,11440,11419,11421,11518,11449,11519,11420,11068,11578,11442,11066,11067,11441,11581,11576,11535,11101,11443,11012,11010,11099,11100,11266,11417,11548,11031,11418,11407,11277,11276,11505,11438,11437,11436,11011,11579,11207,11506,11507,11508,10982,11562,11537,10980,11458,10981,11230,11229,11504,11220,11219,11221,11183,11467,11561,11557,11461,11509,11208,11206,11124,7658,11125,7657,11558,11462,7656,7898,7655,7620,7623,7200,7199,7198,7573,7624,7574,7575,7576,7625,7868,5587,7572,5589,5588,7585,7547,7548,7549,7587,7588,7589,7590,7550,7505,5762,6117,7506,7464,5761,7507,7552,7551,7591,5541,5543,5542,5580,7629,7627,7586,7628,5579,5578,7665,7666,7859,11439,11542,11433,11430,11429,11431,11434,11432,11246,11559,11516,11423,11411,11248,11408,11198,11197,11196,11142,11140,11141,11398,11524,11512,11391,11393,11392,11399,11400,11261,11231,11260,11254,11259,11232,11233,11384,11114,11057,11040,11364,11353,11365,11166,11167,11168,11169,11109,11108,11107,11366,11354,11054,11577,11356,11243,11052,11238,11341,11244,11129,11130,11131,11237,11156,11155,11154,11096,11097,11098,11336,11239,11182,11181,11008,11554,10992,11511,11335,11009,11007,11076,11164,11165,11171,11170,11310,11309,11132,11102,11103,11091,11160,11159,11295,11078,11077,11079,11092,11540,11546,1329,11294,11545,1328,1330,1332,1309,1334,1308,1338,11308,11307,11306,11222,11551,1336,11334,1326,11550,11502,1325,1324,1323,1322,11340,11209,11503,11572,11245,11500,11556,11191,11053,11192,11351,11527,11223,11205,11204,11203,11383,11382,11376,11374,11375,11555,11377,11390,11095,11094,11093,11023,11150,11149,11043,11041,11042,11029,11025,11024,11144,11030,11143,11447,11448,11186,11388,11387,11389,11396,11386,11521,11128,11126,11127,11580,11275,11378,10989,10990,10991,11373,1311,1315,1314,1316,1317,11055,11352,11051,11350,11501,11349,1319,11050,1320,1321,1318,23623,23664,23665,23709,23708,23669,23624,23611,23670,23612,23613,23662,23661,23663,23695,23671,23648,23626,23625,23627,23638,23637,23642,23643,23651,23696,23735,23693,23694,23639,23641,23634,23728,23727,23726,23725,23680,23667,23633,23628,23629,23690,23647,23649,23724,23686,23640,23687,23654,23722,23653,23678,23679,23650,1327,1310,23734,23736,1335,303,1333,23720,23646,23619,23617,363,309,320,300,302,304,305,308,307,11200,11199,11288,11566,11529,11513,11290,11289,11564,11565,11311,11162,11296,11297,11298,11300,11299,11027,11163,11161,11291,11514,11292,11552,11314,11134,11015,11133,11135,11003,11049,11048,11047,11321,11320,11071,11317,11318,11319,11326,11324,11322,11325,4607,4606,11570,4426,5449,5389,4561,5390,4582,4563,4583,4562,4560,4547,4321,4548,4538,5469,4396,5400,4535,4399,4531,4506,4463,4505,4527,4528,4462,4461,4341,4457,4456,4343,7714,4342,11339,11330,11179,10998,11180,11210,4398,4397,5416,11517,11333,10997,11331,11332,11510,10996,10995,11002,11104,11105,11106,11235,11313,11312,11236,11234,11157,11329,11536,1331,11293,1337,23635,23636,23632,23716,23717,23660,23718,23705,1339,23615,23666,23668,23702,23710,23657,23656,23655,23711,23712,23683,23684,23685,23672,1312,4277,1313,4278,4279,23692,23691,4280,4281,4282,4283,4286,23701,4285,11416,11038,11039,11032,11541,11404,11406,11405,11187,11397,4296,11265,11116,11117,11037,11547,11435,11069,11033,11316,11274,11273,23614,11202,11257,11256,11258,10993,11568,11056,11379,11253,11255,11413,11412,11401,11402,11269,11410,11409,11414,11427,11428,7785,7668,7786,7909,7908,7667,7632,7633,7597,7599,7634,7600,7892,7603,7601,7602,7820,5607,5606,5605,7598,7553,7596,7595,5834,5833,5832,7511,7469,5596,5886,7512,7470,5598,7430,7513,7514,7516,7515,7554,7555,7424,7425,7426,7375,7376,7431,5668,7850,7378,5737,5770,5769,5771,7434,7754,7435,5702,7436,7482,7483,7484,7889,5900,5898,7527,6416,6415,6418,5688,5687,5686,7455,6417,7080,5630,5631,7454,7453,7406,7404,7403,7402,5648,5644,5642,7451,7437,7452,7478,5905,7486,7485,5904,5906,7903,7477,7379,7377,5885,5601,5600,5599,7324,7380,7352,5643,7351,7353,7303,7354,7357,7356,7355,7358,7308,7304,7302,6521,7247,7248,7186,7134,7133,7185,7130,7184,6523,6522,6483,6482,6484,7092,7093,7094,7078,7075,7076,7079,7031,7011,7126,6485,7183,7060,7010,7061,7128,7129,7127,7064,7131,7132,5911,7013,5913,7015,7016,7067,7065,7813,7017,6975,7784,6871,7760,6870,6869,6902,7815,6798,6802,6797,6903,6800,6735,6737,6801,6832,6763,7882,7881,6836,6901,6944,6945,6976,5813,5595,5593,6973,6974,7025,7045,5594,6906,6977,6908,6907,6978,6979,7051,7730,7047,5809,7048,7050,7049,6661,7904,7799,7111,6660,7112,7176,7175,7174,5996,5615,5614,7793,7178,7113,7114,7177,7116,7179,7118,7117,7122,6759,6758,6760,7121,7834,7123,5544,7181,7125,5652,7773,7124,5928,5467,7761,5393,5363,5346,5345,5364,5048,4326,4327,5926,5654,7809,5653,7052,6705,6894,6895,6896,7006,7003,7004,7005,6983,6982,6984,7757,6980,6840,6916,6844,6843,6842,7740,7781,6839,6767,6765,5516,6766,6762,6838,6834,6835,6837,6904,6905,6909,6910,6911,6912,6913,7745,7765,5939,5940,5924,5925,7770,6981,6915,6841,6914,5923,7115,6659,6761,6717,6716,6713,5514,6833,7743,6714,6613,6634,6614,6612,6530,6586,6499,6564,6501,6500,6561,6562,6563,7866,6504,6502,6503,5854,5862,5856,6392,6411,6355,7863,6391,6367,6357,6358,5819,6859,6351,7869,6354,6356,7814,6488,5697,7746,6352,6376,6480,6478,6481,6421,6419,6479,6420,6373,6422,6423,6427,6492,6544,6545,6546,6540,6539,6486,6489,6538,6491,6593,6592,6547,6524,6600,6601,6602,6579,7879,6644,6698,6647,6627,6689,6690,6730,6729,7828,6743,6700,6742,6744,5826,6745,5824,6808,6811,7818,6880,6881,6878,6879,6953,6810,6809,6794,6793,6868,6942,5677,7012,6867,5679,6941,6936,6935,6934,7009,7063,5912,6864,6863,6862,6866,6865,6796,6937,6795,5816,5814,5815,6790,6687,6628,6629,5726,5725,5724,6688,6582,6581,6578,6583,6527,6528,5861,6462,6463,6409,6410,6466,6465,5855,6529,6532,6585,6584,6531,7891,5858,7759,6633,6630,6631,5859,6799,6731,6736,6632,5857,6464,5695,5696,6525,6490,6487,6526,6791,6792,5678,6957,6943,6955,7029,7030,6956,6954,7027,7028,7074,7077,7152,7150,7149,7151,7210,7209,7206,7207,7320,7321,7322,7267,7208,7268,7323,7301,5739,5738,7269,7374,7266,7204,7205,5779,7203,7148,7147,7146,5646,5827,5645,5828,6958,7837,7497,6882,6541,6542,6543,6963,6813,6749,6750,5748,5635,6747,6746,6701,6748,6812,5825,5897,5634,5633,5896,6646,6598,6645,6599,5759,6648,6558,6557,5747,6739,6691,6635,6639,6738,7805,6693,6641,6640,6638,6590,6589,6536,6535,6534,6469,6588,6533,6467,6473,6475,6413,7852,6476,6477,6439,6537,6438,6555,6591,5699,6642,6658,6692,6694,6706,6663,6662,5573,6708,6707,5572,6664,6657,6665,6666,6752,6753,6819,6824,6891,6890,6656,6382,6384,6609,6560,6605,6552,6554,5528,6559,5527,6607,6608,5698,5700,6556,6497,5526,6435,6434,6436,6498,6553,6412,6474,6472,6470,6925,6468,6603,6471,6548,5758,6637,6597,6596,6595,6594,6319,6320,6643,5757,6184,6587,6636,6377,6361,6372,6258,6259,6260,6030,5745,7762,7833,6353,6321,6374,6375,6322,6316,7886,6195,7853,6315,6228,6257,6194,6193,6157,6227,5621,6156,7838,5619,6145,6144,6215,6132,6131,6133,6142,5651,5650,6217,6216,6261,5612,6297,6298,5519,5518,5517,5649,5836,6134,7867,6078,6264,6263,6300,6339,6394,6393,5552,6340,6444,6445,6446,6509,6508,5550,5551,7739,6507,6505,6441,6442,6443,6506,6611,7758,7870,7900,7817,6667,5566,6618,5565,6615,6673,6616,6676,6720,5567,6677,7880,6718,7806,6721,6769,5914,6770,5915,5916,6679,6728,5873,6783,6787,6788,6789,6861,6928,6933,5581,5583,5582,6845,6846,7755,6768,6719,6675,5515,6674,6617,6715,5684,5683,5685,6610,5874,6619,6516,6517,6678,6567,6566,6518,6514,6512,6513,7772,6565,6365,6366,6336,6338,6337,5727,5613,5611,7831,6296,5620,6275,6256,6274,6295,5821,5820,7769,6196,6952,7201,7202,7263,7264,7265,5778,5780,7372,7371,7862,7364,7428,7319,7318,7317,7369,7325,7363,7362,7427,7509,7407,7510,7468,7429,5597,5835,7408,7466,7467,7508,7463,7593,7592,7594,7630,5750,7631,5728,5729,5730,5751,11425,11426,11247,11268,7373,7359,7896,7405,5632,7412,7411,7410,5639,7360,5586,7409,5853,7750,7457,5851,5893,5895,7501,5852,7456,5584,5720,5719,5903,7604,11415,11278,11338,11424,11422,11251,11252,11089,11445,11090,11446,11452,7659,6116,6118,7102,7419,7752,5558,5557,5556,7103,7487,7531,7440,7439,7227,7226,7225,7385,7337,7386,7489,5676,5674,5817,5818,7490,7491,7887,7571,7570,7797,7568,7529,7530,7532,7533,7534,7839,7871,5675,7840,7905,7361,16927,16924,16923,17045,16925,17837,17585,17046,17602,17603,17584,17580,17433,17434,17435,611,610,612,640,613,17461,17273,17272,17274,614,564,567,615,641,17617,17615,17612,17599,16997,17583,16996,16995,17331,17594,17222,17103,17102,17101,17165,17164,17163,17354,17355,17582,17581,17600,17616,17036,17062,17035,17034,17094,17391,17108,18053,18055,17565,17566,16913,17218,17008,17219,16915,16914,17284,17285,17250,17598,17009,17630,17382,17383,17628,17228,17160,17229,17161,17162,17249,17154,17155,17156,17255,17029,17323,18024,17631,17665,17028,17030,17287,17286,17171,17597,17596,17169,17577,17418,17350,17200,17201,17202,17203,17414,17415,18021,17109,17567,17110,17553,17568,17117,18051,18037,17593,17145,17146,17130,16990,16989,16991,17040,17051,18050,17039,17347,17348,17349,17651,17632,17475,17652,18033,17639,18039,17670,17671,17459,17650,17458,17649,17648,17393,17392,17177,17312,18029,17095,17096,17666,17190,17189,18032,16959,16960,17325,17324,17318,17316,17317,17357,17394,17395,17696,17172,17174,17012,17153,17011,17387,17327,17142,17695,17307,17308,17663,17662,17681,17306,17378,17377,17376,17425,17426,17680,16912,628,626,627,581,629,631,630,583,582,584,585,632,633,634,17326,17682,17328,17712,17713,17141,17140,18040,17730,17731,17732,17734,636,17733,635,587,589,588,586,10743,10644,9936,10645,9690,9689,9691,10617,8949,8950,8951,10615,10601,9440,9441,10602,10606,10605,10748,10616,10372,8938,8939,8940,9125,9124,9328,9372,9045,9044,9043,9371,9382,10487,10486,10488,10786,10890,10724,10522,8797,8800,8801,8798,9253,9252,8996,10489,10740,8762,8998,8997,8763,8764,9059,9244,9245,9246,9260,10491,10490,10525,10735,10524,9160,10553,9534,10550,9454,9455,9456,9158,10580,10578,10581,10582,10583,10620,10363,10321,10319,10364,10394,10395,10427,10622,10621,10575,10623,10652,10428,10466,10429,10430,10431,10396,10397,10432,10320,10262,10579,10554,10784,10327,9539,10328,10493,9222,10492,8759,8760,8761,8825,8824,9366,8757,8758,8756,9540,10040,10763,10836,10837,9302,9301,8823,10815,10264,10263,10150,9339,10151,10918,10187,10779,10832,9418,8745,8744,8735,8736,10157,9992,10156,9417,9419,10190,10189,9107,10188,10199,10200,8822,8820,8821,10708,10710,10709,10835,10901,10204,9525,8739,8738,8740,8848,8847,8846,9932,9097,9096,9098,9798,9477,9321,9473,9320,10205,10203,10202,10201,10209,8985,8984,8986,10937,10134,10910,10093,10094,9080,9081,10792,9278,10083,10158,9105,9524,9106,9170,9280,10823,9384,9993,10132,8737,9236,9237,9235,10041,10082,10044,10819,10820,9385,9279,10685,10046,10045,10047,10051,10052,9211,10053,10048,9952,10829,9951,9880,10745,9917,9916,9915,9950,9912,9949,9878,9877,9879,9991,9990,9946,9947,9945,9911,10898,10716,10921,9365,8843,8844,9914,9913,10867,10868,9020,9909,9908,9907,10715,10929,9409,9972,9973,9974,10928,9989,8695,8693,8694,10855,10020,9120,10021,10018,9971,9122,9121,9970,10933,10013,10016,9543,9542,9541,10015,10118,10117,10071,10019,10070,10119,10120,8958,10972,10059,10152,8957,8956,10730,10060,10058,10124,10123,10793,10074,10072,10017,10073,8755,10122,10121,9194,9193,8678,8679,8680,10459,10425,10357,10426,8654,10462,10460,9223,10461,9224,9221,9225,10463,10464,9111,9094,10494,9760,10783,9159,10133,8652,8658,8659,10358,10353,10352,10312,10355,10354,10356,9050,9049,9048,9783,9782,9464,10224,9465,9274,10258,9110,10259,10311,10256,10954,10235,9503,10261,8889,9500,10236,10226,8891,10876,10225,10877,10313,10314,10277,9073,9072,9074,10276,9851,10260,9832,9809,9831,9850,9868,9872,10691,9485,9833,9834,8972,8686,10774,9796,10801,9835,8971,8978,9052,9051,9053,8977,8976,9839,9771,9770,9778,10711,10389,10721,9836,9837,9838,9840,9841,9843,9842,8865,10951,9493,9799,10915,9780,9351,9353,9779,9776,9775,9777,8932,8729,8730,8731,9712,9685,9683,9432,9430,9057,9056,9058,9681,9680,9709,9710,9711,9523,9148,10149,9707,9032,9291,9027,9033,10755,9347,9421,9420,9422,9357,8873,9119,9439,9438,10126,9117,10080,10069,10066,10067,10704,8929,8930,8931,10701,10700,10699,10780,9988,9966,9964,9965,10900,10068,10945,9963,9962,10010,8780,8779,8778,9307,9308,9671,9672,9327,9326,9325,10116,9313,9206,9205,9204,8995,8994,8993,9498,9499,10065,10887,10849,8852,8853,8842,8953,9383,9166,10062,9164,10009,9478,10754,10904,8841,10689,10688,10687,10064,8713,10948,10063,10856,9931,10803,10778,9889,9892,9934,9933,9566,8712,9567,10834,9550,9848,9961,9960,9906,9849,9830,9324,9829,9828,9826,9893,9190,9191,10942,8968,8969,10965,10781,9935,9967,9968,9549,9551,10848,8711,10814,9322,10185,9795,9323,9370,9369,9234,9808,9787,9807,9806,10866,9847,9179,9885,9064,9063,9065,10960,9890,10862,8893,8894,8892,9891,9722,9721,10567,9180,9803,9846,9577,9578,9576,9845,9844,9884,10857,9479,9480,9249,9251,9476,10109,9574,9573,10108,9368,9265,9367,8777,9474,9343,9342,8776,8954,8705,8948,9475,10110,8707,8706,9165,10941,10112,10111,10113,10146,10899,9294,10182,10175,9295,8922,10176,8921,8783,9487,9488,10257,10221,10179,9108,9109,9113,10178,10177,8775,8782,8781,8787,8788,10228,10807,10180,10184,10231,10230,10229,10197,10198,10963,9810,9484,10838,10979,9766,10196,10194,10195,9733,9754,9732,10193,9071,10731,10154,10147,10155,10183,10181,10148,9315,9314,9635,9589,9636,9591,9590,10131,10114,10129,9070,9069,10153,9118,10130,10115,10191,10192,9408,10893,9407,9406,9708,9706,9736,9735,9734,10873,9767,9765,9768,9486,10850,10964,9769,10143,9774,9773,9772,10943,10757,10081,9665,9684,9093,9687,8864,8863,9876,9875,9919,8845,9918,8860,8861,8862,9881,9882,8814,8812,8813,9515,9514,9516,10816,10817,9210,9209,10087,9201,10088,10135,10953,9202,10936,10159,10136,9186,10886,10875,9187,9188,10162,10210,10212,8795,10208,10240,9275,9277,9276,8796,10885,8917,8918,8916,9013,10796,8974,10335,8973,8975,10333,10725,10337,10378,10375,10374,10331,10962,10280,9513,10332,10282,10281,9512,9014,10373,10825,10444,10445,8928,10407,10377,10406,10447,9016,9017,10446,8927,9207,10475,8898,10379,10336,10869,10334,10861,8900,9532,9531,9533,10380,10881,10339,10338,10880,8887,10289,10288,10160,10137,10096,10879,10878,10907,10693,10092,9200,10091,10090,10703,10095,9627,9544,10973,9545,9546,9995,9994,9656,9996,9997,9998,9954,9657,8989,9037,9038,9039,8773,8909,9923,8988,8990,10833,10903,10902,8881,9955,10000,10001,10002,10004,10005,10006,9389,10924,10802,10912,8704,8703,8702,9745,10099,9390,9746,10055,10057,10056,9999,9628,10140,10097,9702,9703,9814,9655,10728,10729,9921,10759,9922,10758,9883,10050,10732,10049,10089,9338,10720,9021,8685,8684,10788,9871,9894,9870,9869,9895,9896,9897,9939,9938,9937,8665,8664,8663,9940,10932,10831,10014,9969,10317,10316,10315,8890,9502,9501,10690,10359,10360,10977,10361,10362,9192,10811,10859,10971,10939,10351,10042,10043,9174,9173,9175,10548,10523,10547,9453,9452,10545,10546,10573,10544,9046,9047,10889,10541,10897,10566,10565,10564,10569,10572,10603,10604,10641,10600,10568,9155,10776,10643,10777,10642,10598,9154,9153,10599,10746,8982,8983,10563,9018,9198,10646,10613,10612,10614,574,573,571,570,569,601,568,566,9199,10671,600,616,642,617,618,18031,17625,17332,17626,17640,17642,619,17641,17007,17627,17629,17643,17660,622,620,575,621,577,576,578,579,598,624,623,625,17661,16911,10609,10611,10610,10570,10571,10608,10607,580,572,17224,17223,10649,17753,17754,17755,17766,17768,17447,16994,17363,17756,17242,17241,17240,17010,17737,17736,17714,17735,16917,17758,17173,17251,17352,17373,17191,17770,17757,17644,17664,16961,17031,17033,17032,17613,17097,17227,17463,17170,17052,17179,17311,17167,17460,17168,17180,17175,17673,17704,17706,17176,17725,17724,17740,17762,17723,17702,17166,17472,17474,17473,17659,17112,17019,17111,17088,17087,17089,17638,17038,17611,17305,16979,17614,16978,16977,16957,16956,16958,17144,17143,17217,17283,17116,17118,17551,17592,17609,17120,17121,17122,17610,17622,17074,17073,17072,17635,17633,17654,17656,17624,17623,18018,8522,8440,17634,8441,8442,8444,8306,8305,8307,8333,8510,8452,8451,8519,8520,347,8450,8443,8438,381,346,97,386,8437,8364,8365,8429,8428,348,110,107,108,109,129,130,365,169,206,383,11280,11281,8490,5441,4682,4721,5352,4688,5369,5370,4743,5371,5447,4745,4746,4747,5438,4744,1135,5487,1130,1131,1132,1169,1129,1128,5463,5486,4384,4748,1170,1123,1124,5476,4757,4756,4454,1120,1121,1119,1122,15602,15601,16062,16063,15548,15283,15547,15546,16024,16736,15993,15994,15549,15587,15588,14641,14639,14640,14642,16735,16025,16022,16265,16264,15558,15557,15462,15464,15488,15344,15162,15342,15968,15967,15966,16835,15971,15942,15941,15940,15912,15308,15910,15939,15936,14723,14724,14725,15418,15863,15862,15864,15865,15839,14785,14784,15822,15821,15841,15842,15844,15843,15845,15824,15356,15825,15831,15830,15846,15867,15883,15882,15866,15348,15349,15884,15911,15309,15886,15307,1139,15970,15969,1137,1136,1138,1134,1133,15551,15550,5478,4333,4488,4487,4486,4724,5477,4723,4749,4727,4665,4695,4696,15358,16054,15282,15204,15206,16061,15359,16060,16098,16099,16100,15208,16130,15209,15210,15225,16129,15227,16096,15575,15020,16902,16092,15330,15329,15726,16778,16126,16132,14786,14787,14788,16169,16125,16124,15676,14948,14949,14950,16168,15230,16204,16203,16205,15819,16101,14680,16202,14682,14683,14679,14681,15352,16240,16241,16242,16243,16244,16296,16245,15299,16246,16299,1107,1108,1106,1109,15298,16209,16208,15675,1110,1063,1111,1112,14690,14689,14688,16173,16206,16207,16248,16249,1088,1068,1069,1087,1071,1113,1114,1115,16210,16247,16170,16176,16175,16177,16178,16172,16134,16131,16171,15226,14671,14670,14672,16135,16136,15205,16133,16174,1118,1117,1082,1116,16137,16179,1070,1066,1086,1065,1105,16295,16297,16298,16795,15087,15085,15086,1104,16340,1103,16341,16336,16335,16337,1102,16334,16292,16331,15350,15351,16239,16238,16293,16332,16339,16367,16333,15401,15400,15402,16323,16330,15468,16359,15130,15129,15128,14706,14707,14708,16357,16358,16188,16386,16186,16187,15372,14778,16321,16282,16322,16818,14804,14802,14995,15470,15469,15319,14693,15320,16428,14691,16369,16368,16366,16390,16391,14692,1098,16437,1099,1040,1101,1100,16389,16338,1044,16371,1043,1045,1046,1047,1048,1049,1050,7400,7234,7287,1042,1096,1097,15891,15892,1037,7615,7648,7686,7812,6371,7614,7874,7645,5689,6773,5637,7644,5638,5636,7610,7608,7560,7559,7562,7520,7521,7558,7848,11361,11360,11574,15926,15905,15903,15001,15041,16729,15747,14737,4725,4366,4726,4963,4364,4947,4946,4961,4967,4994,4995,4345,4998,5031,4969,4422,4999,4968,4971,4344,5395,4365,4431,5028,4914,4840,4868,4719,4718,4720,4901,4899,5041,5067,5042,5066,5065,5087,5432,5381,5119,5117,5118,5162,5163,5331,4678,4677,4676,4996,4997,5161,5148,5190,5189,8330,8402,17014,8359,17765,8318,8344,8319,8369,8349,17786,6245,6243,6380,7847,7777,6379,463,461,462,464,519,523,526,10527,10498,10528,10499,6250,801,5562,5563,5564,7160,7216,7911,7161,7084,7192,7313,7312,6949,9042,9507,9041,9040,9137,9600,10556,9442,9078,10557,9006,9030,9375,10513,10864,10697,10696,10561,10822,5756,7062,5608,7393,5941,5942,5943,7394,7335,7333,7282,7281,7240,9226,9227,10250,10249,10411,9410,10806,10388,10386,9112,10387,9259,10417,10414,9258,10413,9257,10415,10860,9583,9305,10769,10248,10252,10251,10253,9899,10892,10254,8682,8681,8683,9415,9414,9413,10455,10305,10306,10307,10255,10218,10219,10302,10346,10304,8907,9011)
   InputName=()
   OutputName=(StringValue="Cuffs_SimSelection")
   RemoveIndices=()
   SelectionOverrideType=ReplaceAll
   SimTransferType=Use2DSimMesh
   Transfer=()
   TransferSelectionThreshold=0.95
  }
 }
 LOW ITERATION FOR TESTING SimulationSolverConfig {
  type=FChaosClothAssetSimulationSolverConfigNode
  guid=#63
  properties {
   MaxNumIterations=2
   NumIterations=2
   NumSelfCollisionSubsteps=10
   NumSubstepsImported=(ImportedValue=4)
   bEnableNumSelfCollisionSubsteps=True
  }
 }
 Remesh_0 {
  type=FChaosClothAssetRemeshNode_v2
  guid=#64
  inputPins [
   Collection {type=FManagedArrayCollection}
   TriangleSelectionSim.StringValue {type=FString}
   DensityMapSim.WeightMap {type=FString}
   TriangleSelectionRender.StringValue {type=FString}
   DensityMapRender.WeightMap {type=FString}
  ]
  outputPins [Collection {type=FManagedArrayCollection}]
  properties {
   DensityMapRender=(Low=100,High=200,WeightMap="DensityMapRender")
   DensityMapSim=(Low=200,High=200,WeightMap="DensityMapSim")
   DensityMapThresholdRender=1
   DensityMapThresholdSim=1
   IterationsRender=10
   IterationsSim=10
   RemeshMethodRender=Remesh
   RenderSeamRemeshIterations=1
   SelectionSourceRender=DensityMapInput
   SelectionSourceSim=DensityMapInput
   SmoothingRender=0.25
   SmoothingSim=1
   TargetPercentRender=100
   TriangleSelectionRender=()
   TriangleSelectionSim=()
   bRemeshRender=False
   bRemeshRenderSeams=False
   bRemeshSim=True
  }
 }
 ProxyDeformer {
  type=FChaosClothAssetProxyDeformerNode_v3
  guid=#65
  properties {InfluenceRadius=5, SelectionFilterSets=(()), bPreserveRenderTangents=False}
 }
 SetPhysicsAsset {
  type=FChaosClothAssetSetPhysicsAssetNode
  guid=#66
  properties {PhysicsAsset=/Script/Engine.PhysicsAsset'&/M_MED_SkippingClouds_MLDKineCollision'}
 }
 SimulationCollisionConfig {
  type=FChaosClothAssetSimulationCollisionConfigNode
  guid=#67
  properties {
   ClothCollisionThicknessImported=(bIsAnimatable=True,WeightMap="ClothCollisionThickness")
   CollisionThicknessImported=(ImportedValue=1)
   FrictionCoefficientWeighted=(bIsAnimatable=True,Low=0.8,High=0.8,WeightMap="FrictionCoefficient")
   InnerCollisionThickness=(WeightMap="InnerCollisionThickness")
   bUseCCD=True
  }
 }
]
